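# Sound cue for the bullet-impact sound of a Cube Queen hit ripple, using shared Cube impact attenuation.
# A quality-level switch sends the first two levels to a random pick of ten impact waves and the last two to a pick of four, each through a Modulator.
# A concurrency override allows one instance at a time, stopping the oldest with a short release.
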SoundCue /CorruptionGameplay/Gameplay/Environmental/Cube/Audio/CubeQueen_HitRipple_Cue
root=QualityLevel_0
nodes [
 {id=Modulator_2, children [Random_5]}
 {id=Modulator_3, children [Random_4]}
 {id=QualityLevel_0, children [Modulator_3,Modulator_3,Modulator_2,Modulator_2]}
 {
  id=Random_4
  children [
   WavePlayer_20
   WavePlayer_24
   WavePlayer_25
   WavePlayer_26
   WavePlayer_27
   WavePlayer_28
   WavePlayer_29
   WavePlayer_30
   WavePlayer_34
   WavePlayer_35
  ]
  properties {weights [1,1,1,1,1,1,1,1,1,1]}
 }
 {
  id=Random_5
  children [WavePlayer_36, WavePlayer_37, WavePlayer_38, WavePlayer_39]
  properties {bShouldExcludeFromBranchCulling=true, weights [1,1,1,1]}
 }
 {id=WavePlayer_20, properties {soundWaveAssetPtr=&/Cube_Queen_Bullet_Impact_01}}
 {id=WavePlayer_24, properties {soundWaveAssetPtr=&/Cube_Queen_Bullet_Impact_02}}
 {id=WavePlayer_25, properties {soundWaveAssetPtr=&/Cube_Queen_Bullet_Impact_03}}
 {id=WavePlayer_26, properties {soundWaveAssetPtr=&/Cube_Queen_Bullet_Impact_04}}
 {id=WavePlayer_27, properties {soundWaveAssetPtr=&/Cube_Queen_Bullet_Impact_05}}
 {id=WavePlayer_28, properties {soundWaveAssetPtr=&/Cube_Queen_Bullet_Impact_06}}
 {id=WavePlayer_29, properties {soundWaveAssetPtr=&/Cube_Queen_Bullet_Impact_07}}
 {id=WavePlayer_30, properties {soundWaveAssetPtr=&/Cube_Queen_Bullet_Impact_08}}
 {id=WavePlayer_34, properties {soundWaveAssetPtr=&/Cube_Queen_Bullet_Impact_09}}
 {id=WavePlayer_35, properties {soundWaveAssetPtr=&/Cube_Queen_Bullet_Impact_10}}
 {id=WavePlayer_36, properties {soundWaveAssetPtr=&Cube_Queen_Bullet_Impact_01}}
 {id=WavePlayer_37, properties {soundWaveAssetPtr=&Cube_Queen_Bullet_Impact_02}}
 {id=WavePlayer_38, properties {soundWaveAssetPtr=&Cube_Queen_Bullet_Impact_07}}
 {id=WavePlayer_39, properties {soundWaveAssetPtr=&Cube_Queen_Bullet_Impact_09}}
]
properties {
 attenuationSettings=/Game/Athena/Sounds/Events/Cube/Cube_Impacts_Att
 bOverrideConcurrency=true
 maxDistance=10000
 soundClassObject=&Fort_Mission_Ingame_Important_PSM
 volumeMultiplier=0.4
 concurrencyOverrides {maxCount=1, resolutionRule=StopOldest, voiceStealReleaseTime=0.15}
}
