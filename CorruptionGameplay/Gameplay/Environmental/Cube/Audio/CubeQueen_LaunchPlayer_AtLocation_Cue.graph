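# Sound cue for the Cube Queen player launch at a location, with no nodes or root, so it references no audio.
# It carries playback settings only: shared cube launch attenuation, the scripted events sound class, reduced volume and a long maximum distance.
# A concurrency override limits it to a single instance per owner and prevents new plays while one is active.
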SoundCue /CorruptionGameplay/Gameplay/Environmental/Cube/Audio/CubeQueen_LaunchPlayer_AtLocation_Cue
root=null
nodes []
properties {
 attenuationSettings=/Game/Athena/Sounds/Events/Cube/Cube_Launch_Att
 bOverrideConcurrency=true
 maxDistance=20000
 soundClassObject=/Game/Sounds/FortSoundClassesAndModes/Fort_Events_Scripted
 volumeMultiplier=0.5
 concurrencyOverrides {bLimitToOwner=true, maxCount=1, resolutionRule=PreventNew}
}
