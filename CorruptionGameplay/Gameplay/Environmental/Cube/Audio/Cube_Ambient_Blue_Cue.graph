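# Sound cue for the looping ambient audio of the blue Cube in the CorruptionGameplay environmental content.
# A single looping wave player plays the blue cube ambience loop through an attenuation node using a shared large cube ambient attenuation asset.
# The cue sits in the Fort_Gadgets sound class at reduced volume, with a set maximum distance and a stereo spread override.
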SoundCue /CorruptionGameplay/Gameplay/Environmental/Cube/Audio/Cube_Ambient_Blue_Cue
root=Attenuation_0
nodes [
 {
  id=Attenuation_0
  children [WavePlayer_1]
  properties {attenuationSettings=&/Cube_Ambient_Large_Att}
 }
 {id=WavePlayer_1, properties {bLooping=true, soundWaveAssetPtr=&/Cube_Blue_Amb_Loop_01}}
]
properties {
 maxDistance=4000
 soundClassObject=&Fort_Gadgets
 volumeMultiplier=0.4
 attenuationOverrides {stereoSpread=512}
}
