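# Cube_Ambient_Cue is a looping ambient sound cue for the Corruption cube, with large, medium and small variants.
# An integer Size parameter drives a Switch that picks a branch, each with its own attenuation asset; the default input is empty.
# All branches share one looping wave player, with fixed-value Modulators raising pitch for medium and further for small, so one wave serves every size.
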SoundCue /CorruptionGameplay/Gameplay/Environmental/Cube/Audio/Cube_Ambient_Cue
root=Switch_0
nodes [
 {
  id=Attenuation_0
  children [WavePlayer_1]
  properties {attenuationSettings=&/Cube_Ambient_Large_Att}
 }
 {
  id=Attenuation_1
  children [Modulator_1]
  properties {attenuationSettings=&/Cube_Ambient_Medium_Att}
 }
 {
  id=Attenuation_2
  children [Modulator_0]
  properties {attenuationSettings=&/Cube_Ambient_Small_Att}
 }
 {id=Modulator_0, children [WavePlayer_1], properties {pitchMax=1.5, pitchMin=1.5, volumeMax=1, volumeMin=1}}
 {id=Modulator_1, children [WavePlayer_1], properties {pitchMax=1.25, pitchMin=1.25, volumeMax=1, volumeMin=1}}
 {id=Switch_0, children [null, Attenuation_0, Attenuation_1, Attenuation_2], properties {intParameterName=Size}}
 {id=WavePlayer_1, properties {bLooping=true, soundWaveAssetPtr=&/Cube_Purple_Amb_Loop_01}}
]
properties {
 maxDistance=4000
 soundClassObject=&Fort_Gadgets
 volumeMultiplier=0.6
 attenuationOverrides {stereoSpread=512}
}
