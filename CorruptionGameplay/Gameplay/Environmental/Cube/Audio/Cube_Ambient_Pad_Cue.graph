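# Sound cue that plays the looping ambient sound of a Cube bounce pad in the Corruption gameplay environment.
# A Random node picks one of two equally weighted looping wave players, then feeds a Modulator left at default settings.
# The cue plays at reduced volume in the Fort_Gadgets sound class, using the small-pad attenuation asset for spatial falloff.
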SoundCue /CorruptionGameplay/Gameplay/Environmental/Cube/Audio/Cube_Ambient_Pad_Cue
root=Modulator_0
nodes [
 {id=Modulator_0, children [Random_0]}
 {id=Random_0, children [WavePlayer_0,WavePlayer_1], properties {weights [1,1]}}
 {
  id=WavePlayer_0
  properties {bLooping=true, soundWaveAssetPtr=&/AMB_Cubetown_BouncePad_Loop_01}
 }
 {
  id=WavePlayer_1
  properties {bLooping=true, soundWaveAssetPtr=&/AMB_Cubetown_BouncePad_Loop_02}
 }
]
properties {
 attenuationSettings=&/Cube_Ambient_Small_Pad_Att
 maxDistance=1536
 soundClassObject=&Fort_Gadgets
 volumeMultiplier=0.4
 attenuationOverrides {stereoSpread=512}
}
