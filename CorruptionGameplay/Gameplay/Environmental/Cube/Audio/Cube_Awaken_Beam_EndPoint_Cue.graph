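# Sound cue for the end point of the Cube's awaken beam in the Corruption gameplay environmental content.
# A single wave player feeds the purple beam start wave into a mixer that boosts its volume.
# It uses the Cube awaken attenuation and the important in-game mission sound class, with a concurrency override that caps simultaneous instances at two.
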SoundCue /CorruptionGameplay/Gameplay/Environmental/Cube/Audio/Cube_Awaken_Beam_EndPoint_Cue
root=Mixer_0
nodes [
 {id=Mixer_0, children [WavePlayer_5], properties {inputVolume [1.5]}}
 {id=WavePlayer_5, properties {soundWaveAssetPtr=&/Cube_Awaken_Beam_Start_Purple_01}}
]
properties {
 attenuationSettings=&/Cube_Awaken_Att
 bOverrideConcurrency=true
 maxDistance=26000
 soundClassObject=&Fort_Mission_Ingame_Important_PSM
 volumeMultiplier=1
 concurrencyOverrides {maxCount=2}
}
