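# Sound cue for the looping audio at the end point of the Cube's awaken beam.
# A fixed one-second delay leads into an Enveloper, which fades a looping purple beam wave up from silence over two seconds.
# It uses awaken-loop attenuation, an important in-game mission sound class, and a concurrency override that caps simultaneous instances.
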SoundCue /CorruptionGameplay/Gameplay/Environmental/Cube/Audio/Cube_Awaken_Beam_Loop_EndPoint_Cue
root=Delay_1
nodes [
 {id=Delay_1, children [Enveloper_0], properties {delayMax=1, delayMin=1}}
 {
  id=Enveloper_0
  children [WavePlayer_4]
  properties {pitchCurve {editorCurveData {keys [{time=0}]}}, volumeCurve {editorCurveData {keys [<at 0 0>,<at 2 1>]}}}
 }
 {
  id=WavePlayer_4
  properties {bLooping=true, soundWaveAssetPtr=&/Cube_Awaken_Beam_Loop_Purple_01}
 }
]
properties {
 attenuationSettings=&/Cube_Awaken_Loop_Att
 bOverrideConcurrency=true
 maxDistance=12000
 soundClassObject=&Fort_Mission_Ingame_Important_PSM
 volumeMultiplier=1
 concurrencyOverrides {maxCount=2}
}
