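# Sound cue for the start-point loop of the Cube awaken beam in CorruptionGameplay, layering a looping beam wave with a looping hover wave.
# A mixer blends the beam, delayed one second, with the hover loop, each shaped by its own volume enveloper.
# The beam rises quickly, holds, then fades out as the hover slowly rises, forming a long crossfade.
# It uses the Cube awaken loop attenuation, an important-mission sound class and overridden concurrency.
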SoundCue /CorruptionGameplay/Gameplay/Environmental/Cube/Audio/Cube_Awaken_Beam_Loop_StartPoint_Cue
root=Mixer_1
nodes [
 {id=Delay_1, children [Enveloper_0], properties {delayMax=1, delayMin=1}}
 {
  id=Enveloper_0
  children [WavePlayer_0]
  properties {pitchCurve {editorCurveData {keys [{time=0}]}}, volumeCurve {editorCurveData {keys [<at 0 0>,<at 2 1>,<at 8 1>,<at 25 0>]}}}
 }
 {
  id=Enveloper_1
  children [WavePlayer_1]
  properties {pitchCurve {editorCurveData {keys [{time=0}]}}, volumeCurve {editorCurveData {keys [<at 0 0>,<at 25 1>]}}}
 }
 {id=Mixer_1, children [Delay_1,Enveloper_1], properties {inputVolume [0.6,0.7]}}
 {
  id=WavePlayer_0
  properties {bLooping=true, soundWaveAssetPtr=&/Cube_Awaken_Beam_Loop_Gold_01}
 }
 {
  id=WavePlayer_1
  properties {bLooping=true, soundWaveAssetPtr=&/Cube_Awaken_Hover_Loop_Gold_02}
 }
]
properties {
 attenuationSettings=&/Cube_Awaken_Loop_Att
 bOverrideConcurrency=true
 maxDistance=12000
 soundClassObject=&Fort_Mission_Ingame_Important_PSM
 volumeMultiplier=1
 concurrencyOverrides {maxCount=2}
}
