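# Sound cue for the stop of the Cube awaken beam loop at its start point, in the Corruption gameplay environmental audio.
# A single wave player feeds the Cube_Awaken_Beam_Stop_Gold_01 sound straight to the output.
# It uses the Cube awaken attenuation and an important in-game mission sound class, and overrides concurrency to cap simultaneous instances.
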SoundCue /CorruptionGameplay/Gameplay/Environmental/Cube/Audio/Cube_Awaken_Beam_Loop_Stop_StartPoint_Cue
root=WavePlayer_0
nodes [{id=WavePlayer_0, properties {soundWaveAssetPtr=&/Cube_Awaken_Beam_Stop_Gold_01}}]
properties {
 attenuationSettings=&/Cube_Awaken_Att
 bOverrideConcurrency=true
 maxDistance=26000
 soundClassObject=&Fort_Mission_Ingame_Important_PSM
 volumeMultiplier=1
 concurrencyOverrides {maxCount=2}
}
