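# Sound cue for the Cube's awaken hover audio, layering a one-shot start sound with a sustained hover loop.
# A mixer plays the start wave at full level alongside the looping wave, which is mixed lower and begins after a short delay.
# An enveloper above the delay fades the loop in, holds it, then eases it down slightly.
# The cue uses Cube awaken attenuation, an important in-game mission sound class and a tight concurrency override.
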SoundCue /CorruptionGameplay/Gameplay/Environmental/Cube/Audio/Cube_Awaken_Hover_Loop_Cue
root=Mixer_0
nodes [
 {id=Delay_0, children [WavePlayer_2], properties {delayMax=0.5, delayMin=0.5}}
 {
  id=Enveloper_0
  children [Delay_0]
  properties {pitchCurve {editorCurveData {keys [{time=0}]}}, volumeCurve {editorCurveData {keys [<at 0 0>,<at 2 1>,<at 17 1>,<at 19 0.7>]}}}
 }
 {id=Mixer_0, children [WavePlayer_0,Enveloper_0], properties {inputVolume [1,0.6]}}
 {id=WavePlayer_0, properties {soundWaveAssetPtr=&/Cube_Awaken_Hover_Start_Gold_01}}
 {
  id=WavePlayer_2
  properties {bLooping=true, soundWaveAssetPtr=&/Cube_Awaken_Hover_Loop_Gold_01}
 }
]
properties {
 attenuationSettings=&/Cube_Awaken_Att
 bOverrideConcurrency=true
 maxDistance=26000
 soundClassObject=&Fort_Mission_Ingame_Important_PSM
 volumeMultiplier=1
 concurrencyOverrides {maxCount=2}
}
