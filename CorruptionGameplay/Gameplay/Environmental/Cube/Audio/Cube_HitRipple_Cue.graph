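# Sound cue for bullet-impact hit ripples on the Cube, with standard and blue sound variants.
# A quality-level node routes its first two slots to ten-sample random pools and its last two to three-sample pools.
# In each tier, a CubeIndex switch selects blue for index 0 and standard otherwise, with each pool passing through a modulator.
# Attenuation and a stop-oldest concurrency override limit overlapping hits.
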SoundCue /CorruptionGameplay/Gameplay/Environmental/Cube/Audio/Cube_HitRipple_Cue
root=QualityLevel_0
nodes [
 {id=Modulator_0, children [Random_1]}
 {id=Modulator_1, children [Random_2]}
 {id=Modulator_2, children [Random_3]}
 {id=Modulator_4, children [Random_0]}
 {id=QualityLevel_0, children [Switch_0,Switch_0,Switch_1,Switch_1]}
 {
  id=Random_0
  children [
   WavePlayer_0
   WavePlayer_2
   WavePlayer_3
   WavePlayer_4
   WavePlayer_9
   WavePlayer_10
   WavePlayer_11
   WavePlayer_12
   WavePlayer_13
   WavePlayer_14
  ]
  properties {weights [1,1,1,1,1,1,1,1,1,1]}
 }
 {
  id=Random_1
  children [
   WavePlayer_1
   WavePlayer_5
   WavePlayer_6
   WavePlayer_7
   WavePlayer_8
   WavePlayer_15
   WavePlayer_16
   WavePlayer_17
   WavePlayer_18
   WavePlayer_19
  ]
  properties {weights [1,1,1,1,1,1,1,1,1,1]}
 }
 {id=Random_2, children [WavePlayer_21, WavePlayer_22, WavePlayer_23], properties {bShouldExcludeFromBranchCulling=true, weights [1,1,1]}}
 {id=Random_3, children [WavePlayer_31, WavePlayer_32, WavePlayer_33], properties {bShouldExcludeFromBranchCulling=true, weights [1,1,1]}}
 {id=Switch_0, children [Modulator_4,Modulator_0,Modulator_4,Modulator_4], properties {intParameterName=CubeIndex}}
 {id=Switch_1, children [Modulator_1,Modulator_2,Modulator_1,Modulator_1], properties {intParameterName=CubeIndex}}
 {id=WavePlayer_0, properties {soundWaveAssetPtr=&/Cube_Impact_Bullet_10}}
 {id=WavePlayer_1, properties {soundWaveAssetPtr=&/Cube_Blue_Impact_Bullet_10}}
 {id=WavePlayer_10, properties {soundWaveAssetPtr=&/Cube_Impact_Bullet_05}}
 {id=WavePlayer_11, properties {soundWaveAssetPtr=&/Cube_Impact_Bullet_06}}
 {id=WavePlayer_12, properties {soundWaveAssetPtr=&/Cube_Impact_Bullet_07}}
 {id=WavePlayer_13, properties {soundWaveAssetPtr=&/Cube_Impact_Bullet_08}}
 {id=WavePlayer_14, properties {soundWaveAssetPtr=&/Cube_Impact_Bullet_09}}
 {id=WavePlayer_15, properties {soundWaveAssetPtr=&/Cube_Blue_Impact_Bullet_05}}
 {id=WavePlayer_16, properties {soundWaveAssetPtr=&/Cube_Blue_Impact_Bullet_06}}
 {id=WavePlayer_17, properties {soundWaveAssetPtr=&/Cube_Blue_Impact_Bullet_07}}
 {id=WavePlayer_18, properties {soundWaveAssetPtr=&/Cube_Blue_Impact_Bullet_08}}
 {id=WavePlayer_19, properties {soundWaveAssetPtr=&/Cube_Blue_Impact_Bullet_09}}
 {id=WavePlayer_2, properties {soundWaveAssetPtr=&/Cube_Impact_Bullet_01}}
 {id=WavePlayer_21, properties {soundWaveAssetPtr=&Cube_Impact_Bullet_01}}
 {id=WavePlayer_22, properties {soundWaveAssetPtr=&/Cube_Impact_Bullet_02}}
 {id=WavePlayer_23, properties {soundWaveAssetPtr=&/Cube_Impact_Bullet_03}}
 {id=WavePlayer_3, properties {soundWaveAssetPtr=&Cube_Impact_Bullet_02}}
 {id=WavePlayer_31, properties {soundWaveAssetPtr=&/Cube_Blue_Impact_Bullet_01}}
 {id=WavePlayer_32, properties {soundWaveAssetPtr=&/Cube_Blue_Impact_Bullet_02}}
 {id=WavePlayer_33, properties {soundWaveAssetPtr=&/Cube_Blue_Impact_Bullet_03}}
 {id=WavePlayer_4, properties {soundWaveAssetPtr=&Cube_Impact_Bullet_03}}
 {id=WavePlayer_5, properties {soundWaveAssetPtr=&Cube_Blue_Impact_Bullet_01}}
 {id=WavePlayer_6, properties {soundWaveAssetPtr=&Cube_Blue_Impact_Bullet_02}}
 {id=WavePlayer_7, properties {soundWaveAssetPtr=&Cube_Blue_Impact_Bullet_03}}
 {id=WavePlayer_8, properties {soundWaveAssetPtr=&/Cube_Blue_Impact_Bullet_04}}
 {id=WavePlayer_9, properties {soundWaveAssetPtr=&/Cube_Impact_Bullet_04}}
]
properties {
 attenuationSettings=/Game/Athena/Sounds/Events/Cube/Cube_Impacts_Att
 bOverrideConcurrency=true
 maxDistance=10000
 soundClassObject=&Fort_Mission_Ingame_Important_PSM
 volumeMultiplier=0.4
 concurrencyOverrides {maxCount=2, resolutionRule=StopOldest, volumeScale=0.5, volumeScaleAttackTime=0.1}
}
