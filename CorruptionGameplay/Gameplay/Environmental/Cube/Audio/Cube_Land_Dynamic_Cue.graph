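# Sound cue for a cube landing, layering an impact SFX with a randomized vocal reaction delayed by a quarter second.
# An integer Size parameter selects large, medium or small layers, each impact with slight random pitch variation.
# Only the large branch checks the Color parameter, which can swap in gold vocals.
# Quality slots split between five-variant and two-variant vocal pools, and concurrency is capped at two instances.
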SoundCue /CorruptionGameplay/Gameplay/Environmental/Cube/Audio/Cube_Land_Dynamic_Cue
root=QualityLevel_0
nodes [
 {id=Delay_0, children [Switch_2], properties {delayMax=0.25, delayMin=0.25}}
 {id=Delay_1, children [Random_1], properties {delayMax=0.25, delayMin=0.25}}
 {id=Delay_2, children [Random_2], properties {delayMax=0.25, delayMin=0.25}}
 {id=Delay_3, children [Switch_3], properties {delayMax=0.25, delayMin=0.25}}
 {id=Delay_4, children [Random_5], properties {delayMax=0.25, delayMin=0.25}}
 {id=Delay_5, children [Random_6], properties {delayMax=0.25, delayMin=0.25}}
 {id=Mixer_0, children [Modulator_2,Delay_0], properties {inputVolume [1.5,1]}}
 {id=Mixer_1, children [Modulator_1,Delay_1], properties {inputVolume [1.5,1]}}
 {id=Mixer_2, children [Modulator_5,Delay_3], properties {inputVolume [1.5,1]}}
 {id=Mixer_3, children [Modulator_0,Delay_2], properties {inputVolume [1.5,1]}}
 {id=Mixer_4, children [Modulator_4,Delay_4], properties {inputVolume [1.5,1]}}
 {id=Mixer_5, children [Modulator_3,Delay_5], properties {inputVolume [1.5,1]}}
 {id=Modulator_0, children [WavePlayer_2], properties {pitchMax=1.2, pitchMin=1, volumeMax=1, volumeMin=1}}
 {id=Modulator_1, children [WavePlayer_1], properties {pitchMax=1.1, pitchMin=1, volumeMax=1, volumeMin=1}}
 {id=Modulator_2, children [WavePlayer_0], properties {pitchMax=1, pitchMin=0.9, volumeMax=1, volumeMin=1}}
 {id=Modulator_3, children [WavePlayer_25], properties {pitchMax=1.2, pitchMin=1, volumeMax=1, volumeMin=1}}
 {id=Modulator_4, children [WavePlayer_24], properties {pitchMax=1.1, pitchMin=1, volumeMax=1, volumeMin=1}}
 {id=Modulator_5, children [WavePlayer_23], properties {pitchMax=1, pitchMin=0.9, volumeMax=1, volumeMin=1}}
 {id=QualityLevel_0, children [Switch_0,Switch_0,Switch_1,Switch_1]}
 {id=Random_0, children [WavePlayer_3,WavePlayer_5,WavePlayer_6,WavePlayer_7,WavePlayer_8], properties {weights [1,1,1,1,1]}}
 {id=Random_1, children [WavePlayer_9, WavePlayer_10, WavePlayer_11, WavePlayer_12, WavePlayer_13], properties {weights [1,1,1,1,1]}}
 {id=Random_2, children [WavePlayer_14, WavePlayer_15, WavePlayer_16, WavePlayer_17, WavePlayer_18], properties {weights [1,1,1,1,1]}}
 {id=Random_3, children [WavePlayer_4, WavePlayer_19, WavePlayer_20, WavePlayer_21, WavePlayer_22], properties {weights [1,1,1,1,1]}}
 {id=Random_4, children [WavePlayer_27, WavePlayer_28], properties {bShouldExcludeFromBranchCulling=true, weights [1,1]}}
 {id=Random_5, children [WavePlayer_31, WavePlayer_32], properties {bShouldExcludeFromBranchCulling=true, weights [1,1]}}
 {id=Random_6, children [WavePlayer_36, WavePlayer_37], properties {bShouldExcludeFromBranchCulling=true, weights [1,1]}}
 {id=Random_7, children [WavePlayer_41, WavePlayer_42], properties {bShouldExcludeFromBranchCulling=true, weights [1,1]}}
 {id=Switch_0, children [Mixer_1,Mixer_0,Mixer_1,Mixer_3], properties {intParameterName=Size}}
 {id=Switch_1, children [Mixer_4,Mixer_2,Mixer_4,Mixer_5], properties {intParameterName=Size}}
 {id=Switch_2, children [Random_0,Random_0,Random_0,Random_3], properties {intParameterName=Color}}
 {id=Switch_3, children [Random_4,Random_4,Random_4,Random_7], properties {intParameterName=Color}}
 {id=WavePlayer_0, properties {soundWaveAssetPtr=&/Cube_SFX_Land_01}}
 {id=WavePlayer_1, properties {soundWaveAssetPtr=&/Cube_SFX_Land_02}}
 {id=WavePlayer_10, properties {soundWaveAssetPtr=&/Cube_Vox_Land_Med_02}}
 {id=WavePlayer_11, properties {soundWaveAssetPtr=&/Cube_Vox_Land_Med_03}}
 {id=WavePlayer_12, properties {soundWaveAssetPtr=&/Cube_Vox_Land_Med_04}}
 {id=WavePlayer_13, properties {soundWaveAssetPtr=&/Cube_Vox_Land_Med_05}}
 {id=WavePlayer_14, properties {soundWaveAssetPtr=&/Cube_Vox_Land_Sm_01}}
 {id=WavePlayer_15, properties {soundWaveAssetPtr=&/Cube_Vox_Land_Sm_02}}
 {id=WavePlayer_16, properties {soundWaveAssetPtr=&/Cube_Vox_Land_Sm_03}}
 {id=WavePlayer_17, properties {soundWaveAssetPtr=&/Cube_Vox_Land_Sm_04}}
 {id=WavePlayer_18, properties {soundWaveAssetPtr=&/Cube_Vox_Land_Sm_05}}
 {id=WavePlayer_19, properties {soundWaveAssetPtr=&/Cube_Vox_Land_Gold_02}}
 {id=WavePlayer_2, properties {soundWaveAssetPtr=&/Cube_SFX_Land_03}}
 {id=WavePlayer_20, properties {soundWaveAssetPtr=&/Cube_Vox_Land_Gold_03}}
 {id=WavePlayer_21, properties {soundWaveAssetPtr=&/Cube_Vox_Land_Gold_04}}
 {id=WavePlayer_22, properties {soundWaveAssetPtr=&/Cube_Vox_Land_Gold_05}}
 {id=WavePlayer_23, properties {soundWaveAssetPtr=&Cube_SFX_Land_01}}
 {id=WavePlayer_24, properties {soundWaveAssetPtr=&Cube_SFX_Land_02}}
 {id=WavePlayer_25, properties {soundWaveAssetPtr=&Cube_SFX_Land_03}}
 {id=WavePlayer_27, properties {soundWaveAssetPtr=&/Cube_Vox_Land_Lg_01}}
 {id=WavePlayer_28, properties {soundWaveAssetPtr=&/Cube_Vox_Land_Lg_02}}
 {id=WavePlayer_3, properties {soundWaveAssetPtr=&/Cube_Vox_Land_Lg_05}}
 {id=WavePlayer_31, properties {soundWaveAssetPtr=&/Cube_Vox_Land_Med_01}}
 {id=WavePlayer_32, properties {soundWaveAssetPtr=&Cube_Vox_Land_Med_02}}
 {id=WavePlayer_36, properties {soundWaveAssetPtr=&Cube_Vox_Land_Sm_01}}
 {id=WavePlayer_37, properties {soundWaveAssetPtr=&Cube_Vox_Land_Sm_02}}
 {id=WavePlayer_4, properties {soundWaveAssetPtr=&/Cube_Vox_Land_Gold_01}}
 {id=WavePlayer_41, properties {soundWaveAssetPtr=&Cube_Vox_Land_Gold_01}}
 {id=WavePlayer_42, properties {soundWaveAssetPtr=&Cube_Vox_Land_Gold_02}}
 {id=WavePlayer_5, properties {soundWaveAssetPtr=&Cube_Vox_Land_Lg_01}}
 {id=WavePlayer_6, properties {soundWaveAssetPtr=&Cube_Vox_Land_Lg_02}}
 {id=WavePlayer_7, properties {soundWaveAssetPtr=&/Cube_Vox_Land_Lg_03}}
 {id=WavePlayer_8, properties {soundWaveAssetPtr=&/Cube_Vox_Land_Lg_04}}
 {id=WavePlayer_9, properties {soundWaveAssetPtr=&Cube_Vox_Land_Med_01}}
]
properties {
 attenuationSettings=&/Cube_Roll_Att
 bOverrideConcurrency=true
 maxDistance=20000
 soundClassObject=&Fort_Mission_Ingame_Important_PSM
 volumeMultiplier=1
 concurrencyOverrides {maxCount=2}
}
