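# Sound cue for the gold Cube landing, layering a landing SFX with a delayed gold Cube vox line, using Cube roll attenuation and a tight concurrency limit.
# Each mixer blends the slightly pitch-randomized, louder SFX with a random vox pick after a fixed short delay.
# A quality switch sends its first two levels to a five-variation vox pool and its last two to a two-variation pool.
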SoundCue /CorruptionGameplay/Gameplay/Environmental/Cube/Audio/Cube_Land_Gold_Cue
root=QualityLevel_0
nodes [
 {id=Delay_0, children [Random_3], properties {delayMax=0.25, delayMin=0.25}}
 {id=Delay_1, children [Random_0], properties {delayMax=0.25, delayMin=0.25}}
 {id=Mixer_0, children [Modulator_2,Delay_0], properties {inputVolume [1.5,1]}}
 {id=Mixer_1, children [Modulator_0,Delay_1], properties {inputVolume [1.5,1]}}
 {id=Modulator_0, children [WavePlayer_1], properties {pitchMax=1.1, pitchMin=0.9, volumeMax=1, volumeMin=1}}
 {id=Modulator_2, children [WavePlayer_0], properties {pitchMax=1.1, pitchMin=0.9, volumeMax=1, volumeMin=1}}
 {id=QualityLevel_0, children [Mixer_0,Mixer_0,Mixer_1,Mixer_1]}
 {id=Random_0, children [WavePlayer_2,WavePlayer_3], properties {bShouldExcludeFromBranchCulling=true, weights [1,1]}}
 {id=Random_3, children [WavePlayer_4, WavePlayer_19, WavePlayer_20, WavePlayer_21, WavePlayer_22], properties {weights [1,1,1,1,1]}}
 {id=WavePlayer_0, properties {soundWaveAssetPtr=&/Cube_SFX_Land_01}}
 {id=WavePlayer_1, properties {soundWaveAssetPtr=&Cube_SFX_Land_01}}
 {id=WavePlayer_19, properties {soundWaveAssetPtr=&/Cube_Vox_Land_Gold_02}}
 {id=WavePlayer_2, properties {soundWaveAssetPtr=&/Cube_Vox_Land_Gold_01}}
 {id=WavePlayer_20, properties {soundWaveAssetPtr=&/Cube_Vox_Land_Gold_03}}
 {id=WavePlayer_21, properties {soundWaveAssetPtr=&/Cube_Vox_Land_Gold_04}}
 {id=WavePlayer_22, properties {soundWaveAssetPtr=&/Cube_Vox_Land_Gold_05}}
 {id=WavePlayer_3, properties {soundWaveAssetPtr=&Cube_Vox_Land_Gold_02}}
 {id=WavePlayer_4, properties {soundWaveAssetPtr=&Cube_Vox_Land_Gold_01}}
]
properties {
 attenuationSettings=&/Cube_Roll_Att
 bOverrideConcurrency=true
 maxDistance=20000
 soundClassObject=&Fort_Mission_Ingame_Important_PSM
 volumeMultiplier=1
 concurrencyOverrides {maxCount=2}
}
