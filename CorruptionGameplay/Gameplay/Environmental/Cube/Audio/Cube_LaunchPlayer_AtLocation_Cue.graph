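# Sound cue for the cube's launch-player bounce vocal, at half volume with event attenuation.
# A CubeIndex switch selects Blue cube vocals for one slot and standard cube vocals for the others, each through a modulator.
# Within each set, quality level picks between six random variations and a two-variation set with its own modulator for the later levels.
# Concurrency allows one instance per owner and blocks new plays meanwhile.
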SoundCue /CorruptionGameplay/Gameplay/Environmental/Cube/Audio/Cube_LaunchPlayer_AtLocation_Cue
root=Switch_0
nodes [
 {id=Modulator_1, children [QualityLevel_0]}
 {id=Modulator_2, children [Random_1]}
 {id=Modulator_4, children [QualityLevel_2]}
 {id=Modulator_5, children [Random_5]}
 {id=QualityLevel_0, children [Random_0,Random_0,Modulator_2,Modulator_2]}
 {id=QualityLevel_2, children [Random_6,Random_6,Modulator_5,Modulator_5]}
 {id=Random_0, children [WavePlayer_0,WavePlayer_1,WavePlayer_2,WavePlayer_3,WavePlayer_4,WavePlayer_5], properties {weights [1,1,1,1,1,1]}}
 {id=Random_1, children [WavePlayer_10, WavePlayer_11], properties {bShouldExcludeFromBranchCulling=true, weights [1,1]}}
 {id=Random_5, children [WavePlayer_30, WavePlayer_31], properties {bShouldExcludeFromBranchCulling=true, weights [1,1]}}
 {
  id=Random_6
  children [WavePlayer_24, WavePlayer_25, WavePlayer_26, WavePlayer_27, WavePlayer_28, WavePlayer_29]
  properties {weights [1,1,1,1,1,1]}
 }
 {id=Switch_0, children [Modulator_1,Modulator_4,Modulator_1,Modulator_1], properties {intParameterName=CubeIndex}}
 {id=WavePlayer_0, properties {soundWaveAssetPtr=&/Cube_Bounce_Vox_01}}
 {id=WavePlayer_1, properties {soundWaveAssetPtr=&/Cube_Bounce_Vox_02}}
 {id=WavePlayer_10, properties {soundWaveAssetPtr=&Cube_Bounce_Vox_01}}
 {id=WavePlayer_11, properties {soundWaveAssetPtr=&Cube_Bounce_Vox_02}}
 {id=WavePlayer_2, properties {soundWaveAssetPtr=&/Cube_Bounce_Vox_03}}
 {id=WavePlayer_24, properties {soundWaveAssetPtr=&/Cube_Blue_Bounce_Vox_06}}
 {id=WavePlayer_25, properties {soundWaveAssetPtr=&/Cube_Blue_Bounce_Vox_01}}
 {id=WavePlayer_26, properties {soundWaveAssetPtr=&/Cube_Blue_Bounce_Vox_02}}
 {id=WavePlayer_27, properties {soundWaveAssetPtr=&/Cube_Blue_Bounce_Vox_03}}
 {id=WavePlayer_28, properties {soundWaveAssetPtr=&/Cube_Blue_Bounce_Vox_04}}
 {id=WavePlayer_29, properties {soundWaveAssetPtr=&/Cube_Blue_Bounce_Vox_05}}
 {id=WavePlayer_3, properties {soundWaveAssetPtr=&/Cube_Bounce_Vox_04}}
 {id=WavePlayer_30, properties {soundWaveAssetPtr=&Cube_Blue_Bounce_Vox_01}}
 {id=WavePlayer_31, properties {soundWaveAssetPtr=&Cube_Blue_Bounce_Vox_02}}
 {id=WavePlayer_4, properties {soundWaveAssetPtr=&/Cube_Bounce_Vox_05}}
 {id=WavePlayer_5, properties {soundWaveAssetPtr=&/Cube_Bounce_Vox_06}}
]
properties {
 attenuationSettings=/Game/Athena/Sounds/Events/Cube/Cube_Launch_Att
 bOverrideConcurrency=true
 maxDistance=20000
 soundClassObject=/Game/Sounds/FortSoundClassesAndModes/Fort_Events_Scripted
 volumeMultiplier=0.5
 concurrencyOverrides {bLimitToOwner=true, maxCount=1, resolutionRule=PreventNew}
}
