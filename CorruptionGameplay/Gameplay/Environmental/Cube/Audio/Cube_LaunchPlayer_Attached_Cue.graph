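# Sound cue for a cube launching the player, with attenuation and a one-instance-per-owner limit that stops the oldest.
# A CubeIndex switch picks the Blue bounce set for one input and the standard set for all others, each through a modulator.
# Quality-level nodes choose between six random waves and a reduced two-wave set with wider pitch variation, whose random node is excluded from branch culling.
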SoundCue /CorruptionGameplay/Gameplay/Environmental/Cube/Audio/Cube_LaunchPlayer_Attached_Cue
root=Switch_0
nodes [
 {id=Modulator_0, children [QualityLevel_1]}
 {id=Modulator_1, children [QualityLevel_0]}
 {id=Modulator_2, children [Random_4], properties {pitchMax=1.1, pitchMin=0.9}}
 {id=Modulator_3, children [Random_6], properties {pitchMax=1.1, pitchMin=0.9}}
 {id=QualityLevel_0, children [Random_2,Random_2,Modulator_2,Modulator_2]}
 {id=QualityLevel_1, children [Random_1,Random_1,Modulator_3,Modulator_3]}
 {
  id=Random_1
  children [WavePlayer_6, WavePlayer_7, WavePlayer_8, WavePlayer_9, WavePlayer_10, WavePlayer_11]
  properties {weights [1,1,1,1,1,1]}
 }
 {
  id=Random_2
  children [WavePlayer_16, WavePlayer_17, WavePlayer_18, WavePlayer_19, WavePlayer_20, WavePlayer_21]
  properties {weights [1,1,1,1,1,1]}
 }
 {id=Random_4, children [WavePlayer_22, WavePlayer_23], properties {bShouldExcludeFromBranchCulling=true, weights [1,1]}}
 {id=Random_6, children [WavePlayer_12, WavePlayer_13], properties {bShouldExcludeFromBranchCulling=true, weights [1,1]}}
 {id=Switch_0, children [Modulator_0,Modulator_1,Modulator_0,Modulator_0], properties {intParameterName=CubeIndex}}
 {id=WavePlayer_10, properties {soundWaveAssetPtr=&/Cube_Bounce_Player_05}}
 {id=WavePlayer_11, properties {soundWaveAssetPtr=&/Cube_Bounce_Player_06}}
 {id=WavePlayer_12, properties {soundWaveAssetPtr=&/Cube_Bounce_Player_01}}
 {id=WavePlayer_13, properties {soundWaveAssetPtr=&/Cube_Bounce_Player_02}}
 {id=WavePlayer_16, properties {soundWaveAssetPtr=&/Cube_Blue_Bounce_Player_01}}
 {id=WavePlayer_17, properties {soundWaveAssetPtr=&/Cube_Blue_Bounce_Player_02}}
 {id=WavePlayer_18, properties {soundWaveAssetPtr=&/Cube_Blue_Bounce_Player_03}}
 {id=WavePlayer_19, properties {soundWaveAssetPtr=&/Cube_Blue_Bounce_Player_04}}
 {id=WavePlayer_20, properties {soundWaveAssetPtr=&/Cube_Blue_Bounce_Player_05}}
 {id=WavePlayer_21, properties {soundWaveAssetPtr=&/Cube_Blue_Bounce_Player_06}}
 {id=WavePlayer_22, properties {soundWaveAssetPtr=&Cube_Blue_Bounce_Player_01}}
 {id=WavePlayer_23, properties {soundWaveAssetPtr=&Cube_Blue_Bounce_Player_02}}
 {id=WavePlayer_6, properties {soundWaveAssetPtr=&Cube_Bounce_Player_01}}
 {id=WavePlayer_7, properties {soundWaveAssetPtr=&Cube_Bounce_Player_02}}
 {id=WavePlayer_8, properties {soundWaveAssetPtr=&/Cube_Bounce_Player_03}}
 {id=WavePlayer_9, properties {soundWaveAssetPtr=&/Cube_Bounce_Player_04}}
]
properties {
 attenuationSettings=/Game/Athena/Sounds/Events/Cube/Cube_Launch_Att
 bOverrideConcurrency=true
 maxDistance=20000
 soundClassObject=/Game/Sounds/FortSoundClassesAndModes/Fort_Events_Scripted
 volumeMultiplier=0.4
 concurrencyOverrides {bLimitToOwner=true, maxCount=1, resolutionRule=StopOldest}
}
