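# Sound cue for the Corruption gameplay Cube's rolling sound, layering a modulated roll SFX with a random vox clip under shared roll attenuation.
# A Size parameter switches between small, medium and large layers, with medium as default.
# Within large, a Color parameter can swap in gold vox.
# The quality node gives the first two levels five vox variations and the last two only two, and a concurrency override limits simultaneous instances.
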SoundCue /CorruptionGameplay/Gameplay/Environmental/Cube/Audio/Cube_Roll_Dynamic_Cue
root=QualityLevel_0
nodes [
 {id=Attenuation_0, children [Mixer_1], properties {attenuationSettings=&/Cube_Roll_Att}}
 {id=Attenuation_1, children [Mixer_2], properties {attenuationSettings=&Cube_Roll_Att}}
 {id=Attenuation_2, children [Mixer_3], properties {attenuationSettings=&Cube_Roll_Att}}
 {id=Attenuation_3, children [Mixer_0], properties {attenuationSettings=&Cube_Roll_Att}}
 {id=Attenuation_4, children [Mixer_4], properties {attenuationSettings=&Cube_Roll_Att}}
 {id=Attenuation_5, children [Mixer_5], properties {attenuationSettings=&Cube_Roll_Att}}
 {id=Mixer_0, children [Modulator_0,Random_0], properties {inputVolume [1,1]}}
 {id=Mixer_1, children [Modulator_2,Random_3], properties {inputVolume [1,1]}}
 {id=Mixer_2, children [Modulator_3,Random_4], properties {inputVolume [1,1]}}
 {id=Mixer_3, children [Modulator_4,Switch_2], properties {inputVolume [1,1]}}
 {id=Mixer_4, children [Modulator_1,Random_1], properties {inputVolume [1,1]}}
 {id=Mixer_5, children [Modulator_5,Switch_3], properties {inputVolume [1,1]}}
 {id=Modulator_0, children [WavePlayer_5]}
 {id=Modulator_1, children [WavePlayer_6]}
 {id=Modulator_2, children [WavePlayer_14]}
 {id=Modulator_3, children [WavePlayer_15]}
 {id=Modulator_4, children [WavePlayer_16]}
 {id=Modulator_5, children [WavePlayer_7]}
 {id=QualityLevel_0, children [Switch_0,Switch_0,Switch_1,Switch_1]}
 {id=Random_0, children [WavePlayer_0,WavePlayer_1], properties {bShouldExcludeFromBranchCulling=true, weights [1,1]}}
 {id=Random_1, children [WavePlayer_27, WavePlayer_28], properties {bShouldExcludeFromBranchCulling=true, weights [1,1]}}
 {id=Random_2, children [WavePlayer_38, WavePlayer_39, WavePlayer_40, WavePlayer_41, WavePlayer_42], properties {weights [1,1,1,1,1]}}
 {id=Random_3, children [WavePlayer_9, WavePlayer_10, WavePlayer_11, WavePlayer_12, WavePlayer_13], properties {weights [1,1,1,1,1]}}
 {id=Random_4, children [WavePlayer_17, WavePlayer_18, WavePlayer_19, WavePlayer_20, WavePlayer_21], properties {weights [1,1,1,1,1]}}
 {id=Random_5, children [WavePlayer_22, WavePlayer_23, WavePlayer_24, WavePlayer_25, WavePlayer_26], properties {weights [1,1,1,1,1]}}
 {id=Random_6, children [WavePlayer_32, WavePlayer_33], properties {bShouldExcludeFromBranchCulling=true, weights [1,1]}}
 {id=Random_7, children [WavePlayer_37, WavePlayer_43], properties {bShouldExcludeFromBranchCulling=true, weights [1,1]}}
 {id=Switch_0, children [Attenuation_1, Attenuation_0, Attenuation_1, Attenuation_2], properties {intParameterName=Size}}
 {id=Switch_1, children [Attenuation_4, Attenuation_3, Attenuation_4, Attenuation_5], properties {intParameterName=Size}}
 {id=Switch_2, children [Random_5,Random_5,Random_5,Random_2], properties {intParameterName=Color}}
 {id=Switch_3, children [Random_6,Random_6,Random_6,Random_7], properties {intParameterName=Color}}
 {id=WavePlayer_0, properties {soundWaveAssetPtr=&/Cube_Vox_Roll_Small_01}}
 {id=WavePlayer_1, properties {soundWaveAssetPtr=&/Cube_Vox_Roll_Small_02}}
 {id=WavePlayer_10, properties {soundWaveAssetPtr=&Cube_Vox_Roll_Small_02}}
 {id=WavePlayer_11, properties {soundWaveAssetPtr=&/Cube_Vox_Roll_Small_03}}
 {id=WavePlayer_12, properties {soundWaveAssetPtr=&/Cube_Vox_Roll_Small_04}}
 {id=WavePlayer_13, properties {soundWaveAssetPtr=&/Cube_Vox_Roll_Small_05}}
 {id=WavePlayer_14, properties {soundWaveAssetPtr=&/Cube_SFX_Roll_01}}
 {id=WavePlayer_15, properties {soundWaveAssetPtr=&/Cube_SFX_Roll_02}}
 {id=WavePlayer_16, properties {soundWaveAssetPtr=&/Cube_SFX_Roll_03}}
 {id=WavePlayer_17, properties {soundWaveAssetPtr=&/Cube_Vox_Roll_Med_05}}
 {id=WavePlayer_18, properties {soundWaveAssetPtr=&/Cube_Vox_Roll_Med_01}}
 {id=WavePlayer_19, properties {soundWaveAssetPtr=&/Cube_Vox_Roll_Med_02}}
 {id=WavePlayer_20, properties {soundWaveAssetPtr=&/Cube_Vox_Roll_Med_03}}
 {id=WavePlayer_21, properties {soundWaveAssetPtr=&/Cube_Vox_Roll_Med_04}}
 {id=WavePlayer_22, properties {soundWaveAssetPtr=&/Cube_Vox_Roll_Lg_05}}
 {id=WavePlayer_23, properties {soundWaveAssetPtr=&/Cube_Vox_Roll_Lg_01}}
 {id=WavePlayer_24, properties {soundWaveAssetPtr=&/Cube_Vox_Roll_Lg_02}}
 {id=WavePlayer_25, properties {soundWaveAssetPtr=&/Cube_Vox_Roll_Lg_03}}
 {id=WavePlayer_26, properties {soundWaveAssetPtr=&/Cube_Vox_Roll_Lg_04}}
 {id=WavePlayer_27, properties {soundWaveAssetPtr=&Cube_Vox_Roll_Med_01}}
 {id=WavePlayer_28, properties {soundWaveAssetPtr=&Cube_Vox_Roll_Med_02}}
 {id=WavePlayer_32, properties {soundWaveAssetPtr=&Cube_Vox_Roll_Lg_01}}
 {id=WavePlayer_33, properties {soundWaveAssetPtr=&Cube_Vox_Roll_Lg_02}}
 {id=WavePlayer_37, properties {soundWaveAssetPtr=&/Cube_Vox_Roll_Gold_01}}
 {id=WavePlayer_38, properties {soundWaveAssetPtr=&/Cube_Vox_Roll_Gold_05}}
 {id=WavePlayer_39, properties {soundWaveAssetPtr=&Cube_Vox_Roll_Gold_01}}
 {id=WavePlayer_40, properties {soundWaveAssetPtr=&/Cube_Vox_Roll_Gold_02}}
 {id=WavePlayer_41, properties {soundWaveAssetPtr=&/Cube_Vox_Roll_Gold_03}}
 {id=WavePlayer_42, properties {soundWaveAssetPtr=&/Cube_Vox_Roll_Gold_04}}
 {id=WavePlayer_43, properties {soundWaveAssetPtr=&Cube_Vox_Roll_Gold_02}}
 {id=WavePlayer_5, properties {soundWaveAssetPtr=&Cube_SFX_Roll_01}}
 {id=WavePlayer_6, properties {soundWaveAssetPtr=&Cube_SFX_Roll_02}}
 {id=WavePlayer_7, properties {soundWaveAssetPtr=&Cube_SFX_Roll_03}}
 {id=WavePlayer_9, properties {soundWaveAssetPtr=&Cube_Vox_Roll_Small_01}}
]
properties {
 bOverrideConcurrency=true
 maxDistance=20000
 soundClassObject=&Fort_Mission_Ingame_Important_PSM
 volumeMultiplier=1
 concurrencyOverrides {maxCount=2}
}
