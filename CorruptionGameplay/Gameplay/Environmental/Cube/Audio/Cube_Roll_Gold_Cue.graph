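# Sound cue for the gold Cube's roll in CorruptionGameplay, layering a roll effect with a randomly chosen gold vocal.
# A quality-level node routes its first two slots to a five-vocal branch and its last two to a two-vocal branch.
# Each branch mixes the same modulated SFX wave with its random vocal under roll attenuation.
# The cue overrides concurrency to limit simultaneous instances and uses an important mission sound class.
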SoundCue /CorruptionGameplay/Gameplay/Environmental/Cube/Audio/Cube_Roll_Gold_Cue
root=QualityLevel_0
nodes [
 {id=Attenuation_0, children [Mixer_0], properties {attenuationSettings=&/Cube_Roll_Att}}
 {id=Attenuation_2, children [Mixer_3], properties {attenuationSettings=&Cube_Roll_Att}}
 {id=Mixer_0, children [Modulator_0,Random_0], properties {inputVolume [1,1]}}
 {id=Mixer_3, children [Modulator_4,Random_2], properties {inputVolume [1,1]}}
 {id=Modulator_0, children [WavePlayer_0]}
 {id=Modulator_4, children [WavePlayer_16]}
 {id=QualityLevel_0, children [Attenuation_2, Attenuation_2, Attenuation_0, Attenuation_0]}
 {id=Random_0, children [WavePlayer_2,WavePlayer_3], properties {bShouldExcludeFromBranchCulling=true, weights [1,1]}}
 {id=Random_2, children [WavePlayer_38, WavePlayer_39, WavePlayer_40, WavePlayer_41, WavePlayer_42], properties {weights [1,1,1,1,1]}}
 {id=WavePlayer_0, properties {soundWaveAssetPtr=&/Cube_SFX_Roll_03}}
 {id=WavePlayer_16, properties {soundWaveAssetPtr=&Cube_SFX_Roll_03}}
 {id=WavePlayer_2, properties {soundWaveAssetPtr=&/Cube_Vox_Roll_Gold_01}}
 {id=WavePlayer_3, properties {soundWaveAssetPtr=&/Cube_Vox_Roll_Gold_02}}
 {id=WavePlayer_38, properties {soundWaveAssetPtr=&/Cube_Vox_Roll_Gold_05}}
 {id=WavePlayer_39, properties {soundWaveAssetPtr=&Cube_Vox_Roll_Gold_01}}
 {id=WavePlayer_40, properties {soundWaveAssetPtr=&Cube_Vox_Roll_Gold_02}}
 {id=WavePlayer_41, properties {soundWaveAssetPtr=&/Cube_Vox_Roll_Gold_03}}
 {id=WavePlayer_42, properties {soundWaveAssetPtr=&/Cube_Vox_Roll_Gold_04}}
]
properties {
 bOverrideConcurrency=true
 maxDistance=20000
 soundClassObject=&Fort_Mission_Ingame_Important_PSM
 volumeMultiplier=1
 concurrencyOverrides {maxCount=2}
}
