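# Sound Cue named for the spawn telegraph of the Cube in the CorruptionGameplay environmental content.
# It has no nodes and no root, so it produces no audio as it stands.
# It keeps only playback settings: the gadgets sound class, a maximum distance and an attenuation override for stereo spread.
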SoundCue /CorruptionGameplay/Gameplay/Environmental/Cube/Audio/Cube_SpawnTelegraph_Cue
root=null
nodes []
properties {maxDistance=4000, soundClassObject=&Fort_Gadgets, attenuationOverrides {stereoSpread=512}}
