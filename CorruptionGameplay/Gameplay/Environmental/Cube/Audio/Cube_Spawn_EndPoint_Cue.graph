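# Sound cue for a cube spawn end point in the Corruption gameplay environment, judging by its name and path.
# It routes a single beam-start wave through a one-input mixer, with no randomization or variation.
# The cue uses cube spawn attenuation with a long audible range, an important in-game mission sound class and boosted volume.
# It overrides concurrency to cap simultaneous instances at a small count.
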SoundCue /CorruptionGameplay/Gameplay/Environmental/Cube/Audio/Cube_Spawn_EndPoint_Cue
root=Mixer_0
nodes [
 {id=Mixer_0, children [WavePlayer_1], properties {inputVolume [1]}}
 {id=WavePlayer_1, properties {soundWaveAssetPtr=&/Cube_Spawn_Beam_Child_Start_01}}
]
properties {
 attenuationSettings=&/Cube_Spawn_Att
 bOverrideConcurrency=true
 maxDistance=26000
 soundClassObject=&Fort_Mission_Ingame_Important_PSM
 volumeMultiplier=1.5
 concurrencyOverrides {maxCount=2}
}
