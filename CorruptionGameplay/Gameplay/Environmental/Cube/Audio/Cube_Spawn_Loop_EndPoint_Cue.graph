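# Sound cue for the looping beam audio at the end point of the Cube spawn effect.
# A looping beam wave runs through an enveloper that ramps volume from silence to full over two seconds, slowly at first, then into a mixer that halves the level.
# It uses the spawn loop attenuation asset and an important in-game mission sound class, with a concurrency override allowing two simultaneous instances.
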SoundCue /CorruptionGameplay/Gameplay/Environmental/Cube/Audio/Cube_Spawn_Loop_EndPoint_Cue
root=Mixer_0
nodes [
 {
  id=Enveloper_0
  children [WavePlayer_2]
  properties {pitchCurve {editorCurveData {keys [{time=0}]}}, volumeCurve {editorCurveData {keys [<at 0 0>,<at 0.859 0.1429>,<at 2 1>]}}}
 }
 {id=Mixer_0, children [Enveloper_0], properties {inputVolume [0.5]}}
 {
  id=WavePlayer_2
  properties {bLooping=true, soundWaveAssetPtr=&/Cube_Spawn_Beam_Child_Loop_01}
 }
]
properties {
 attenuationSettings=&/Cube_Spawn_Loop_Att
 bOverrideConcurrency=true
 maxDistance=10000
 soundClassObject=&Fort_Mission_Ingame_Important_PSM
 volumeMultiplier=1
 concurrencyOverrides {maxCount=2}
}
