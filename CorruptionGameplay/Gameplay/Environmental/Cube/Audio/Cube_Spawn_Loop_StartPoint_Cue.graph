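# Sound cue for the looping parent beam sound of a Cube spawn in the environmental Cube gameplay, named for the start point.
# A single looping wave runs through an enveloper that fades volume in over the first two seconds, then a mixer that lowers its level.
# It uses a shared spawn-loop attenuation asset and an important in-game mission sound class, and overrides concurrency to cap it at two simultaneous instances.
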SoundCue /CorruptionGameplay/Gameplay/Environmental/Cube/Audio/Cube_Spawn_Loop_StartPoint_Cue
root=Mixer_0
nodes [
 {
  id=Enveloper_0
  children [WavePlayer_5]
  properties {pitchCurve {editorCurveData {keys [{time=0}]}}, volumeCurve {editorCurveData {keys [<at 0 0>,<at 2 1>]}}}
 }
 {id=Mixer_0, children [Enveloper_0], properties {inputVolume [0.7]}}
 {
  id=WavePlayer_5
  properties {bLooping=true, soundWaveAssetPtr=&/Cube_Spawn_Parent_Beam_Loop_01}
 }
]
properties {
 attenuationSettings=&/Cube_Spawn_Loop_Att
 bOverrideConcurrency=true
 maxDistance=10000
 soundClassObject=&Fort_Mission_Ingame_Important_PSM
 volumeMultiplier=1
 concurrencyOverrides {maxCount=2}
}
