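# Sound cue for the end point of the cube spawn loop stop, layering a beam shutdown with cube and lightning appearance sounds.
# The root mixer plays the beam stop at reduced volume alongside a sub-mix of the two appearance layers, each briefly delayed.
# Each layer has its own attenuation node, with the lightning using an extra-wide asset.
# It uses an important mission sound class and overrides concurrency to limit simultaneous instances.
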SoundCue /CorruptionGameplay/Gameplay/Environmental/Cube/Audio/Cube_Spawn_Loop_Stop_EndPoint_Cue
root=Mixer_1
nodes [
 {
  id=Attenuation_0
  children [WavePlayer_2]
  properties {attenuationSettings=&/Cube_Spawn_ExtraWide_Att}
 }
 {
  id=Attenuation_1
  children [WavePlayer_3]
  properties {attenuationSettings=&/Cube_Spawn_Att}
 }
 {
  id=Attenuation_3
  children [WavePlayer_1]
  properties {attenuationSettings=&Cube_Spawn_Att}
 }
 {id=Delay_0, children [Mixer_0]}
 {id=Delay_1, children [Attenuation_0], properties {delayMax=0.25, delayMin=0.25}}
 {id=Delay_2, children [Attenuation_1], properties {delayMax=0.25, delayMin=0.25}}
 {id=Mixer_0, children [Delay_1,Delay_2], properties {inputVolume [1,1]}}
 {id=Mixer_1, children [Attenuation_3, Delay_0], properties {inputVolume [0.7,1]}}
 {id=WavePlayer_1, properties {soundWaveAssetPtr=&/Cube_Spawn_Beam_Child_Stop_01}}
 {
  id=WavePlayer_2
  properties {soundWaveAssetPtr=&/Cube_Spawn_Appear_Child_Start_Lightning_01}
 }
 {id=WavePlayer_3, properties {soundWaveAssetPtr=&/Cube_Spawn_Appear_Child_Start_Cube_01}}
]
properties {
 bOverrideConcurrency=true
 maxDistance=42500
 soundClassObject=&Fort_Mission_Ingame_Important_PSM
 volumeMultiplier=1
 concurrencyOverrides {maxCount=2}
}
