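# Sound cue for the stop of a cube spawn beam loop at its start point, in the Corruption gameplay's environmental cube audio.
# A single wave player plays the beam parent stop sound through an attenuation node using the shared cube spawn attenuation.
# It routes to the important in-game mission sound class and overrides concurrency to cap simultaneous instances at two.
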SoundCue /CorruptionGameplay/Gameplay/Environmental/Cube/Audio/Cube_Spawn_Loop_Stop_StartPoint_Cue
root=Attenuation_2
nodes [
 {
  id=Attenuation_2
  children [WavePlayer_1]
  properties {attenuationSettings=&/Cube_Spawn_Att}
 }
 {id=WavePlayer_1, properties {soundWaveAssetPtr=&/Cube_Spawn_Beam_Parent_Stop_01}}
]
properties {
 bOverrideConcurrency=true
 maxDistance=26000
 soundClassObject=&Fort_Mission_Ingame_Important_PSM
 volumeMultiplier=1
 concurrencyOverrides {maxCount=2}
}
