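# Sound cue for the start-point sound of a Cube spawn beam in the Corruption gameplay environment.
# A quality-level root routes its first two slots to an evenly weighted random pick of four start waves, and its last two to a pick of two, each through a single-input mixer.
# It plays boosted on an important mission sound class with Cube spawn attenuation and a concurrency override allowing two instances.
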SoundCue /CorruptionGameplay/Gameplay/Environmental/Cube/Audio/Cube_Spawn_StartPoint_Cue
root=QualityLevel_0
nodes [
 {id=Mixer_0, children [Random_0], properties {inputVolume [1]}}
 {id=Mixer_1, children [Random_1], properties {inputVolume [1]}}
 {id=QualityLevel_0, children [Mixer_0,Mixer_0,Mixer_1,Mixer_1]}
 {id=Random_0, children [WavePlayer_1,WavePlayer_2,WavePlayer_3,WavePlayer_4], properties {weights [1,1,1,1]}}
 {id=Random_1, children [WavePlayer_0,WavePlayer_5], properties {bShouldExcludeFromBranchCulling=true, weights [1,1]}}
 {id=WavePlayer_0, properties {soundWaveAssetPtr=&/Cube_Spawn_Beam_Parent_Start_01}}
 {id=WavePlayer_1, properties {soundWaveAssetPtr=&Cube_Spawn_Beam_Parent_Start_01}}
 {id=WavePlayer_2, properties {soundWaveAssetPtr=&/Cube_Spawn_Beam_Parent_Start_02}}
 {id=WavePlayer_3, properties {soundWaveAssetPtr=&/Cube_Spawn_Beam_Parent_Start_03}}
 {id=WavePlayer_4, properties {soundWaveAssetPtr=&/Cube_Spawn_Beam_Parent_Start_04}}
 {id=WavePlayer_5, properties {soundWaveAssetPtr=&Cube_Spawn_Beam_Parent_Start_02}}
]
properties {
 attenuationSettings=&/Cube_Spawn_Att
 bOverrideConcurrency=true
 maxDistance=26000
 soundClassObject=&Fort_Mission_Ingame_Important_PSM
 volumeMultiplier=1.5
 concurrencyOverrides {maxCount=2}
}
