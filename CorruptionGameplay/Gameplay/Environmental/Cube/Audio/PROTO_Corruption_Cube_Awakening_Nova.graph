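# Prototype SoundCue for the Corruption Cube's awakening nova, built from borrowed Constructor gadget bass sounds.
# A mixer plays three waves at once, two TurnUpTheBaseBass variants and FeelTheBaseBass, with one TurnUp layer slightly boosted.
# It overrides attenuation with a long-range logarithmic falloff, wide stereo spread and a non-spatialized near zone.
# Output routes to an important in-game mission sound class.
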SoundCue /CorruptionGameplay/Gameplay/Environmental/Cube/Audio/PROTO_Corruption_Cube_Awakening_Nova
root=Mixer_0
nodes [
 {id=Mixer_0, children [WavePlayer_0,WavePlayer_3,WavePlayer_1], properties {inputVolume [1.25,1,1]}}
 {id=WavePlayer_0, properties {soundWaveAssetPtr=/Game/Sounds/Fort_Gadgets/ConstructionWorker/Constructor_TurnUpTheBaseBass07}}
 {id=WavePlayer_1, properties {soundWaveAssetPtr=/Game/Sounds/Fort_Gadgets/ConstructionWorker/Constructor_TurnUpTheBaseBass05}}
 {id=WavePlayer_3, properties {soundWaveAssetPtr=/Game/Sounds/Fort_Gadgets/ConstructionWorker/Constructor_FeelTheBaseBass}}
]
properties {
 bOverrideAttenuation=true
 maxDistance=11500
 soundClassObject=&Fort_Mission_Ingame_Important_PSM
 attenuationOverrides {distanceAlgorithm=Logarithmic, falloffDistance=7000, nonSpatializedRadiusStart=800, stereoSpread=3000, attenuationShapeExtents {x=4500}}
}
