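; Corruption gameplay cube actor that follows a scripted static path, with empty spawn, roll and awaken events as hooks.
; Each step either teleports for begin-play calendar skips or dispatches on a movement gameplay tag to Roll, Levitate or Linear.
; Roll uses a timeline with dual-quaternion transform interpolation and a CVar-gated large-gap check, Linear interpolates location, and Levitate simply snaps.
; Steps report completion to the path system; the height offset function returns zero here.
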(event Custom|StaticPathBeginPlaySkipOccurred
  ;[1 1 1] Children Overriden Events
  (Variables|Default|SetBeginPlaySkipOccured true)
  (Default|CallOnBeginPlaySkipOccured))

(event Custom|RollEnd)

(event Custom|SpawnCube)

(event Custom|RollStart)

(event Custom|AwakenCube)

(event StaticPathFollowing|EventBeginMovingToStaticPathNode
  (params (StepData :type Scripted%20Object%20Movement%20Static%20Path%20Step%20Data%20Structure%20%28by%20ref%29))
  (bind (_startworldtransform _endworldtransform) (Utilities|Struct|BreakScriptedObjectMovementStaticPathStepData StepData))
  (bind (_startworldtransform_2 _endworldtransform_2 _movestyletag _bteleport) (Utilities|Struct|BreakScriptedObjectMovementStaticPathStepData StepData))
  ;[1 1 1] Choose Movement Style
  (if _bteleport
    ;;[0 0 0] Begin Play Skip To Calendar Location
    (Transformation|SetActorTransform (CallFunction|VerticalOffsetTransformLocationforCubePlacement _endworldtransform_2))
    (StaticPathFollowing|NotifyStaticPathStepComplete false) ; if we have invalid tag, just  end it.
    (Variables|Default|SetStaticPathStart (.location _startworldtransform))
    (Variables|Default|SetStaticPathEnd (.location _endworldtransform))
    (CallFunction|StaticPathBeginPlaySkipOccurred)
    (else
      (switch Utilities|FlowControl|Switch|SwitchonGameplayTag _movestyletag
        (:Default
          (Transformation|SetActorTransform _endworldtransform_2)
          (StaticPathFollowing|NotifyStaticPathStepComplete false) ; if we have invalid tag, just  end it.
          (Development|PrintString "Invalid Move Tag"))
        (:Cube.Movement.Roll
          (CallFunction|BeginMoveingStaticPathNodeRoll StepData))
        (:Cube.Movement.Levitate
          (CallFunction|BeginMoveingStaticPathNodeLevitate StepData))
        (:Cube.Movement.Linear
          (CallFunction|BeginMoveingStaticPathNodeLinear StepData))))))

(event EventBeginPlay
  ;[1 1 1] Begin Play
  (Variables|Default|SetCachedActorForwardVector (Transformation|GetActorForwardVector))
  (Variables|Default|SetCachedActorRightVector (Transformation|GetActorRightVector))
  (CallFunction|SpawnCube))

(event Custom|BeginMoveingStaticPathNode_Roll
  (params (StepData :type Scripted%20Object%20Movement%20Static%20Path%20Step%20Data%20Structure))
  (bind (_startworldtransform _endworldtransform) (Utilities|Struct|BreakScriptedObjectMovementStaticPathStepData StepData))
  (bind (_outvalue _returnvalue) (Fortnite|GetNumericCVar "Fort.Som.StaticPath.AttemptLargeStaticPathDistancesAtRuntime"))
  (bind _location (.location _endworldtransform))
  (bind _location_2 (.location _startworldtransform))
  (Variables|Default|SetStaticPathStart _location_2)
  (Variables|Default|SetStaticPathEnd _location)
  (Components|Timeline|SetPlayRate (Variables|BGA_AthenaCorruptionGameplay_StaticPathFollowingCube|GetMoveCubeTimeline) (Math|Float|SafeDivide 1.0 (Variables|Default|GetMoveDurationSeconds)))
  (CallFunction|GetStaticPathRotation _location _location_2)
  (Variables|Default|SetStart%20World%20Transform _startworldtransform)
  (Variables|Default|SetEnd%20World%20Transform _endworldtransform)
  (if (> (Math|Vector|DistanceSquared(Vector) (Variables|Default|GetStaticPathEnd) (Variables|Default|GetStaticPathStart)) (Math|Float|Square (* (.x (CallFunction|GetStaticPathTravelingBounds)) 3.0)))
    ;[1 1 1] Begin Play Skip Water
    (Development|PrintString "BeginPlaySkippingLargeGap" true true "(R=1.000000,G=0.000000,B=0.000000,A=1.000000)" 5.0)
    (if _returnvalue
      (if (> _outvalue 0.0)
        (Development|PrintString "Canceling Begin Play As Fort.Som.StaticPath.AttemptLargeStaticPathDistancesAtRuntime 
is true" true true "(R=1.000000,G=0.815959,B=0.000000,A=1.000000)" 5.0)
        (CallFunction|RollStart)
        (bind (_direction _movepercent _tipamount) (Timeline|MoveCubeTimeline PlayFromStart)
          (:Update
            (Transformation|SetWorldTransform (Variables|Default|GetDefaultSceneRoot) (CallFunction|VerticalOffsetTransformLocationforCubePlacement (Math|Transform|Lerp(Transform) (Variables|Default|GetStart%20World%20Transform) (Variables|Default|GetEnd%20World%20Transform) _tipamount "DualQuatInterp")) false false))
          (:Finished
            ;[1 0 0] must be called for code to know we're done
            (StaticPathFollowing|NotifyStaticPathStepComplete false)
            (CallFunction|RollEnd)))
        (else
          _))
      (else
        _))
    (StaticPathFollowing|NotifyStaticPathStepComplete true)))

(event Custom|BeginMoveingStaticPathNode_Levitate
  (params (StepData :type Scripted%20Object%20Movement%20Static%20Path%20Step%20Data%20Structure))
  (bind (_startworldtransform _endworldtransform) (Utilities|Struct|BreakScriptedObjectMovementStaticPathStepData StepData))
  (Transformation|SetActorTransform _endworldtransform)
  ;[1 0 0] must be called for code to know we're done
  (StaticPathFollowing|NotifyStaticPathStepComplete false))

(event Custom|BeginMoveingStaticPathNode_Linear
  (params (StepData :type Scripted%20Object%20Movement%20Static%20Path%20Step%20Data%20Structure))
  (bind (_startworldtransform _endworldtransform) (Utilities|Struct|BreakScriptedObjectMovementStaticPathStepData StepData))
  ;[0.151 0.151 0.151] Debug Linear Movement
  (Components|Timeline|SetPlayRate (Variables|BGA_AthenaCorruptionGameplay_StaticPathFollowingCube|GetLinearMoveTimeline) (Math|Float|SafeDivide 1.0 (Variables|Default|GetMoveDurationSeconds)))
  (Variables|Default|SetStaticPathStart (.location _startworldtransform))
  (Variables|Default|SetStaticPathEnd (.location _endworldtransform))
  (bind (_direction _linearmovepercent) (Timeline|LinearMoveTimeline PlayFromStart)
    (:Update
      (Transformation|SetActorLocation (Math|Vector|MakeVector (.x (+ (* (- (Variables|Default|GetStaticPathEnd) (Variables|Default|GetStaticPathStart)) _linearmovepercent) (Variables|Default|GetStaticPathStart))) (.y (+ (* (- (Variables|Default|GetStaticPathEnd) (Variables|Default|GetStaticPathStart)) _linearmovepercent) (Variables|Default|GetStaticPathStart))) (+ (.z (+ (* (- (Variables|Default|GetStaticPathEnd) (Variables|Default|GetStaticPathStart)) _linearmovepercent) (Variables|Default|GetStaticPathStart))) (CallFunction|GetPathHeightOffset)))))
    (:Finished
      ;;[1 0 0] must be called for code to know we're done
      (StaticPathFollowing|NotifyStaticPathStepComplete false))))

(fn GetPathHeightOffset
  (return 0.0))

(fn GetStaticPathRotation
  (params (EndLocation :type Vector) (StartLocation :type Vector) (RotationAngleMagnitude :type Float%20%28double-precision%29 :default 9e+01))
  (bind _cachedactorforwardvector (Variables|Default|GetCachedActorForwardVector))
  (bind _returnvalue (- EndLocation StartLocation))
  (bind _returnvalue_2 (Math|Vector|DotProduct _returnvalue _cachedactorforwardvector))
  (bind _cachedactorrightvector (Variables|Default|GetCachedActorRightVector))
  (bind _returnvalue_3 (Math|Vector|DotProduct _returnvalue _cachedactorrightvector))
  (if (> (Math|Float|Absolute(Float) _returnvalue_3) (Math|Float|Absolute(Float) _returnvalue_2))
    (if (> _returnvalue_3 0.0)
      (Variables|Default|SetRotationOffset (Math|Rotator|MakeRotator (Variables|GetRotationAngleMagnitude)))
      (return)
      (else
        (Variables|Default|SetRotationOffset (Math|Rotator|MakeRotator (* (Variables|GetRotationAngleMagnitude) -1.0)))
        (return)))
    (elif (> _returnvalue_2 0.0)
      (Variables|Default|SetRotationOffset (Math|Rotator|MakeRotator 0.0 (* (Variables|GetRotationAngleMagnitude) -1.0)))
      (return)
      (else
        (Variables|Default|SetRotationOffset (Math|Rotator|MakeRotator 0.0 (Variables|GetRotationAngleMagnitude)))
        (return)))))

(fn GetStaticPathRotationForHills
  (params (StepData :type Scripted%20Object%20Movement%20Static%20Path%20Step%20Data%20Structure))
  (bind (_startworldtransform _endworldtransform) (Utilities|Struct|BreakScriptedObjectMovementStaticPathStepData StepData))
  (bind _returnvalue (Math|Vector|GetForwardVector (Transformation|GetActorRotation)))
  (bind _returnvalue_2 (Math|Vector|DotProduct _returnvalue "0, 0,1.000000"))
  (bind _returnvalue_3 (Math|Vector|GetRightVector (Transformation|GetActorRotation)))
  (bind _returnvalue_4 (Math|Vector|DotProduct _returnvalue_3 "0, 0,1.000000"))
  (bind _returnvalue_5 (Math|Vector|GetUpVector (Transformation|GetActorRotation)))
  (bind _returnvalue_6 (Math|Vector|DotProduct _returnvalue_5 "0, 0,1.000000"))
  (bind _returnvalue_7 (Math|Float|Absolute(Float) _returnvalue_4))
  (bind _returnvalue_8 (Math|Vector|GetForwardVector (Transformation|GetActorRotation)))
  (bind _returnvalue_9 (Math|Vector|GetRightVector (Transformation|GetActorRotation)))
  (bind _returnvalue_10 (Transformation|GetActorRotation))
  (bind _returnvalue_11 (Math|Vector|GetUpVector _returnvalue_10))
  (bind _returnvalue_12 (Transformation|GetActorLocation))
  (bind _endlocation_noheightoffset (LocalVariables|GetEndLocation_NoHeightOffset))
  (bind _returnvalue_13 (- _endlocation_noheightoffset _returnvalue_12))
  (bind _returnvalue_14 (Math|Vector|MakeVector 0.0 0.0 1.0))
  (bind _returnvalue_15 (Math|Vector|Normalize _returnvalue_13))
  (bind _returnvalue_16 (Math|Vector|CrossProduct _returnvalue_14 _returnvalue_15))
  (Utilities|FlowControl|Sequence
    (:then_0
      (LocalVariables|SetStartLocation_NoHeightOffset (.location _startworldtransform))
      (LocalVariables|SetEndLocation_NoHeightOffset (.location _endworldtransform))
      (LocalVariables|SetTerrainNormal (Math|Transform|TransformDirection _endworldtransform "0, 0,1.000000"))
      (Utilities|FlowControl|Sequence
        (:then_0
          ;[0 1 0.07514] Figure out which vector is closes to the world up vector (Z)
          ; This is used adjust the rotation to match hill height
          (if (> (Math|Float|Absolute(Float) _returnvalue_2) _returnvalue_7)
            (if (> _returnvalue_2 0.0)
              (LocalVariables|SetCurrentUp _returnvalue_8)
              (else
                (LocalVariables|SetCurrentUp (* _returnvalue_8 -1.0))))
            (elif (> _returnvalue_7 (Math|Float|Absolute(Float) _returnvalue_6))
              (if (> _returnvalue_4 0.0)
                (LocalVariables|SetCurrentUp _returnvalue_9)
                (else
                  (LocalVariables|SetCurrentUp (* _returnvalue_9 -1.0))))
              (elif (> _returnvalue_6 0.0)
                (LocalVariables|SetCurrentUp _returnvalue_11)
                (else
                  (LocalVariables|SetCurrentUp (* _returnvalue_11 -1.0)))))))
        (:then_1
          ;[0.1927 0.5185 1] Project the terrain normal onto the plane of a flat rotation.
          ; this is useful because we can break the rotation into 2 peices (as if cube rolled forward, then rolled toside)
          (LocalVariables|SetFirstRotationAngle (- 9e+01 (Math|Trig|RadiansToDegrees (Math|Trig|Acos(Radians) (Math|Vector|DotProduct (Math|Vector|Normalize (Math|Vector|ProjectVectorontoPlane (LocalVariables|GetTerrainNormal) (Math|Vector|Normalize _returnvalue_16))) (Math|Vector|Normalize (Math|Vector|ProjectVectorontoPlane (LocalVariables|GetCurrentUp) (Math|Vector|Normalize _returnvalue_16))))))))
          ;[0.1927 0.5185 1] sideways projection, the second rotation to bring this level with the hill
          (LocalVariables|SetSecondRotationAngle (Math|Trig|RadiansToDegrees (Math|Trig|Acos(Radians) (Math|Vector|DotProduct (Math|Vector|ProjectVectorontoPlane (LocalVariables|GetTerrainNormal) (Math|Vector|Normalize (Math|Vector|CrossProduct "0, 0,1.000000" _returnvalue_16))) (Math|Vector|ProjectVectorontoPlane (LocalVariables|GetCurrentUp) (Math|Vector|Normalize (Math|Vector|CrossProduct "0, 0,1.000000" _returnvalue_16))))))))
        (:then_2
          (CallFunction|GetStaticPathRotation (LocalVariables|GetEndLocation_NoHeightOffset) (LocalVariables|GetStartLocation_NoHeightOffset) (LocalVariables|GetFirstRotationAngle)))))
    (:then_1
      (if true
        (Rendering|Debug|DrawDebugLine (.location _endworldtransform) (+ (.location _endworldtransform) (* (Math|Transform|TransformDirection _endworldtransform "0, 0,1.000000") 1e+04)) "(R=1.000000,G=0.309210,B=0.000000,A=1.000000)" 1e+02)))))

(fn GetStaticPathZOffsetHeight
  (return (* (.z (.scale (Transformation|GetActorTransform))) (CallFunction|GetPathHeightOffset))))

(fn ConstructionScript)

(fn VerticalOffsetTransformLocationForCubePlacement
  (params (InTransform :type Transform%20%28by%20ref%29))
  ;[1 1 1] NOTE: The start/end transforms are at the height of the landscape
  ; This logic is responsible for adding the vertical offset which keeps the leading edge at the height of the landscape as the cube rolls
  (return (Math|Transform|MakeTransform (Math|Vector|MakeVector (.x (.location InTransform)) (.y (.location InTransform)) (+ (.z (.location InTransform)) (* (.z (.scale InTransform)) (CallFunction|GetPathHeightOffset)))) (.rotation InTransform) (.scale InTransform))))
