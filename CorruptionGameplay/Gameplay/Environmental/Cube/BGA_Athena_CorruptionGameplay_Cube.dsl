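; Indestructible Corruption cube actor for Fortnite that rolls, bounces players, vehicles and physics objects, and destroys what it lands on.
; Active faces (all, timer-randomized or damage-activated) launch hitting pawns and objects, while overlaps launch vehicles.
; A delayed surface-change event backs up pawn hits, with tags and cooldowns limiting repeat bounces.
; Async targeting clears overlaps on spawn and during rolls, and gameplay cues play roll, landing and hit-ripple feedback.
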(event Custom|KillAllOnSpawn
  ;[1 1 1] Cube Calls KillAllOnSpawn at BeginPlay to clear anything overlap
  (switch Utilities|FlowControl|Switch|SwitchonE_CubeSizes (Variables|Default|GetCubeSize)
    (:NewEnumerator0
      (Variables|Default|SetOnSpawnTargetingPreset "/CorruptionGameplay/Gameplay/Environmental/Cube/Cube_KillActorsOnSpawn_OriginalSize_Targeting.Cube_KillActorsOnSpawn_OriginalSize_Targeting"))
    (:NewEnumerator1
      (Variables|Default|SetOnSpawnTargetingPreset "/CorruptionGameplay/Gameplay/Environmental/Cube/Cube_KillActorsOnSpawn_MediumSize_Targeting.Cube_KillActorsOnSpawn_MediumSize_Targeting"))
    (:NewEnumerator2
      (Variables|Default|SetOnSpawnTargetingPreset "/CorruptionGameplay/Gameplay/Environmental/Cube/Cube_KillActorsOnSpawn_SmallSize_Targeting.Cube_KillActorsOnSpawn_SmallSize_Targeting")))
  (TargetingSystem|AsyncRequest|StartAsyncTargetingRequest (GameInstanceSubsystems|GetTargetingSubsystem) (Variables|Default|GetOnSpawnTargetingPreset) (Utilities|Struct|MakeTargetingSourceContext self self (Transformation|GetActorLocation self)) (delegate Custom|KillAllOnSpawn_Targeting)))

(event Custom|KillAllOnSpawn_Targeting
  (params (TargetingRequestHandle :type Targeting%20Request%20Handle%20Structure))
  ;[1 1 1] Cube Calls KillAllOnSpawn at BeginPlay to clear anything overlap
  (for _array_element (TargetingSystem|TargetingResults|GetTargetingResultsActors (GameInstanceSubsystems|GetTargetingSubsystem) TargetingRequestHandle)
    (bind _asbuilding_actor (Utilities|Casting|CastToBuildingActor _array_element)
      (:then
        (Utilities|Casting|CastToBGA_Athena_CorruptionGameplay_Cube _asbuilding_actor
          (:CastFailed
            (Fortnite|ApplyGameplayEffectsToActor _asbuilding_actor (Variables|Default|GetEffects%20to%20Apply) (Transformation|GetWorldLocation (Variables|Default|GetStaticMesh))))))
      (:CastFailed
        (if (not (Variables|Default|GetShouldOnlyDestroyBuildings))
          (bind _asfort_pawn (Utilities|Casting|CastToFortPawn _array_element)
            (:then
              (if (GameplayTags|HasMatchingGameplayTag _asfort_pawn "(TagName=\"Athena.Vehicle.Ostrich\")")
                (Actor|SetLifeSpan _asfort_pawn 0.1)
                (else
                  (Environment|ApplyGlobalEnvironmentDamagetoActor _asfort_pawn 1e+04 "(GameplayTags=((TagName=\"Athena.Gameplay.Status.ForceSkipDBNO\"),(TagName=\"Cube.Movement.Crushed\")))")
                  (GameplayEffects|ApplyGameplayEffectToSelf (Ability|GetFortAbilitySystemComponent _asfort_pawn) "/CorruptionGameplay/Gameplay/Environmental/Cube/GE_Cube_KillPawns.GE_Cube_KillPawns_C" 0.0 (Utilities|Struct|MakeGameplayEffectContextHandle)))))
            (:CastFailed
              (bind _asfort_athena_vehicle (Utilities|Casting|CastToFortAthenaVehicle _array_element))
              (GameplayEffects|ApplyGameplayEffectToSelf :self (Ability|GetFortAbilitySystemComponent _asfort_athena_vehicle) :GameplayEffectClass "/CorruptionGameplay/Gameplay/Environmental/Cube/GE_Cube_KillPawns.GE_Cube_KillPawns_C" :Level 0.0))))))))

(event EventOnDamagePlayEffects
  (params (Damage :type Float%20%28single-precision%29) (DamageTags :type Gameplay%20Tag%20Container%20Structure%20%28by%20ref%29) (Momentum :type Vector :default "0, 0, 0") (HitInfo :type Hit%20Result%20Structure%20%28by%20ref%29) (InstigatedBy :type Fort%20Pawn%20Object%20Reference) (DamageCauser :type Actor%20Object%20Reference) (EffectContext :type Gameplay%20Effect%20Context%20Handle%20Structure))
  (bind (_bblockinghit _binitialoverlap _time _distance _location _impactpoint _normal) (Collision|BreakHitResult HitInfo))
  ;[1 1 1] Visual Reation to being shot
  (CallFunction|RippleGC _location _normal DamageCauser (Math|Float|Clamp(Float) (/ Damage 5e+01) 0.0 1.25)))

(event Custom|RippleGC
  (params (Location :type Vector) (Normal :type Vector) (Source :type Actor%20Object%20Reference) (Scale :type Float%20%28double-precision%29))
  ;[1 1 1] Visual Reation to being shot
  (Utilities|Casting|CastToFortPlayerPawnAthena Source
    (:then
      (Variables|GameplayCues|SetGC_PawnHitFace true))
    (:CastFailed
      (Variables|GameplayCues|SetGC_PawnHitFace false)))
  (GameplayCue|ExecuteGameplayCue%28Burst%29Local self (Variables|GameplayCues|GetGC_HitRipple) (Ability|GameplayCue|MakeGameplayCueParameters :NormalizedMagnitude 0.0 :RawMagnitude Scale :EffectContext "()" :AggregatedSourceTags "(GameplayTags=)" :AggregatedTargetTags "(GameplayTags=)" :Location (Math|Vector|MakeVectorNetQuantize10 (.x Location) (.y Location) (.z Location)) :Normal (Math|Vector|MakeVectorNetQuantizeNormal (.x (select (Variables|GameplayCues|GetGC_PawnHitFace) (* Normal -1.0) Normal)) (.y (select (Variables|GameplayCues|GetGC_PawnHitFace) (* Normal -1.0) Normal)) (.z (select (Variables|GameplayCues|GetGC_PawnHitFace) (* Normal -1.0) Normal))) :EffectCauser Source :GameplayEffectLevel 1 :AbilityLevel (Variables|Default|GetCubeColorIndex) :bReplicateLocationWhenUsingMinimalRepProxy false)))

(event Custom|RemovePawnFromLaunchedArray
  (Utilities|Array|RemoveItem (Variables|Default|GetRecentlyLaunchedPawns) (Utilities|Array|Get(acopy) (Variables|Default|GetRecentlyLaunchedPawns))))

(event EventOnDamageServer
  (params (Damage :type Float%20%28single-precision%29) (DamageTags :type Gameplay%20Tag%20Container%20Structure%20%28by%20ref%29) (Momentum :type Vector :default "0, 0, 0") (HitInfo :type Hit%20Result%20Structure%20%28by%20ref%29) (InstigatedBy :type Controller%20Object%20Reference) (DamageCauser :type Actor%20Object%20Reference) (EffectContext :type Gameplay%20Effect%20Context%20Handle%20Structure))
  (bind (_bblockinghit _binitialoverlap _time _distance _location _impactpoint _normal) (Collision|BreakHitResult HitInfo))
  (bind (_returnvalue_x _returnvalue_y _returnvalue_z) (Math|Vector|BreakVector (* _normal 1e+03)))
  ;[1 1 1] On Server Damage, store the value of the damage for later
  (Development|PrintString (Utilities|String|ToString(Float) Damage))
  (Health|SetHealth (+ (Variables|FaceBehavior|GetDamageRequiredToActivateFace) 1.0))
  (if (Utilities|Enum|Equal(Enum) (Variables|FaceBehavior|GetActiveFaceConfig) "NewEnumerator2")
    (CallFunction|AddDamage (AddDamage|GetFacefromDirection _normal) Damage))
  (if (GameplayTags|HasTag DamageTags "(TagName=\"Weapon.Melee\")")
    (bind _asplayer_pawn_athena (Utilities|Casting|CastToPlayerPawn_Athena (Actor|GetOwner DamageCauser)))
    (Pawn|Character|LaunchCharacterJump _asplayer_pawn_athena (Math|Vector|MakeVector _returnvalue_x _returnvalue_y (+ _returnvalue_z 5e+02)) true true true true)))

(event Custom|StartLaunchCooldownForPawn
  (params (LaunchedFortPawn :type Fort%20Pawn%20Object%20Reference))
  ;[1 1 1] Pawn Launch Cooldon
  (Utilities|IsValid LaunchedFortPawn
    (:Is%20Valid
      (Utilities|Array|Add (Variables|Default|GetRecentlyLaunchedPawns) LaunchedFortPawn)
      (Utilities|Time|SetTimerbyEvent (delegate Custom|RemovePawnFromLaunchedArray) 0.25))))

(event Custom|FinalPlacement)

(event EventBeginPlay
  ;[1 1 1] Begin Play, check Active Face Config
  (|CallParentFunction)
  (if (ScalableFloat|GetValueasBool (Variables|Default|GetCubesEnabled))
    (CallFunction|FinalPlacement)
    (bind (_asynctaskproxy _payload) (Ability|Async|WaitGameplayEventToActor 0 "(TagName=\"GameplayEvent.SurfaceChange.Cube\")")
      (:then
        (Networking|SwitchHasAuthority
          (:Authority
            (CallFunction|KillAllonSpawn)
            (switch Utilities|FlowControl|Switch|SwitchonE_CubeTownConfig (Variables|FaceBehavior|GetActiveFaceConfig)
              (:NewEnumerator0
                (Variables|FaceBehavior|SetActiveFaces (Variables|FaceBehavior|GetValidFaces)))
              (:NewEnumerator1
                (Utilities|Time|SetTimerbyEvent (EventDispatchers|CreateEvent :!event RandomizeFace) (Variables|FaceBehavior|GetSwitchFacesTimer) true))))
          (:Remote
            (CallFunction|UpdateAmbientAudio))))
      (:EventReceived
        (CallFunction|TryLaunchPawnSurfaceChange (Utilities|Casting|CastToFortPawn (.Target (Utilities|Struct|BreakGameplayEventData _payload))))))
    (else
      (Actor|SetLifeSpan 0.1))))

(event Custom|TryLaunchPawnSurfaceChange
  (params (FortPawnOnSurface :type Fort%20Pawn%20Object%20Reference))
  ;[1 1 1] Surface Change Fail Safe
  (Utilities|FlowControl|Delay 0.2) ; prefer hit event first\nthis will have no effect \nif we already launched player recently
  (Utilities|IsValid FortPawnOnSurface
    (:Is%20Valid
      (if (and (not (GameplayTags|HasMatchingGameplayTag FortPawnOnSurface "(TagName=\"Granted.CorruptionGameplay.Cube.Player.Bounced\")")) (not (Utilities|Array|ContainsItem (Variables|Default|GetRecentlyLaunchedPawns))))
        (if (ScalableFloat|GetValueasBool (Variables|Default|GetCubeImpulseEnabled))
          (CallFunction|LaunchFortPawn (Math|Vector|Normalize (- (Transformation|GetActorLocation FortPawnOnSurface) (Transformation|GetActorLocation))) FortPawnOnSurface)
          (CallFunction|StartLaunchCooldownforPawn FortPawnOnSurface)
          (Utilities|FlowControl|Sequence
            (:then_0
              ;;[1 1 1] Debug Only
              (if (Variables|Default|GetDebugPawnLaunch)
                (Development|PrintString "Surface Pawn Launch Attempt" true true "(R=1.000000,G=0.894344,B=0.000000,A=1.000000)"))))
          (elif (Variables|Default|GetDebugPawnLaunch)
            (Development|PrintString "Skiped Surface Event, already launched!")))))))

(event Collision|EventHit
  (params (MyComp :type Primitive%20Component%20Object%20Reference) (Other :type Actor%20Object%20Reference) (OtherComp :type Primitive%20Component%20Object%20Reference) (bSelfMoved :type Boolean) (HitLocation :type Vector :default "0, 0, 0") (HitNormal :type Vector :default "0, 0, 0") (NormalImpulse :type Vector :default "0, 0, 0") (Hit :type Hit%20Result%20Structure%20%28by%20ref%29))
  (bind (_bblockinghit _binitialoverlap _time _distance _location _impactpoint _normal) (Collision|BreakHitResult Hit))
  (bind (_active _face) (CallFunction|CheckifActiveFace _normal))
  (bind _returnvalue (* _normal -1.0))
  ;[1 1 1] On Hit, check if we hit an active face then decide what to do
  (if _active
    (Variables|Default|SetCubeHitLocation HitLocation)
    (Variables|Default|SetCubeHitNormal HitNormal)
    (bind _isdownwardface (CallFunction|CheckDownwardFace _face))
    (Variables|Default|SetHitDownwardFace _isdownwardface)
    (bind _asfort_pawn (Utilities|Casting|CastToFortPawn Other)
      (:then
        (Utilities|IsValid _asfort_pawn
          (:Is%20Valid
            (Variables|Default|SetFortPawn _asfort_pawn)
            (if (Variables|Default|GetIsCubeTownFloor)
              ;[1 1 1] Ensure we dont bounce non stop
              (CallFunction|LaunchFortPawn _returnvalue (Variables|Default|GetFortPawn))
              (elif (and (not (Utilities|Array|ContainsItem (Variables|Default|GetRecentlyLaunchedPawns) (Variables|Default|GetFortPawn))) (not (GameplayTags|HasMatchingGameplayTag (Variables|Default|GetFortPawn) "(TagName=\"Granted.CorruptionGameplay.Cube.Player.Bounced\")")))
                (if (ScalableFloat|GetValueasBool (Variables|Default|GetCubeImpulseEnabled))
                  ;[1 1 1] Handle Players - impulse regardless of jump
                  (CallFunction|LaunchFortPawn _returnvalue (Variables|Default|GetFortPawn))
                  (CallFunction|StartLaunchCooldownforPawn (Variables|Default|GetFortPawn))
                  (Utilities|FlowControl|Sequence
                    (:then_0
                      ;[1 1 1] Debug Only
                      (if (Variables|Default|GetDebugPawnLaunch)
                        (Development|PrintString "Hit Event Pawn Launch Attempt" true true "(R=1.000000,G=0.085713,B=0.000000,A=1.000000)")))))
                (elif (Variables|Default|GetDebugPawnLaunch)
                  (Development|PrintString "Skiped Hit Event, already launched!")))))))
      (:CastFailed
        ;[1 1 1] Handle Toys
        (bind _asphysics_ball_master (Utilities|Casting|CastToPhysicsBall_Master Other)
          (:then
            (CallFunction|HandleToys _asphysics_ball_master _normal))
          (:CastFailed
            ;[1 1 1] Handle Other Physics
            (bind _asfort_physics_pawn_object (Utilities|Casting|CastToFortPhysicsPawnObject Other)
              (:then
                (PhysicsObject|ApplyImpulse _asfort_physics_pawn_object self (* (Transformation|GetActorUpVector) (Variables|ImpulseValues|GetImpulsePhysicsStrength))))
              (:CastFailed
                (bind _asathena_supply_drop (Utilities|Casting|CastToAthenaSupplyDrop Other)
                  (:then
                    (Actor|DestroyActor _asathena_supply_drop))
                  (:CastFailed
                    _))
                ;[1 1 1] Handle hits that don't fall into any other catagory but should still play FX
                (CallFunction|RippleGC (Variables|Default|GetCubeHitLocation) (* (Variables|Default|GetCubeHitNormal) -1.0) Other 2.0)))))
        (CallFunction|RippleGC (Variables|Default|GetCubeHitLocation) (Variables|Default|GetCubeHitNormal) Other 1.0)))))

(event Custom|TargetAndEliminateBelowCube
  ;[1 1 1] Cube Calls Targeting event which runs through all valid targets, casts to them and kills where possible
  (TargetingSystem|AsyncRequest|StartAsyncTargetingRequest (GameInstanceSubsystems|GetTargetingSubsystem) "/CorruptionGameplay/Gameplay/Environmental/Cube/Cube_KillActors_Targeting.Cube_KillActors_Targeting" (Utilities|Struct|MakeTargetingSourceContext self self (Variables|Default|GetStaticPathEnd)) (delegate Custom|GetTargetsAndKill)))

(event Custom|GetTargetsAndKill
  (params (TargetingRequestHandle :type Targeting%20Request%20Handle%20Structure))
  ;[1 1 1] Cube Calls Targeting event which runs through all valid targets, casts to them and kills where possible
  (for _array_element (TargetingSystem|TargetingResults|GetTargetingResultsActors (GameInstanceSubsystems|GetTargetingSubsystem) TargetingRequestHandle)
    (bind _asbuilding_actor (Utilities|Casting|CastToBuildingActor _array_element)
      (:then
        (Fortnite|ApplyGameplayEffectsToActor _asbuilding_actor (Variables|Default|GetEffects%20to%20Apply) (Transformation|GetWorldLocation (Variables|Default|GetStaticMesh))))
      (:CastFailed
        (bind _asfort_pawn (Utilities|Casting|CastToFortPawn _array_element)
          (:then
            (if (GameplayTags|HasMatchingGameplayTag _asfort_pawn "(TagName=\"Athena.Vehicle.Ostrich\")")
              (Actor|SetLifeSpan _asfort_pawn 0.1)
              (else
                (Environment|ApplyGlobalEnvironmentDamagetoActor _asfort_pawn 1e+04 (Variables|Default|GetAdditional%20Context%20Tags)))))
          (:CastFailed
            (bind _asfort_athena_vehicle (Utilities|Casting|CastToFortAthenaVehicle _array_element))
            (GameplayEffects|ApplyGameplayEffectToSelf :self (Ability|GetFortAbilitySystemComponent _asfort_athena_vehicle) :GameplayEffectClass "/CorruptionGameplay/Gameplay/Environmental/Cube/GE_Cube_KillPawns.GE_Cube_KillPawns_C" :Level 0.0)))))))

(event OnComponentBeginOverlap(Pawn&VehicleOverlap)
  (params (OverlappedComponent :type Primitive%20Component%20Object%20Reference) (OtherActor :type Actor%20Object%20Reference) (OtherComp :type Primitive%20Component%20Object%20Reference) (OtherBodyIndex :type Integer) (bFromSweep :type Boolean) (SweepResult :type Hit%20Result%20Structure%20%28by%20ref%29))
  (bind (_bblockinghit _binitialoverlap _time _distance _location _impactpoint _normal) (Collision|BreakHitResult SweepResult))
  ;[1 1 1] Handle Vehicles
  (bind _asfort_athena_vehicle (Utilities|Casting|CastToFortAthenaVehicle OtherActor)
    (:then
      (CallFunction|LaunchVehicle _normal _asfort_athena_vehicle _location))
    (:CastFailed
      ;[1 1 1] Gives Pawns no fall damage until landing.
      (bind _asfort_player_pawn_athena (Utilities|Casting|CastToFortPlayerPawnAthena OtherActor))
      (Utilities|IsValid (Ability|GetAbilitySystemComponent _asfort_player_pawn_athena)
        (:Is%20Valid
          (bind _returnvalue (GameplayEffects|ApplyGameplayEffectToSelf :self (Ability|GetAbilitySystemComponent _asfort_player_pawn_athena) :GameplayEffectClass "/Game/Athena/Items/Parents/FallImmunity/GE_Athena_TillLandFallDamageImmunity_Grant_ApplyInAirOnly.GE_Athena_TillLandFallDamageImmunity_Grant_ApplyInAirOnly_C" :Level 0.0))
          (Variables|Default|SetFallDamageImmuneGE _returnvalue))
        (:Is%20Not%20Valid
          (Development|PrintString "NOT VALID" true true "(R=0.000000,G=0.660000,B=1.000000,A=1.000000)" 1e+01))))))

(event Custom|ShiftFacesTimeline
  (bind _returnvalue_x (Math|Vector|BreakVector (Transformation|GetActorScale3D)))
  (bind _staticmesh (Variables|Default|GetStaticMesh))
  ;[0.9096 0.4331 1] Animation for changing active face
  (Variables|Default|SetActiveFaceMaterialPrev (Variables|Default|GetActiveFaceMaterial))
  (bind (_direction _newtrack_0 _zap) (Timeline|FaceChangeVisuals PlayFromStart)
    (:Update
      (bind _output_get (Variables|Default|SetActiveFaceMaterial (Math|Vector|Lerp(Vector) (Variables|Default|GetActiveFaceMaterialPrev) (* (Transformation|GetForwardVector (Variables|Default|GetArrow)) (* 1.6e+03 _returnvalue_x)) _newtrack_0)))
      (Rendering|Material|SetVectorParameterValueonMaterials _staticmesh "SphereOffset" _output_get)
      (Rendering|Material|SetScalarParameterValueonMaterials _staticmesh "Zap" _zap))))

(event OnComponentEndOverlap(Pawn&VehicleOverlap)
  (params (OverlappedComponent :type Primitive%20Component%20Object%20Reference) (OtherActor :type Actor%20Object%20Reference) (OtherComp :type Primitive%20Component%20Object%20Reference) (OtherBodyIndex :type Integer))
  ;[1 1 1] Removes fall damage immune and jump GE
  (bind _asfort_player_pawn_athena (Utilities|Casting|CastToFortPlayerPawnAthena OtherActor))
  (Utilities|IsValid _asfort_player_pawn_athena
    (:Is%20Valid
      (GameplayEffects|RemoveActiveGameplayEffect (Ability|GetFortAbilitySystemComponent _asfort_player_pawn_athena) (Variables|Default|GetFallDamageImmuneGE) -1))))

(event EventStaticPathBeginPlaySkipOccurred
  ;[1 1 1] Begin Play Skip
  (|CallParentFunction)
  (Class|FortAsyncActionAuthorityMatchReady|MatchReadyAsync
    (:OnReady
      (CallFunction|KillAllonSpawn)
      (CallFunction|AwakenCube))))

(event Custom|LandingFeedback
  ;[1 0 0] FX, audio, cam shake and FF handled in GC
  (bind _facesocket (CallFunction|FindFaceSocket "0, 0,-1.000000"))
  (Variables|Default|SetGroundSocket _facesocket)
  (GameplayCue|ExecuteGameplayCue%28Burst%29Local self (Variables|GameplayCues|GetGC_CubeLand) (Ability|GameplayCue|MakeGameplayCueParameters :NormalizedMagnitude 0.0 :RawMagnitude 0.0 :Location "0,0,0" :Normal "0,0,0" :GameplayEffectLevel 1 :AbilityLevel (Math|Conversions|ToInteger(Byte) (Variables|Default|GetCubeSize)) :bReplicateLocationWhenUsingMinimalRepProxy false)))

(event EventRollEnd
  ;[1 1 1] Event on Roll End
  (Navigation|SetCanEverAffectNavigation true (Variables|Default|GetStaticMesh)) ; Set for AI navigation once the cube has stopped moving
  (Audio|Components|Audio|FadeIn (Variables|Default|GetAmbientAudio) 3.0 1.0 0.0 "Linear" :!class /Script/Engine.AudioComponent)
  (CallFunction|TargetandEliminateBelowCube)
  (CallFunction|LandingFeedback)
  (Projectile|RemoveGameplayCue%28Looping%29Local self (Variables|GameplayCues|GetGC_CubeRoll) (Ability|GameplayCue|MakeGameplayCueParameters 0.0)))

(event EventRollStart
  ;[1 1 1] Event  on roll start
  (|CallParentFunction)
  (Navigation|SetCanEverAffectNavigation false (Variables|Default|GetStaticMesh)) ; Does not affect navigation while moving
  (Audio|Components|Audio|FadeOut (Variables|Default|GetAmbientAudio) 5.0 0.0 "Linear" :!class /Script/Engine.AudioComponent)
  (bind _facesocket (CallFunction|FindFaceSocket "0, 0,-1.000000"))
  (Variables|Default|SetGroundSocket _facesocket)
  (Projectile|AddGameplayCue%28Looping%29Local self (Variables|GameplayCues|GetGC_CubeRoll) (Ability|GameplayCue|MakeGameplayCueParameters 0.0))
  (Development|PrintString "STARTED ROLLING" true true "(R=1.000000,G=0.519027,B=0.337497,A=1.000000)" 5.0)
  (Variables|Default|SetShouldOnlyDestroyBuildings true)
  (CallFunction|KillAllonSpawn)
  (Utilities|FlowControl|Delay (* (Variables|Default|GetMoveDurationSeconds) 0.75))
  (CallFunction|TargetandEliminateBelowCube)
  (Development|PrintString "KILL" true true "(R=1.000000,G=0.519027,B=0.337497,A=1.000000)" 5.0)
  (Variables|Default|SetShouldOnlyDestroyBuildings false))

(fn AddDamage
  (params (Face :type E_CubeFace%20Enum%20%28by%20ref%29 :default "NewEnumerator7") (Damage :type Float%20%28double-precision%29%20%28by%20ref%29))
  (bind _facedamage (Variables|FaceBehavior|Maps|GetFaceDamage))
  (bind (_value _returnvalue) (Utilities|Map|Find _facedamage Face))
  (bind _facedamage_2 (Variables|FaceBehavior|Maps|GetFaceDamage))
  (if _returnvalue
    (LocalVariables|SetLocalDamage _value))
  (Utilities|Map|Add (Variables|FaceBehavior|Maps|GetFaceDamage) Face (+ Damage (LocalVariables|GetLocalDamage)))
  (bind _keys (Utilities|Map|Keys (Variables|FaceBehavior|Maps|GetFaceDamage)))
  (LocalVariables|SetLocal%20Face%20Array (Variables|FaceBehavior|GetActiveFaces))
  (for _array_element _keys
    (if (.ReturnValue (Utilities|Map|Find _facedamage_2 _array_element))
      (if (>= (Utilities|Map|Find _facedamage_2 _array_element) (Variables|FaceBehavior|GetDamageRequiredToActivateFace))
        (Utilities|Array|AddUnique (LocalVariables|GetLocal%20Face%20Array) _array_element)
        (Development|PrintString "Face Active" true true "(R=0.000000,G=0.660000,B=1.000000,A=1.000000)" 1e+01)
        (else
          _))
      (else
        _))
    (Utilities|Array|RemoveItem (LocalVariables|GetLocal%20Face%20Array) _array_element))
  (Variables|FaceBehavior|SetActiveFaces (LocalVariables|GetLocal%20Face%20Array)))

(fn CheckDownwardFace
  (params (HitFace :type E_CubeFace%20Enum :default "NewEnumerator7"))
  (bind _returnvalue (Transformation|GetActorTransform))
  (bind _facemap (Variables|FaceBehavior|Maps|GetFaceMap))
  (bind _keys (Utilities|Map|Keys _facemap))
  (for _array_element _keys
    (if (> (Math|Vector|DotProduct (Math|Transform|TransformDirection _returnvalue (Utilities|Map|Find _facemap _array_element)) "0, 0,-1.000000") (LocalVariables|GetClosestToDown))
      (LocalVariables|SetClosestToDown (Math|Vector|DotProduct (Math|Transform|TransformDirection _returnvalue (Utilities|Map|Find _facemap _array_element)) "0, 0,-1.000000"))
      (LocalVariables|SetClosestToBottomFace _array_element)))
  (return (Utilities|Enum|Equal(Enum) HitFace (LocalVariables|GetClosestToBottomFace))))

(fn CheckIfActiveFace
  (params (Hit%20Normal :type Vector))
  (bind _facemap (Variables|FaceBehavior|Maps|GetFaceMap))
  (for _array_element (Variables|FaceBehavior|GetActiveFaces)
    (if (.ReturnValue (Utilities|Map|Find _facemap _array_element))
      (if (Math|Vector|Equal(Vector) (Utilities|Map|Find _facemap _array_element) (Math|Vector|UnrotateVector (* Hit%20Normal -1.0) (Transformation|GetActorRotation)) 0.01)
        (return true _array_element))))
  (return false "NewEnumerator7"))

(fn FindFaceSocket
  (params (LookAtDirection :type Vector))
  (bind _staticmesh (Variables|Default|GetStaticMesh))
  (LocalVariables|SetNearestSocket -1.0)
  (for _array_element (Utilities|Array|MakeArray "A" "B" "C" "D" "E" "F")
    (if (> (Math|Vector|DotProduct (Math|Vector|GetForwardVector (Transformation|GetSocketRotation _staticmesh _array_element)) LookAtDirection) (LocalVariables|GetNearestSocket))
      (LocalVariables|SetNearestSocket (Math|Vector|DotProduct (Math|Vector|GetForwardVector (Transformation|GetSocketRotation _staticmesh _array_element)) LookAtDirection))
      (LocalVariables|SetBestSocket _array_element)))
  (return (LocalVariables|GetBestSocket)))

(fn GetFaceFromDirection
  (params (Hit%20Normal :type Vector))
  (bind _facemap (Variables|FaceBehavior|Maps|GetFaceMap))
  (for _array_element (Variables|FaceBehavior|GetValidFaces)
    (if (.ReturnValue (Utilities|Map|Find _facemap _array_element))
      (if (Math|Vector|Equal(Vector) (Utilities|Map|Find _facemap _array_element) (Math|Vector|UnrotateVector Hit%20Normal (Transformation|GetActorRotation)) 0.01)
        (return _array_element))))
  (return "NewEnumerator7"))

(fn GetPathHeightOffset
  (bind (_min _max) (Components|StaticMesh|GetLocalBounds (Variables|Default|GetStaticMesh)))
  (return (* (/ (.z (- _max _min)) 2.0) 0.9)))

(fn GetScriptedObjectStaticMeshComponent
  (return (Variables|Default|GetStaticMesh)))

(fn HandleProjectiles
  (params (Projectile :type Fort%20Projectile%20Base%20Object%20Reference) (HitNormal :type Vector))
  (if (not (Actor|ActorHasTag Projectile "NoBounce"))
    (Variables|/Script/Engine.MovementComponent|SetVelocity (Math|Vector|MirrorVectorbyNormal (Variables|/Script/Engine.MovementComponent|GetVelocity (Variables|/Script/FortniteGame.FortProjectileBase|GetProjectileMovementComponent Projectile)) HitNormal) (Variables|/Script/FortniteGame.FortProjectileBase|GetProjectileMovementComponent Projectile))))

(fn HandleToys
  (params (Toy :type Physics%20Ball%20Master%20Object%20Reference) (Normal :type Vector))
  (bind _projectilemovement (Variables|/Game/Building/ActorBlueprints/Prop/PhysicsBall_Master.PhysicsBall_Master_C|GetProjectileMovement Toy))
  (bind _velocity (Variables|/Script/Engine.MovementComponent|GetVelocity _projectilemovement))
  (bind _returnvalue (* _velocity 1.2))
  (bind _returnvalue_2 (Math|Vector|MirrorVectorbyNormal _returnvalue Normal))
  (bind _projectilegravityscale (Variables|/Script/Engine.ProjectileMovementComponent|GetProjectileGravityScale _projectilemovement))
  (if (> _projectilegravityscale 0.0)
    (Variables|/Script/Engine.ProjectileMovementComponent|SetProjectileGravityScale (* _projectilegravityscale -0.5) _projectilemovement)
    (Variables|/Script/Engine.MovementComponent|SetVelocity _returnvalue_2 _projectilemovement)
    (else
      (Variables|/Script/Engine.ProjectileMovementComponent|SetProjectileGravityScale (+ _projectilegravityscale (Variables|Default|GetZeroGravPrj)) _projectilemovement)
      (Variables|/Script/Engine.MovementComponent|SetVelocity _returnvalue_2 _projectilemovement))))

(fn LaunchFortPawn
  (params (Normal :type Vector) (FortPawn :type Fort%20Pawn%20Object%20Reference))
  (bind _impulseplayerstrengthmult (Variables|ImpulseValues|GetImpulsePlayerStrengthMult))
  (bind _returnvalue (Math|Vector|VectorLength (Pawn|Components|CharacterMovement|GetCurrentAcceleration (Variables|/Script/Engine.Character|GetCharacterMovement FortPawn))))
  (bind _returnvalue_2 (> _returnvalue 0.0))
  (bind _returnvalue_3 (Pawn|Components|CharacterMovement|GetLastUpdateVelocity (Variables|/Script/Engine.Character|GetCharacterMovement FortPawn)))
  (bind _returnvalue_4 (Math|Vector|Normalize _returnvalue_3))
  (bind _returnvalue_5 (select _returnvalue_2 (Math|Vector|Normalize (Pawn|Components|CharacterMovement|GetCurrentAcceleration (Variables|/Script/Engine.Character|GetCharacterMovement FortPawn))) _returnvalue_4))
  (bind _returnvalue_6 (Math|Vector|MirrorVectorbyNormal _returnvalue_5 Normal))
  (bind _returnvalue_7 (* _returnvalue_6 _impulseplayerstrengthmult))
  (bind _maxplayerlaunchvel (Variables|Default|GetMaxPlayerLaunchVel))
  (bind _returnvalue_8 (ScalableFloat|GetValueatLevel _maxplayerlaunchvel))
  (bind _returnvalue_9 (Math|Vector|Normalize (Pawn|Components|CharacterMovement|GetCurrentAcceleration (Variables|/Script/Engine.Character|GetCharacterMovement FortPawn))))
  (bind _hitdownwardface (Variables|Default|GetHitDownwardFace))
  (bind _impulseplayeradditionalz (Variables|ImpulseValues|GetImpulsePlayerAdditionalZ))
  (bind _returnvalue_10 (* _impulseplayeradditionalz -1.0))
  (bind _impulseplayeradditionaly (Variables|Default|GetImpulsePlayerAdditionalY))
  (bind _impulseplayeradditionalx (Variables|Default|GetImpulsePlayerAdditionalX))
  (bind _returnvalue_11 (Math|Vector|vector+vector _returnvalue_7 (Math|Vector|MakeVector _impulseplayeradditionalx _impulseplayeradditionaly _returnvalue_10)))
  (bind _impulseplayeradditionalz_2 (Variables|ImpulseValues|GetImpulsePlayerAdditionalZ))
  (bind _impulseplayeradditionaly_2 (Variables|Default|GetImpulsePlayerAdditionalY))
  (bind _impulseplayeradditionalx_2 (Variables|Default|GetImpulsePlayerAdditionalX))
  (bind _returnvalue_12 (Math|Vector|vector+vector _returnvalue_7 (Math|Vector|MakeVector _impulseplayeradditionalx_2 _impulseplayeradditionaly_2 _impulseplayeradditionalz_2)))
  (bind _returnvalue_13 (select _hitdownwardface _returnvalue_11 _returnvalue_12))
  (bind _returnvalue_14 (Math|Vector|ClampVectorSize _returnvalue_13 0.0 _returnvalue_8))
  (bind _returnvalue_15 (Ability|GetFortAbilitySystemComponent FortPawn))
  (bind _charactermovement (Variables|/Script/Engine.Character|GetCharacterMovement FortPawn))
  (bind _returnvalue_16 (Pawn|Components|CharacterMovement|GetCurrentAcceleration _charactermovement))
  (Variables|/Script/FortniteGame.FortPawn|SetbIgnoreNextFallingDamage true FortPawn)
  (if (not (Health|IsDBNO FortPawn))
    (Character|LaunchCharacter FortPawn _returnvalue_14 true false)
    (Networking|SwitchHasAuthority
      (:Authority
        (GameplayEffects|ApplyGameplayEffectToSelf :self _returnvalue_15 :GameplayEffectClass "/CorruptionGameplay/Gameplay/Environmental/Cube/GE_Cube_BouncedPlayer.GE_Cube_BouncedPlayer_C" :Level 0.0)
        (GameplayEffects|ApplyGameplayEffectToSelf :self _returnvalue_15 :GameplayEffectClass "/CorruptionGameplay/Gameplay/Environmental/Cube/GE_Cube_LowGrav.GE_Cube_LowGrav_C" :Level 0.0)
        (GameplayEffects|ApplyGameplayEffectToSelf :self _returnvalue_15 :GameplayEffectClass "/Game/Athena/Items/Parents/FallImmunity/GE_Athena_TillLandFallDamageImmunity_NoFX_Grant_Parent.GE_Athena_TillLandFallDamageImmunity_NoFX_Grant_Parent_C" :Level 0.0)
        (if (GameplayTags|HasMatchingGameplayTag (Variables|Default|GetFortPawn) "(TagName=\"Granted.CorruptionGameplay.Cube.Player.GCPlayed\")")
          (bind _outhits (Collision|MultiCapsuleTraceForObjects (Transformation|GetActorLocation (Variables|Default|GetFortPawn)) (+ (Transformation|GetActorLocation (Variables|Default|GetFortPawn)) _returnvalue_14) 8e+01 1.5e+02 (Variables|Default|GetDestroyObjectTypes) false (Utilities|Array|MakeArray self (Variables|Default|GetFortPawn)) "None" true "(R=1.000000,G=0.000000,B=0.000000,A=1.000000)" "(R=0.000000,G=1.000000,B=0.000000,A=1.000000)" 8.0))
          (for _array_element _outhits
            (if (Variables|/Script/Engine.Actor|GetbCanBeDamaged (.HitActor (Collision|BreakHitResult _array_element)))
              (bind _bdidactorhavetags (Class|AthenaQuestFunctionLibrary|AthenaCheckActorHasTags (.HitActor (Collision|BreakHitResult _array_element)) "(GameplayTags=((TagName=\"Deimos.Spawner\")))"))
              (if (not _bdidactorhavetags)
                (Utilities|Casting|CastToBuildingTrap (.HitActor (Collision|BreakHitResult _array_element))
                  (:CastFailed
                    (GameplayEffects|ApplyGameplayEffectToTarget :self (Ability|GetAbilitySystemComponent (Variables|Default|GetFortPawn)) :GameplayEffectClass "/CorruptionGameplay/Gameplay/Environmental/Cube/GE_Cube_KillActors.GE_Cube_KillActors_C" :Target (Ability|GetAbilitySystemComponent (.HitActor (Collision|BreakHitResult _array_element))) :Level 0.0))))))
          (else
            (GameplayEffects|ApplyGameplayEffectToSelf :self _returnvalue_15 :GameplayEffectClass "/CorruptionGameplay/Gameplay/Environmental/Cube/GE_Cube_CueSpamPrevention.GE_Cube_CueSpamPrevention_C" :Level 0.0))))
      (:Remote
        (if (not (GameplayTags|HasMatchingGameplayTag (Variables|Default|GetFortPawn) "(TagName=\"Granted.CorruptionGameplay.Cube.Player.GCPlayed\")"))
          (GameplayCue|ExecuteGameplayCue%28Burst%29Local FortPawn (Variables|GameplayCues|GetGC_PlayerBounce) (Ability|GameplayCue|MakeGameplayCueParameters :NormalizedMagnitude 0.0 :RawMagnitude 0.0 :Location "0,0,0" :Normal Normal :GameplayEffectLevel 1 :AbilityLevel (Variables|Default|GetCubeColorIndex) :bReplicateLocationWhenUsingMinimalRepProxy false)))))))

(fn LaunchPlayerOnJump
  (params (Player :type Fort%20Player%20Pawn%20Athena%20Object%20Reference) (Normal :type Vector))
  (if (and (Math|Boolean|ORBoolean (Variables|/Script/Engine.Character|GetbWasJumping Player) (AI|Components|NavMovement|IsFalling (Variables|/Script/Engine.Character|GetCharacterMovement Player)) (GameplayTags|HasMatchingGameplayTag (Ability|GetFortAbilitySystemComponent Player) "(TagName=\"MovementMode.Falling\")")) (not (Health|IsDBNO Player)))
    _))

(fn LaunchVehicle
  (params (Normal :type Vector) (Vehicle :type Fort%20Athena%20Vehicle%20Object%20Reference) (Location :type Vector))
  (bind _returnvalue (Transformation|GetActorLocation))
  (bind _returnvalue_2 (Transformation|GetActorLocation Vehicle))
  (bind _returnvalue_3 (- _returnvalue_2 _returnvalue))
  (bind _returnvalue_4 (Transformation|GetActorTransform))
  (bind _largestfaceprojection_e (LocalVariables|GetLargestFaceProjection_E))
  (bind _facemap (Variables|FaceBehavior|Maps|GetFaceMap))
  (bind (_value _returnvalue_5) (Utilities|Map|Find _facemap _largestfaceprojection_e))
  (if (not (GameplayTags|HasMatchingGameplayTag Vehicle "(TagName=\"Granted.CorruptionGameplay.Cube.Vehicle.Impulsed\")"))
    (Vehicle|SetIgnoreNextFallingDamage Vehicle true)
    (bind _keys (Utilities|Map|Keys _facemap))
    (LocalVariables|SetLargestFaceProjection -9999999827968.0)
    (for _array_element _keys
      (if (> (Math|Vector|DotProduct (Math|Transform|TransformDirection _returnvalue_4 (Utilities|Map|Find _facemap _array_element)) _returnvalue_3) (LocalVariables|GetLargestFaceProjection))
        (LocalVariables|SetLargestFaceProjection (Math|Vector|DotProduct (Math|Transform|TransformDirection _returnvalue_4 (Utilities|Map|Find _facemap _array_element)) _returnvalue_3))
        (LocalVariables|SetLargestFaceProjection_E _array_element)))
    (if _returnvalue_5
      (Audio|BroadcastAbilityImpactatLocation "/Game/Athena/Sounds/Events/Cube/Cube_LaunchPlayer_Cue.Cube_LaunchPlayer_Cue" Vehicle (Math|Vector|MakeVector (.x (+ (* (Math|Vector|ProjectVectorontoVector (Vehicle|GetVehicleVelocity Vehicle) (Math|Vector|Normalize (Math|Vector|MakeVector (.x _value) (.y _value)))) -2.0) (Vehicle|GetVehicleVelocity Vehicle))) (.y (+ (* (Math|Vector|ProjectVectorontoVector (Vehicle|GetVehicleVelocity Vehicle) (Math|Vector|Normalize (Math|Vector|MakeVector (.x _value) (.y _value)))) -2.0) (Vehicle|GetVehicleVelocity Vehicle))) (+ (.z (+ (* (Math|Vector|ProjectVectorontoVector (Vehicle|GetVehicleVelocity Vehicle) (Math|Vector|Normalize (Math|Vector|MakeVector (.x _value) (.y _value)))) -2.0) (Vehicle|GetVehicleVelocity Vehicle))) (select (Utilities|Enum|Equal(Enum) (LocalVariables|GetLargestFaceProjection_E) "NewEnumerator5") (* (Variables|ImpulseValues|GetImpulseStrengthZVehicle) -1.0) (Variables|ImpulseValues|GetImpulseStrengthZVehicle)))) 7.8e+02) ; Set Physics Linear Velocity
      (GameplayEffects|ApplyGameplayEffectToSelf :self (Ability|GetFortAbilitySystemComponent Vehicle) :GameplayEffectClass "/CorruptionGameplay/Gameplay/Environmental/Cube/GE_Cube_ImpulsedVehicle.GE_Cube_ImpulsedVehicle_C" :Level 0.0))))

(fn OnRep_ActiveFace
  (bind _facemap (Variables|FaceBehavior|Maps|GetFaceMap))
  (Networking|CanExecuteCosmeticEvents
    (:True
      (for _array_element (Variables|FaceBehavior|GetActiveFaces)
        ;[1 1 1] Temp visibility logic - to be replaced by art
        (if (.ReturnValue (Utilities|Map|Find _facemap _array_element))
          (Transformation|SetRelativeRotation (Variables|Default|GetArrow) (Math|Rotator|MakeRotfromX (Utilities|Map|Find _facemap _array_element)) false false))))))

(fn RandomizeFace
  (Networking|SwitchHasAuthority
    (:Authority
      (bind _output_get (LocalVariables|SetLocalSelectedFaces (Variables|FaceBehavior|GetValidFaces)))
      (Utilities|Array|RemoveItem _output_get (Utilities|Array|Get(acopy) (Variables|FaceBehavior|GetActiveFaces)))
      (bind _output_get_2 (LocalVariables|SetRandomFace (Utilities|Array|MakeArray (Utilities|Array|RandomArrayItem (LocalVariables|GetLocalSelectedFaces)))))
      (Variables|FaceBehavior|SetActiveFaces _output_get_2))))

(fn ShouldDie
  (params (Damage :type Float%20%28single-precision%29) (EventInstigator :type Controller%20Object%20Reference) (DamageCauser :type Actor%20Object%20Reference) (bIsTest :type Boolean) (EffectContext :type Gameplay%20Effect%20Context%20Handle%20Structure%20%28by%20ref%29))
  (return false))

(fn UpdateAmbientAudio
  (bind _ambientaudio (Variables|Default|GetAmbientAudio))
  (Audio|Parameter|SetIntegerParameter (Variables|Default|GetAmbientAudio) "Size" (Math|Conversions|ToInteger(Byte) (Variables|Default|GetCubeSize)) :!class /Script/AudioExtensions.AudioParameterControllerInterface)
  (if (Variables|Audio|GetAmbientAudioEnabled)
    (Audio|Components|Audio|FadeIn _ambientaudio 3.0 1.0 0.0 "Linear" :!class /Script/Engine.AudioComponent)
    (else
      (Audio|Components|Audio|FadeOut _ambientaudio 3.0 0.0 "Linear" :!class /Script/Engine.AudioComponent))))

(fn ConstructionScript)
