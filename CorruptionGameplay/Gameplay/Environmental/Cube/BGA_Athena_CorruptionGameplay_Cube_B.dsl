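; Corruption gameplay cube actor that gives shield effects to pawns it launches and removes itself when cubes are disabled.
; On BeginPlay, the authority activates all valid faces or starts a looping RandomizeFace timer, per the face config, while remotes update ambient audio.
; The LaunchFortPAwn override checks jump/fall and DBNO state for impulse-on-jump, and applies shielding alongside the parent launch.
; Shield and FX are separate gameplay effects so visuals play only on application, not each tick.
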(event EventBeginPlay
  (if (ScalableFloat|GetValueasBool (Variables|Default|GetCubesEnabled))
    (Networking|SwitchHasAuthority
      (:Authority
        (switch Utilities|FlowControl|Switch|SwitchonE_CubeTownConfig (Variables|FaceBehavior|GetActiveFaceConfig)
          (:NewEnumerator0
            (Variables|FaceBehavior|SetActiveFaces (Variables|FaceBehavior|GetValidFaces)))
          (:NewEnumerator1
            (Utilities|Time|SetTimerbyEvent (EventDispatchers|CreateEvent :!event RandomizeFace) (Variables|FaceBehavior|GetSwitchFacesTimer) true))))
      (:Remote
        (CallFunction|UpdateAmbientAudio)))
    (else
      (Actor|SetLifeSpan 0.1))))

(event Collision|EventActorBeginOverlap
  (params (OtherActor :type Actor%20Object%20Reference))
  (|CallParentFunction OtherActor) ; This node is disabled and will not be called.\nDrag off pins to build functionality.
)

(event EventTick
  (params (DeltaSeconds :type Float%20%28single-precision%29))
  (|CallParentFunction DeltaSeconds) ; This node is disabled and will not be called.\nDrag off pins to build functionality.
)

(fn ApplyShield
  (params (Actor :type Actor%20Object%20Reference))
  (bind _self self)
  (bind _returnvalue (Ability|GetFortAbilitySystemComponent _self))
  (bind _returnvalue_2 (Ability|GetFortAbilitySystemComponent Actor))
  (bind _returnvalue_3 (GameplayEffects|MakeFortEffectContext _returnvalue _self _self 0 0))
  ;[1 1 1] Apply the shield effect and the visuals for the shield effect (these need to be different so it only plays on application not on tick of the GE)
  (GameplayEffects|FortApplyGameplayEffectToTarget _returnvalue (Variables|Default|GetGE_Shielding) _returnvalue_2 _self _self 0 0.0 _returnvalue_3)
  (GameplayEffects|FortApplyGameplayEffectToTarget _returnvalue (Variables|Default|GetGE_ShieldingFX) _returnvalue_2 _returnvalue 0 0 0.0 _returnvalue_3))

(fn LaunchFortPAwn
  (params (Normal :type Vector :default "0, 0, 0") (FortPawn :type Fort%20Pawn%20Object%20Reference))
  (if (Variables|FaceBehavior|GetImpulseOnJump?)
    (if (and (Math|Boolean|ORBoolean (GameplayTags|HasMatchingGameplayTag (Ability|GetFortAbilitySystemComponent FortPawn) "(TagName=\"MovementMode.Falling\")") (Variables|/Script/Engine.Character|GetbWasJumping FortPawn) (AI|Components|NavMovement|IsFalling (Variables|/Script/Engine.Character|GetCharacterMovement FortPawn))) (not (Health|IsDBNO FortPawn)))
      _))
  (CallFunction|ApplyShield FortPawn)
  (|CallParentFunction Normal FortPawn))

(fn LaunchPlayerOnJump
  (params (Player :type Fort%20Player%20Pawn%20Athena%20Object%20Reference) (Normal :type Vector :default "0, 0, 0"))
  (|CallParentFunction Player Normal))

(fn ConstructionScript
  (|CallParentFunction))
