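# Additive, unlit, two-sided material for the Corruption Cube's aura, used on Niagara mesh particles, with vertex blue separating cube faces from spline strips.
# Faces show panning world-aligned caustics warped by scrolling noise; strips use an oversaturated corona gradient mixed with the caustics.
# UV-edge glow and particle color tint the emissive, while vertex red and depth fades shape opacity.
# SplineThicken world position offset widens the strips, scaled by a particle dynamic parameter.
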Material /CorruptionGameplay/Gameplay/Environmental/Cube/FX/M_Corruption_Cube_Aura
graph {
 properties {
  bUsedWithNiagaraMeshParticles=true
  blendMode=BLEND_Additive
  pixelDepthOffsetMode=PDOM_Legacy
  previewMesh=&/SM_Corruption_Cube_Aura
  shadingModel=MSM_Unlit
  twoSided=true
  usedShadingModels=Unlit
 }
 nodes [
  n1 = VectorParameter() {parameterName=Tile, defaultValue=<rgba 700 700 1000 0>}
  n2 = Multiply(A=Multiply(A=@n1.RGB, B=VectorParameter() {parameterName=PanSpeed, defaultValue=<rgba 0.05 -0.05 -0.15 0>}.RGB), B=Time())
  n3 = WorldPosition()
  n4 = TextureCoordinate() {vTiling=2}
  n5 = TextureCoordinate() {uTiling=3, vTiling=0.5}
  n6 = VertexColor()
  n7 = Panner(Coordinate=Multiply(A=LinearInterpolate(A=@n4, B=@n5, Alpha=@n6.B), B=ScalarParameter() {defaultValue=1, parameterName=DistortionTile})) {speedX=-0.04, speedY=-0.1}
  n8 = Multiply(
   A=Multiply(A=@n1.RGB, B=TextureSample(UVs=@n7) {texture=/Game/Effects/Fort_Effects/Textures/Tile/T_Smooth_Tile_Chromatic}.RGB)
   B=ScalarParameter() {defaultValue=0.5, parameterName=DistortionMult}
  )
  n9 = OneMinus(None=VertexColor().B)
  n10 = MaterialFunctionCall(
   TextureObject=TextureObject() {texture=/Game/Effects/Fort_Effects/Textures/Tile/T_ShieldCaustics_Tile_Chromatic}
   TextureSize=@n1.RGB
   WorldPosition=Add(A=Add(A=@n2, B=@n3.XYZ), B=Multiply(A=@n8, B=@n9))
  ) {materialFunction=&WorldAlignedTexture}
  n11 = MaterialFunctionCall(XY=Add(A=TextureCoordinate(), B=ScalarParameter() {defaultValue=-0.0026, parameterName=ChromaOffset})) {materialFunction=&Swizzle}
  n12 = Add(
   A=Multiply(A=NamedRerouteUsage(), B=TextureSample(UVs=@n11.YX) {texture=/Game/Effects/Fort_Effects/Textures/Grad/T_Halo_Corona_Gradient}.RGB)
   B=Multiply(A=@n10.XYZ Texture) {constB=0.5}
  ) {constB=0.25}
  n13 = Constant() {r=-1}
  n14 = LinearInterpolate(
   A=Add(A=@n10.XYZ Texture, B=ScalarParameter() {defaultValue=0.4, parameterName=TextureAdd})
   B=Multiply(A=Desaturation(None=@n12, Fraction=@n13), B=ScalarParameter() {defaultValue=1, parameterName=RainbowBrightness})
   Alpha=VertexColor().B
  )
  n15 = MaterialFunctionCall(Float2=TextureCoordinate()) {materialFunction=&BreakOutFloat2Components}
  n16 = Constant() {r=0.5}
  n17 = ScalarParameter() {defaultValue=0.5, parameterName=EdgeGlowRadius}
  n18 = ScalarParameter() {defaultValue=0.8, parameterName=EdgeGlowHardness}
  n19 = Max(
   A=OneMinus(None=SphereMask(A=@n15.R, B=@n16, Radius=@n17, Hardness=@n18) {hardnessPercent=0})
   B=OneMinus(None=SphereMask(A=@n15.G, B=@n16, Radius=@n17, Hardness=@n18) {hardnessPercent=0})
  )
  n20 = Add(
   A=VectorParameter() {parameterName=BaseGlow, defaultValue=<rgba 0.6 0.2925 0.0961 0>}.RGB
   B=Multiply(A=VectorParameter() {parameterName=EdgeGlow, defaultValue=<rgba 1 0.7517 0.0351 0>}.RGB, B=Power(Base=@n19) {constExponent=8})
  )
  n21 = ParticleColor()
  MaterialFunctionCall(Emissive=Multiply(A=Multiply(A=@n14, B=@n20), B=@n21.RGB)) {materialFunction=&MF_FX_EmissiveAdjustments}
  n22 = VertexColor()
  n23 = Constant() {r=0.5}
  n24 = SphereMask(A=ComponentMask(None=TextureCoordinate()) {r=true}, B=@n23, Radius=@n23) {hardnessPercent=0}
  n25 = LinearInterpolate(
   A=LinearInterpolate(A=ScalarParameter() {defaultValue=0.8, parameterName=FaceOpacity}, Alpha=DepthFade()) {constB=0}
   B=DepthFade(Opacity=@n24, FadeDistance=ScalarParameter() {parameterName=SplineDepthFade})
   Alpha=@n22.B
  ) {constA=0.5, constB=0}
  Multiply(A=@n21.A, B=Power(Base=Multiply(A=@n22.R, B=@n25)) {constExponent=3})
  NamedRerouteDeclaration(None=@n24) {name=R-SphereMask, nodeColor=<rgba 1 0.6563 0 1>}
  n26 = Multiply(
   A=ScalarParameter() {defaultValue=300, parameterName=SplineWidth}
   B=DynamicParameter() {paramNames [SplineWidthMult, Param2, Param3, Param4]}.SplineWidthMult
  )
  Reroute(None=Multiply(
    A=@n22.B
    B=MaterialFunctionCall(WidthBase=@n26, WidthTip=@n26) {materialFunction=&SplineThicken}.WorldPositionOffset
   ))
 ]
}
