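# Masked mesh-particle material that reveals a Corruption cube inside a growing sphere mask, driven by Niagara dynamic parameters for reveal, radius and final glow.
# Two sphere masks form a glowing edge band, broken up by panning dust and world-aligned caustics that feed emissive and dithered opacity mask.
# Bump-offset sparkles and height-blended edge tint build the surface color, with normals flattened outside the band.
# A PreviousFrameSwitch offset fakes motion vectors.
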Material /CorruptionGameplay/Gameplay/Environmental/Cube/FX/M_Corruption_Cube_Reveal
graph {
 properties {bUsedWithNiagaraMeshParticles=true, blendMode=BLEND_Masked, pixelDepthOffsetMode=PDOM_Legacy, twoSided=true, usedShadingModels=Default Lit}
 nodes [
  Divide(A=Time() {bOverride_Period=true, period=3}) {constB=3}
  n1 = WorldPosition()
  n2 = Constant3Vector() {constant=<rgba 0 0 -1 0>}
  n3 = Power(Base=NamedRerouteUsage()) {constExponent=0.4}
  n4 = DynamicParameter() {defaultValue=<rgba 1 768 0 1>, paramNames [Reveal,CubeRadius,FinalGlow,Param4]}
  n5 = Add(A=ParticlePositionWS(), B=Multiply(A=LinearInterpolate(A=@n2, Alpha=@n3) {constB=0}, B=@n4.CubeRadius))
  n6 = Multiply(
   A=LinearInterpolate(Alpha=NamedRerouteDeclaration(None=@n4.Reveal) {name=RevealParam, nodeColor=<rgba 0.1484 0 1 1>})
   B=Multiply(A=@n4.CubeRadius, B=ScalarParameter() {defaultValue=4, parameterName=RevealRadiusMult}) {constB=4}
  )
  n7 = Multiply(A=@n6) {constB=3}
  n8 = SphereMask(A=@n1.XYZ, B=@n5, Radius=@n6) {hardnessPercent=0}
  n9 = Saturate(None=Subtract(A=SphereMask(A=@n1.XYZ, B=@n5, Radius=@n7) {hardnessPercent=50}, B=Multiply(A=@n8) {constB=1.5}))
  n10 = OneMinus(None=@n9)
  MaterialFunctionCall(
   Normal=TextureSample(UVs=NamedRerouteUsage()) {samplerType=SAMPLERTYPE_Normal, texture=/Game/Athena/Environments/Island/CubeGrowth/T_CubeNoise01}.RGB
   Flatness=Clamp(None=Power(Base=@n10) {constExponent=7}) {minDefault=0.15}
  ) {materialFunction=&FlattenNormal}
  n11 = VectorParameter() {parameterName=BaseDiffuse, defaultValue=<rgba 0.1684 0.1471 0.25 1>}
  n12 = VectorParameter() {parameterName=SparklesColor, defaultValue=<rgba 8.7858 3 15 1>}
  n13 = Constant() {r=14}
  n14 = TextureCoordinate()
  n15 = Constant() {r=-2}
  n16 = TextureSample(UVs=BumpOffset(Coordinate=Multiply(A=@n13, B=@n14), Height=@n15, HeightRatioInput=Constant() {r=0.02})) {samplerType=SAMPLERTYPE_Masks, texture=/Game/Global/Textures/Default/Noise/T_CubeNoise}
  n17 = Constant() {r=28}
  n18 = TextureSample(UVs=BumpOffset(Coordinate=Multiply(A=@n14, B=@n17), Height=@n15, HeightRatioInput=Constant() {r=0.2})) {samplerType=SAMPLERTYPE_Masks, texture=&T_CubeNoise}
  n19 = Multiply(A=Power(Base=Multiply(A=@n16.B, B=@n18.B), Exp=Constant() {r=7}), B=Constant() {r=60})
  n20 = Multiply(A=Constant() {r=3.5}, B=TextureCoordinate())
  n21 = BumpOffset(Coordinate=@n20, Height=Constant() {r=0.58}, HeightRatioInput=Constant() {r=-0.5})
  n22 = ScalarParameter() {defaultValue=5, parameterName=mipbias1}
  n23 = Add(
   A=Power(
    Base=TextureSample!(UVs=@n21, Tex, Bias=@n22, Apply View MipBias) {mipValueMode=TMVM_MipBias, samplerType=SAMPLERTYPE_Masks, texture=&T_CubeNoise}.R
    Exp=Constant() {r=3}
   )
   B=Constant() {r=-0.05}
  )
  n24 = Add(A=@n11.RGB, B=Multiply(A=Multiply(A=@n12.RGB, B=@n19), B=Saturate(None=@n23)))
  n25 = TextureCoordinate() {coordinateIndex=2}
  n26 = Constant2Vector() {g=0.5, r=0.5}
  n27 = MaterialFunctionCall(Float2=Abs(None=Subtract(A=@n25, B=@n26))) {materialFunction=&BreakOutFloat2Components}
  n28 = Saturate(None=Multiply(A=Max(A=@n27.R, B=@n27.R), B=Constant() {r=1.95}))
  n29 = Constant() {r=5}
  n30 = LinearInterpolate(
   A=QualitySwitch(Default=@n24, Low=@n11.RGB, High=@n24)
   B=VectorParameter() {parameterName=EdgeDiffuse, defaultValue=<rgba 0.6745 0.349 1 1>}.RGB
   Alpha=MaterialFunctionCall(
    Transition Phase=Power(Base=@n28, Exp=@n29)
    Height Texture=TextureSample!(UVs=@n20, Tex, Apply View MipBias) {samplerType=SAMPLERTYPE_Masks, texture=&T_CubeNoise}.R
   ) {materialFunction=&HeightLerp}.Lerp Alpha No Contrast
  )
  n31 = VectorParameter() {parameterName=EdgeGlowColor, defaultValue=<rgba 4.25 0.5 5 0>}
  n32 = Multiply(A=@n30, B=@n31.RGB)
  n33 = Power(Base=NamedRerouteUsage()) {constExponent=7}
  n34 = Multiply(A=@n31.RGB) {constB=10}
  n35 = TextureSample(UVs=NamedRerouteDeclaration(None=TextureCoordinate() {uTiling=4, vTiling=4}) {name=DetailUVs, nodeColor=<rgba 0 1 0.9297 1>}) {samplerType=SAMPLERTYPE_LinearColor, texture=/Game/Athena/Environments/Island/CubeGrowth/T_CubeNoise01_H}
  n36 = NamedRerouteUsage()
  n37 = Panner(Coordinate=TextureCoordinate() {uTiling=3, vTiling=3}, Time=@n36) {speedX=0.02, speedY=0.05}
  n38 = Panner(Time=@n36) {speedX=-0.015, speedY=0.05}
  n39 = Add(
   A=OneMinus(None=@n35.G)
   B=Add(
    A=TextureSample(UVs=@n37) {texture=/Game/Global/Textures/Default/PointSplatter/T_Tiling_Dust}.RGB
    B=TextureSample(UVs=@n38) {texture=&T_Tiling_Dust}.RGB
   )
  )
  n40 = VectorParameter() {parameterName=Tile, defaultValue=<rgba 300 300 1500 0>}
  n41 = NamedRerouteUsage()
  n42 = Multiply(A=@n40.RGB, B=LinearInterpolate(A=Constant3Vector() {constant=<rgba 1 1 10 0>}, Alpha=Power(Base=@n41) {constExponent=0.25}))
  n43 = Multiply(A=@n40.R) {constB=0.5}
  n44 = Multiply(A=Multiply(A=@n42, B=VectorParameter() {parameterName=PanSpeed, defaultValue=<rgba 0.02 -0.02 -2 0>}.RGB), B=@n41)
  n45 = WorldPosition()
  n46 = MaterialFunctionCall(
   TextureObject=TextureObject() {texture=/Game/Effects/Fort_Effects/Textures/Tile/T_ShieldCaustics_Tile_Chromatic}
   TextureSize=@n42
   WorldPosition=Add(A=Multiply(A=@n43, B=@n35.G), B=Add(A=@n44, B=@n45.XYZ))
  ) {materialFunction=&WorldAlignedTexture}
  n47 = Multiply(A=@n46.XYZ Texture) {constB=1.5}
  n48 = Power(Base=@n9)
  n49 = Multiply(A=Add(A=Add(A=@n39, B=@n47), B=@n48) {constB=0.25}, B=@n48)
  n50 = LinearInterpolate(A=Add(A=Multiply(A=@n32, B=@n33), B=Multiply(A=@n34, B=@n49)), B=@n31.RGB, Alpha=NamedRerouteUsage())
  n51 = Clamp(None=@n50)
  MaterialFunctionCall(Emissive=QualitySwitch(
    Default=MaterialFunctionCall(LDR Input=@n51, HDR Tint=@n50) {materialFunction=&DeriveHDRfromLDR}.RemappedRange
    Low=@n50
   )) {materialFunction=&MF_FX_EmissiveAdjustments}
  n52 = Multiply(
   A=LinearInterpolate(
    A=ScalarParameter() {defaultValue=-5, parameterName=Min, sliderMin=1}
    B=ScalarParameter() {defaultValue=0.5, parameterName=Max}
    Alpha=ComponentMask(None=@n46.XYZ Texture) {g=true}
   ) {constA=-1, constB=0}
   B=@n48
  )
  n53 = NamedRerouteUsage()
  n54 = Add(A=@n8, B=LinearInterpolate(Alpha=@n53) {constA=-0.5, constB=1.5})
  MaterialFunctionCall(Alpha Threshold=Saturate(None=Add(A=@n52, B=@n54))) {materialFunction=&DitherTemporalAA}
  Multiply(A=@n30) {constB=0.1}
  NamedRerouteDeclaration(None=@n4.FinalGlow) {name=FinalGlowParam, nodeColor=<rgba 0 1 0.625 1>}
  n55 = VectorParameter() {parameterName=PreviousFrameOffset, defaultValue=<rgba 1 0 0 5>}
  n56 = Transform(None=@n55.RGB) {transformSourceType=TRANSFORMSOURCE_Local}
  PreviousFrameSwitch(Current Frame=Constant(), Previous Frame=Multiply(A=OneMinus(None=@n53), B=Multiply(A=@n56, B=@n55.A)))
 ]
}
