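# Niagara effect for the Corruption cube's awakening beam, running from the owner to User.EndPoint, oriented by EndNormal and EndUp.
# Beam and lightning meshes scale to the distance, while sparkles drift along the line.
# Glows, lights and GPU sparks use spawn groups to give each end its own color.
# System loop age times the flashes; distance-culled runes contract on a snapped grid at the start, then expand at the endpoint after a delay.
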NiagaraSystem /CorruptionGameplay/Gameplay/Environmental/Cube/FX/NS_Corruption_Awakening_Beam
dependencies {
 usedDataInterfaces [
  &NiagaraDataInterfaceCurve
  &NiagaraDataInterfaceArrayFloat
  &NiagaraDataInterfaceVectorCurve
  &NiagaraDataInterfaceColorCurve
 ]
 usedDynamicInputs [&NormalizeVector, &Waveform]
 usedModules [
  &SystemState
  &self:SystemSpawnScript.SystemScriptSource.SystemScriptGraph.NiagaraNodeAssignment_0.SetVariables_#1
  &self:SystemSpawnScript.SystemScriptSource.SystemScriptGraph.NiagaraNodeAssignment_1.SetVariables_#2
 ]
}
emitters {
 BeamMeshes {
  emitterUpdateScript [
   &EmitterState()
   &SpawnBurst_Instantaneous(Use Spawn Probability=false, UseLoopCountLimit=false)
   &SpawnBurst_Instantaneous(Spawn Group=1, Spawn Time=1, Use Spawn Probability=false, UseLoopCountLimit=false) {module=SpawnBurst_Instantaneous001}
  ]
  particleSpawnScript [
   &InitializeParticle(
    Color Maximum=<rgba 1.5 1.1886 1.0983 1>
    Color Minimum=<rgba 1.2 1.2 1.2 1>
    Color Mode=::NewEnumerator2
    Lifetime=4
    Mesh Scale Mode=::NewEnumerator3
    UsePositionOffset=false
    Write Material Random=true
   )
   &CylinderLocation(Cylinder Height=5, MaskSpawnGroup=false, Override Seed=false, Rotation Mode:ENiagara_RotationMode=::NewEnumerator1, UseOwnerScale=true)
   &InitialMeshOrientation(
    Mesh Orientation Mode=::NewEnumerator2
    Orientation Vector=&Subtract_Vector(
     A=Linked() {linkedVariable {name=User.EndPoint, type=NiagaraPosition}}
     B=Linked() {linkedVariable {name=Particles.Position, type=NiagaraPosition}}
    )
    Rotation=&RandomRangeVector(Minimum=<0 0 0>, Maximum=<1 0 0>)
    Use Rotation Vector=true
   )
   /SetVariables(Particles.Scale=&MakeVector(
     X=&Divide_Float(
      A=&VectorLength(Vector=&Subtract_Vector(
        A=Linked() {linkedVariable {name=User.EndPoint, type=NiagaraPosition}}
        B=Linked() {linkedVariable {name=Particles.Position, type=NiagaraPosition}}
       ))
      B=1000
     )
     Y=&RandomRangeFloat(Minimum=4, Maximum=8)
     Z=&RandomRangeFloat(Minimum=4, Maximum=8)
    )) {isSetParameters=true}
  ]
  particleUpdateScript [
   &ParticleState()
   &DynamicMaterialParameters(
    Index 0 Param 1=&RandomRangeFloat(Minimum=0, Maximum=1)
    Index 0 Param 2=&FloatFromCurve(
     FloatCurve=DataInterface() {propertyValues {Curve {keys [
         {interpMode=RCIM_Cubic, tangentMode=RCTM_User, time=0, value=0}
         {interpMode=RCIM_Cubic, tangentMode=RCTM_User, time=0.25, value=1}
         {arriveTangent=-0.9245, interpMode=RCIM_Cubic, leaveTangent=-0.9245, tangentMode=RCTM_User, time=0.7, value=0.8}
         {tangentMode=RCTM_User, time=1, value=0}
        ]}}}
     Scale Curve=&SelectFloatFromArray(Direct Array Index=&SpawnGroup(), Float Selection Array=DataInterface() {propertyValues {FloatData [0.2,1]}})
    )
    Index 0 Param 3=&FloatFromCurve(
     FloatCurve=DataInterface() {propertyValues {Curve {keys [{interpMode=RCIM_Cubic, tangentMode=RCTM_User, time=0.05, value=0}, <at 0.125 1 RCIM_Cubic>, {tangentMode=RCTM_User, time=0.4, value=0}]}}}
     CurveIndex=$System.NormalizedLoopAge
    )
    Index 0 Param 4=&RemapRange(
     Input Value To Remap=&MakeFloatFromVector(Vector=$Particles.Scale, Channel=::NewEnumerator0)
     Input Min=1
     Input Max=3
     Output Min=0.7
     Output Max=1.5
    )
    Module.Param0WriteEnabled=true
    Module.Param1WriteEnabled=true
    Module.Param2WriteEnabled=true
    Module.Param3WriteEnabled=true
   )
   &UpdateMeshOrientation(
    Rotation Coordinate Space=::NewEnumerator3
    Rotation Rate=&RandomRangeFloat(Minimum=0.02, Maximum=0.07)
    Rotation Vector=<1 0 0>
   )
   &ScaleColor(
    Scale Alpha=&FloatFromCurve(FloatCurve=DataInterface() {propertyValues {Curve {keys [
         {interpMode=RCIM_Cubic, tangentMode=RCTM_User, time=0, value=1}
         {interpMode=RCIM_Cubic, tangentMode=RCTM_User, time=0.5, value=0.5}
         {tangentMode=RCTM_User, time=1, value=1}
        ]}}})
    ScaleA=true
    ScaleRGB=true
   )
  ]
  rendererClasses [&NiagaraSpriteRendererProperties, &NiagaraMeshRendererProperties]
  renderers [
   {rendererClass=&NiagaraSpriteRendererProperties}
   {rendererClass=&NiagaraMeshRendererProperties, rendererIndex=1}
  ]
 }
 GlowSurface {
  emitterUpdateScript [
   &EmitterState()
   &SpawnRate(SpawnRate=3, Use Spawn Probability=false)
  ]
  particleSpawnScript [
   &FN_SetSpawnGroup(Number of Spawn Groups=2)
   &InitializeParticle(
    Color=&MaskLinearColorBySpawnGroup(Masked Linear Color=<rgba 10 6.4963 0.4985 1>, Passthrough Linear Color=<rgba 3.8712 0.6667 4 1>)
    Lifetime Max=2.8
    Lifetime Min=2.2
    Lifetime Mode=::NewEnumerator1
    Position=&MaskVectorBySpawnGroup(
     Masked Vector=Linked() {linkedVariable {name=Engine.Owner.Position, type=NiagaraPosition}}
     Passthrough Vector=Linked() {linkedVariable {name=User.EndPoint, type=NiagaraPosition}}
    )
    Position Mode=::NewEnumerator1
    Sprite Rotation Mode=::NewEnumerator2
    Sprite Size Mode=::NewEnumerator3
    Uniform Sprite Size=&RandomRangeFloat(Minimum=1100, Maximum=1400)
    UsePositionOffset=false
    Write Material Random=true
   )
   &SpriteFacingAndAlignment(
    Sprite Facing=&MaskVectorBySpawnGroup(Masked Vector=$Engine.Owner.SystemXAxis, Passthrough Vector=$User.EndNormal)
    WriteAlignment=true
    WriteFacing=true
   )
   &CylinderLocation(
    Cylinder Axis:Vector3f=&MaskVectorBySpawnGroup(Masked Vector=$Engine.Owner.SystemXAxis, Passthrough Vector=$User.EndNormal)
    Cylinder Height=10
    Cylinder Radius=50
    MaskSpawnGroup=false
    Override Seed=false
    Rotation Coordinate Space:ENiagaraCoordinateSpace=::Simulation
    Rotation Mode:ENiagara_RotationMode=::NewEnumerator1
    UseOwnerScale=true
   )
   &DynamicMaterialParameters(
    Index 0 Param 1=&RandomRangeFloat(Minimum=0, Maximum=1)
    Index 0 Param 2=&RandomRangeFloat(Minimum=0, Maximum=1)
    Module.Param0WriteEnabled=true
    Module.Param1WriteEnabled=true
    Module.Param2WriteEnabled=true
    Module.Param3WriteEnabled=false
   )
   &CameraOffset(Camera Offset Amount=5)
  ]
  particleUpdateScript [
   &ParticleState()
   &SpriteRotationRate(Rotation Rate=&Multiply_Float(
     A=&RandomRangeFloat(Minimum=4, Maximum=15)
     B=&MaskFloatBySpawnGroup(Masked Float=1, Passthrough Float=-1)
    ))
   &ScaleColor(
    Scale Alpha=&FloatFromCurve(FloatCurve=DataInterface() {propertyValues {Curve {keys [
         {interpMode=RCIM_Cubic, tangentMode=RCTM_User, time=0, value=0}
         <at 0.3 1>
         {interpMode=RCIM_Cubic, tangentMode=RCTM_User, time=0.7, value=1}
         {tangentMode=RCTM_User, time=1, value=0}
        ]}}})
    Scale RGB=&VectorFromFloat(Value=&FloatFromCurve(
      FloatCurve=DataInterface() {propertyValues {Curve {keys [
          {interpMode=RCIM_Cubic, tangentMode=RCTM_User, time=0.05, value=1}
          {arriveTangent=-0.8571, interpMode=RCIM_Cubic, leaveTangent=-0.8571, time=0.17, value=2}
          {arriveTangentWeight=0.1637, interpMode=RCIM_Cubic, leaveTangentWeight=0.2135, tangentMode=RCTM_User, time=0.4, value=0.7}
          <at 0.7 1>
         ]}}}
      CurveIndex=&MaskFloatBySpawnGroup(Masked Float=$System.NormalizedLoopAge, Passthrough Float=&Add_Float(A=$System.NormalizedLoopAge, B=-0.2))
     ))
    ScaleA=true
    ScaleRGB=true
   )
   &ScaleSpriteSize(
    Scale Sprite Size Mode=::NewEnumerator0
    Uniform Scale Factor=&MaskFloatBySpawnGroup(
     Masked Float=&FloatFromCurve(FloatCurve=DataInterface() {propertyValues {Curve {keys [
          {interpMode=RCIM_Cubic, tangentMode=RCTM_User, time=0, value=1}
          {arriveTangent=-1.1258, interpMode=RCIM_Cubic, leaveTangent=-1.1258, tangentMode=RCTM_User, time=1, value=0.25}
         ]}}})
     Passthrough Float=&FloatFromCurve(FloatCurve=DataInterface() {propertyValues {Curve {keys [
          {arriveTangent=0.4673, time=0, value=0.25}
          {arriveTangent=0.8098, interpMode=RCIM_Cubic, leaveTangent=0.8098, tangentMode=RCTM_User, time=1, value=1}
         ]}}})
    )
   )
  ]
  rendererClasses [&NiagaraSpriteRendererProperties]
  renderers [{rendererClass=&NiagaraSpriteRendererProperties}]
 }
 Lightning {
  emitterUpdateScript [
   &EmitterState()
   &SpawnBurst_Instantaneous(Spawn Count=1, Spawn Time=0.05, Use Spawn Probability=false)
   &SpawnBurst_Instantaneous(Spawn Count=1, Spawn Time=0.12, Use Spawn Probability=false) {module=SpawnBurst_Instantaneous001}
   &SpawnBurst_Instantaneous(Spawn Count=1, Spawn Time=0.4, Use Spawn Probability=false) {module=SpawnBurst_Instantaneous002}
   &SpawnRate(SpawnRate=&Multiply_Float(A=$System.SpawnOn, B=25), Use Spawn Probability=false)
   &SpawnRate(SpawnGroup=1, SpawnRate=&Multiply_Float(A=$System.SpawnOn, B=20), Use Spawn Probability=false) {module=SpawnRate001}
  ]
  particleSpawnScript [
   /Niagara/Modules/Spawn/Initialization/InitializeParticle(
    Color=<rgba 0.4383 0.5 0.4275 1>
    Lifetime=&SelectFloatFromArray(
     Direct Array Index=&SpawnGroup()
     Interpolation=&MaskFloatBySpawnGroup(Masked Float=0, Passthrough Float=1)
     Float Selection Array=DataInterface() {propertyValues {FloatData [0.12,0.07]}}
    )
    Mesh Scale=&VectorFromFloat(Value=2)
    Module.Write Color=true
    Module.Write Lifetime=true
    Module.Write Mass=true
    Module.Write Material Random=true
    Module.Write Position=false
    Module.Write Scale=true
    Module.Write SpriteRotation=false
    Module.Write SpriteSize=true
    Position=&SimulationPosition()
    Sprite Size=&Vector2DFromFloat(Value=&RandomRangeFloat(Minimum=400, Maximum=800))
   )
   &CylinderLocation(
    Apply To Particle Position:ENiagara_AttributeSamplingApplyOutput=::NewEnumerator1
    Cylinder Height=2
    Cylinder Origin:NiagaraPosition=$Engine.Owner.Position
    Cylinder Radius=50
    MaskSpawnGroup=false
    Override Seed=false
    Rotation Mode:ENiagara_RotationMode=::NewEnumerator1
    UseOwnerScale=true
   ) {module=CylinderLocation001}
   /SetVariables(Particles.StartPosition:Vector3f=$Particles.CylinderLocation001.CylinderPosition) {isSetParameters=true}
   &CylinderLocation(
    Cylinder Axis:Vector3f=$User.EndNormal
    Cylinder Height=2
    Cylinder Origin:NiagaraPosition=$User.EndPoint
    Cylinder Radius=50
    MaskSpawnGroup=false
    Override Seed=false
    Rotation Mode:ENiagara_RotationMode=::NewEnumerator1
    UseOwnerScale=true
   )
   /SetVariables(
    Particles.LightEnabled:NiagaraBool=&SetBoolByIntComparison(A=&SpawnGroup(), B=0, Comparison Type=::NewEnumerator0)
    Particles.LightPosition:Vector3f=&Lerp_Vector(
     A=Linked() {linkedVariable {name=Engine.Owner.Position, type=NiagaraPosition}}
     B=Linked() {linkedVariable {name=User.EndPoint, type=NiagaraPosition}}
     Alpha=&FloatFromCurve(
      FloatCurve=DataInterface() {propertyValues {Curve {keys [
          {interpMode=RCIM_Cubic, tangentMode=RCTM_User, time=0, value=0.08}
          {arriveTangent=1.5668, interpMode=RCIM_Cubic, leaveTangent=1.5668, tangentMode=RCTM_User, time=0.5, value=0.38}
          {arriveTangent=1.2055, interpMode=RCIM_Cubic, leaveTangent=1.2055, tangentMode=RCTM_User, time=0.505, value=0.62}
          {interpMode=RCIM_Cubic, tangentMode=RCTM_User, time=1, value=0.92}
         ]}}}
      CurveIndex=&RandomRangeFloat(Minimum=0, Maximum=1)
     )
    )
   ) {isSetParameters=true}
   &InitialMeshOrientation(
    Mesh Orientation Mode=::NewEnumerator2
    Orientation Vector=&Subtract_Vector(A=$Particles.StartPosition, B=Linked() {linkedVariable {name=Particles.Position, type=NiagaraPosition}})
    Rotation=&RandomRangeVector(Minimum=<0 0 0>, Maximum=<1 0 0>)
    Use Rotation Vector=true
   )
   &DynamicMaterialParameters(
    Index 0 Param 2=&UniformRangedFloat(Minimum=1, Maximum=1)
    Index 0 Param 3=1
    Index 0 Param 4=1
    Index 1 Param 1=&FloatFromCurve(
     FloatCurve=DataInterface() {propertyValues {Curve {keys [{tangentMode=RCTM_User, time=0, value=0}, {tangentMode=RCTM_User, time=1, value=1.4}]}}}
     CurveIndex=$System.AnimTime
    )
    Index 1 Param 2=&FloatFromCurve(
     FloatCurve=DataInterface() {propertyValues {Curve {keys [{interpMode=RCIM_Cubic, tangentMode=RCTM_User, time=0, value=0}, {interpMode=RCIM_Cubic, tangentMode=RCTM_User, time=1, value=1}]}}}
     CurveIndex=$System.AnimTime
     Scale Curve=-600
    )
    Index 1 Param 3=1.5
    Module.Param0WriteEnabled=false
    Module.Param0WriteEnabled1=true
    Module.Param1WriteEnabled=true
    Module.Param1WriteEnabled1=true
    Module.Param2WriteEnabled=false
    Module.Param2WriteEnabled1=true
    Module.Param3WriteEnabled=false
    Module.Param3WriteEnabled1=false
    Write Parameter Index 1=true
   ) {enabled=false}
   /SetVariables(Particles.MeshIndex=&RandomRangeInt(Minimum=0, Maximum=2)) {isSetParameters=true}
   /SetVariables(Particles.Scale=&MakeVector(
     X=&Divide_Float(
      A=&VectorLength(Vector=&Subtract_Vector(A=$Particles.StartPosition, B=Linked() {linkedVariable {name=Particles.Position, type=NiagaraPosition}}))
      B=1000
     )
     Y=&RandomRangeFloat(Minimum=0.6, Maximum=1.5)
     Z=&RandomRangeFloat(Minimum=0.6, Maximum=1.5)
    )) {isSetParameters=true}
   /SetVariables(Particles.DynParamsSprite:Vector4f=<0.4 2 1 1>) {isSetParameters=true}
   /SetVariables(Particles.CamOffsetSprite=80) {isSetParameters=true}
   /SetVariables(Particles.LightRadius=&RandomRangeFloat(Minimum=90, Maximum=120)) {isSetParameters=true}
  ]
  particleUpdateScript [
   &ParticleState()
   &ScaleColor(
    Scale Alpha=&FloatFromCurve(FloatCurve=DataInterface() {propertyValues {Curve {keys [
         <at 0 0>
         {interpMode=RCIM_Cubic, tangentMode=RCTM_User, time=0.1258, value=0.1952}
         <at 0.341 0.0902>
         {interpMode=RCIM_Cubic, tangentMode=RCTM_User, time=0.496, value=0.2532}
         <at 1 0>
        ]}}})
    ScaleA=true
    ScaleRGB=false
   )
   /SetVariables(Particles.SpriteColor=&MultiplyLinearColor(
     Linear Color=$Particles.Color
     Scale Factor=&MakeVector4_FromLinearColor(LinearColor=&MultiplyLinearColor(
       Linear Color=<rgba 47.3997 4.5799 50 1>
       Scale Factor=&MakeVector4(
        X=1
        Y=1
        Z=1
        W=&FloatFromCurve(FloatCurve=DataInterface() {propertyValues {Curve {keys [
             {interpMode=RCIM_Cubic, tangentMode=RCTM_User, time=0, value=1}
             {arriveTangentWeight=0.031, interpMode=RCIM_Cubic, leaveTangentWeight=0.0366, tangentMode=RCTM_User, time=0.0918, value=0.05}
             {
              arriveTangentWeight=0.16
              interpMode=RCIM_Cubic
              leaveTangent=18.7153
              leaveTangentWeight=0.1479
              tangentMode=RCTM_Break
              time=0.1619
              value=0
             }
             {arriveTangent=-0.2815, interpMode=RCIM_Cubic, leaveTangent=-0.2815, tangentMode=RCTM_User, time=0.3, value=0.95}
             {arriveTangentWeight=0.0345, interpMode=RCIM_Cubic, leaveTangentWeight=0.0566, tangentMode=RCTM_User, time=0.4, value=0.1}
             {
              arriveTangentWeight=0.1589
              interpMode=RCIM_Cubic
              leaveTangent=29.6956
              leaveTangentWeight=0.1018
              tangentMode=RCTM_Break
              time=0.4765
              value=0.05
             }
             {arriveTangent=0.512, interpMode=RCIM_Cubic, leaveTangent=0.512, tangentMode=RCTM_User, time=0.5641, value=0.8648}
             <at 1 0>
            ]}}})
       )
      ))
    )) {isSetParameters=true}
   /SetVariables(Particles.SpriteColorStart:LinearColor=&MultiplyLinearColor(
     Linear Color=$Particles.Color
     Scale Factor=&MakeVector4_FromLinearColor(LinearColor=&MultiplyLinearColor(
       Linear Color=<rgba 50 14.4426 1.8518 1>
       Scale Factor=&MakeVector4(
        X=1
        Y=1
        Z=1
        W=&FloatFromCurve(FloatCurve=DataInterface() {propertyValues {Curve {keys [
             {interpMode=RCIM_Cubic, tangentMode=RCTM_User, time=0, value=1}
             {arriveTangentWeight=0.031, interpMode=RCIM_Cubic, leaveTangentWeight=0.0366, tangentMode=RCTM_User, time=0.0918, value=0.05}
             {
              arriveTangentWeight=0.16
              interpMode=RCIM_Cubic
              leaveTangent=18.7153
              leaveTangentWeight=0.1479
              tangentMode=RCTM_Break
              time=0.1619
              value=0
             }
             {arriveTangent=-0.2815, interpMode=RCIM_Cubic, leaveTangent=-0.2815, tangentMode=RCTM_User, time=0.3, value=0.95}
             {arriveTangentWeight=0.0345, interpMode=RCIM_Cubic, leaveTangentWeight=0.0566, tangentMode=RCTM_User, time=0.4, value=0.1}
             {
              arriveTangentWeight=0.1589
              interpMode=RCIM_Cubic
              leaveTangent=29.6956
              leaveTangentWeight=0.1018
              tangentMode=RCTM_Break
              time=0.4765
              value=0.05
             }
             {arriveTangent=0.512, interpMode=RCIM_Cubic, leaveTangent=0.512, tangentMode=RCTM_User, time=0.5641, value=0.8648}
             <at 1 0>
            ]}}})
       )
      ))
    )) {isSetParameters=true}
   &DynamicMaterialParameters(
    Index 0 Param 1=&FloatFromCurve(FloatCurve=DataInterface() {propertyValues {Curve {keys [{interpMode=RCIM_Cubic, tangentMode=RCTM_User, time=0, value=1}, {interpMode=RCIM_Cubic, tangentMode=RCTM_User, time=1, value=0}]}}})
    Index 0 Param 2=&FloatFromCurve(
     FloatCurve=DataInterface() {propertyValues {Curve {keys [
         {interpMode=RCIM_Cubic, leaveTangent=8.981, tangentMode=RCTM_Break, time=0, value=0}
         {arriveTangent=0.0136, interpMode=RCIM_Cubic, leaveTangent=-5.3647, tangentMode=RCTM_Break, time=0.4833, value=1.5}
         <at 1 -0.1>
        ]}}}
     Scale Curve=&RandomRangeFloat(Minimum=1, Maximum=2)
    )
    Index 0 Param 3=&RandomRangeFloat(Minimum=1, Maximum=3)
    Index 1 Param 1=&FloatFromCurve(
     FloatCurve=DataInterface() {propertyValues {Curve {keys [
         {arriveTangent=1.9721, interpMode=RCIM_Cubic, leaveTangent=1.9721, tangentMode=RCTM_User, time=0, value=-0.25}
         {
          arriveTangent=0.6732
          arriveTangentWeight=0.4269
          interpMode=RCIM_Cubic
          leaveTangent=0.6732
          leaveTangentWeight=0.1067
          tangentMode=RCTM_User
          time=1
          value=1
         }
        ]}}}
     Scale Curve=-0.2
    )
    Index 1 Param 2=&FloatFromCurve(
     FloatCurve=DataInterface() {propertyValues {Curve {keys [
         {interpMode=RCIM_Cubic, tangentMode=RCTM_User, time=0.05, value=0.05}
         {arriveTangent=-9.0587, interpMode=RCIM_Cubic, leaveTangent=-9.0587, tangentMode=RCTM_User, time=0.125, value=1}
         {tangentMode=RCTM_User, time=0.4, value=0.05}
        ]}}}
     CurveIndex=$System.NormalizedLoopAge
     Scale Curve=&RandomRangeFloat(Minimum=100, Maximum=300)
    )
    Index 1 Param 3=&RandomRangeFloat(Minimum=0.25, Maximum=0.5)
    Module.Param0WriteEnabled=true
    Module.Param0WriteEnabled1=true
    Module.Param1WriteEnabled=true
    Module.Param1WriteEnabled1=true
    Module.Param2WriteEnabled=true
    Module.Param2WriteEnabled1=true
    Module.Param3WriteEnabled=false
    Module.Param3WriteEnabled1=false
    Write Parameter Index 1=true
   ) {module=DynamicMaterialParameters001}
   &ScaleSpriteSize(Uniform Curve Sprite Scale=DataInterface() {propertyValues {Curve {keys [
        {arriveTangent=21.5526, interpMode=RCIM_Cubic, leaveTangent=21.5526, tangentMode=RCTM_User, time=0, value=0}
        {arriveTangent=0.5, interpMode=RCIM_Cubic, leaveTangent=0.5, time=0.1, value=1}
        {arriveTangent=-1.8758, interpMode=RCIM_Cubic, leaveTangent=-1.8758, tangentMode=RCTM_User, time=1, value=0}
       ]}}})
  ]
  rendererClasses [
   &NiagaraMeshRendererProperties
   &NiagaraSpriteRendererProperties
   &NiagaraLightRendererProperties
  ]
  renderers [
   {rendererClass=&NiagaraMeshRendererProperties, rendererIndex=0}
   {rendererClass=&NiagaraSpriteRendererProperties, rendererIndex=1}
   {rendererClass=&NiagaraLightRendererProperties, rendererIndex=2}
   {rendererClass=&NiagaraSpriteRendererProperties, rendererIndex=3}
  ]
 }
 Lights {
  emitterUpdateScript [
   &EmitterState()
   &SpawnBurst_Instantaneous(Spawn Count=2, Use Spawn Probability=false, UseLoopCountLimit=false)
  ]
  particleSpawnScript [
   &FN_SetSpawnGroup(Number of Spawn Groups=2)
   /Niagara/Modules/Spawn/Initialization/V2/InitializeParticle(
    Color=&MaskLinearColorBySpawnGroup(Masked Linear Color=<rgba 15 4.5614 1.0902 1>, Passthrough Linear Color=<rgba 9.6779 1.6667 10 1>)
    Lifetime=6
    Position=&MaskVectorBySpawnGroup(
     Masked Vector=&Add_Vector(
      A=Linked() {linkedVariable {name=Engine.Owner.Position, type=NiagaraPosition}}
      B=&Multiply_VectorByFloat(Vector=$Engine.Owner.SystemXAxis, Float=60)
     )
     Passthrough Vector=&Add_Vector(
      A=Linked() {linkedVariable {name=User.EndPoint, type=NiagaraPosition}}
      B=&Multiply_VectorByFloat(Vector=$User.EndNormal, Float=50)
     )
    )
    Position Mode=::NewEnumerator1
    Sprite Rotation Mode=::NewEnumerator2
    Sprite Size Mode=::NewEnumerator3
    Uniform Sprite Size=100
    UsePositionOffset=false
    Write Material Random=true
   )
   &DynamicMaterialParameters(
    Index 0 Param 1=&RandomRangeFloat(Minimum=0, Maximum=1)
    Index 0 Param 2=&RandomRangeFloat(Minimum=0, Maximum=1)
    Module.Param0WriteEnabled=true
    Module.Param1WriteEnabled=true
    Module.Param2WriteEnabled=true
    Module.Param3WriteEnabled=false
   )
  ]
  particleUpdateScript [
   &ParticleState()
   &SpriteRotationRate(Rotation Rate=&Multiply_Float(
     A=&RandomRangeFloat(Minimum=4, Maximum=15)
     B=&MaskFloatBySpawnGroup(Masked Float=1, Passthrough Float=-1)
    ))
   &ScaleColor(
    Scale Alpha=&FloatFromCurve(FloatCurve=DataInterface() {propertyValues {Curve {keys [
         {interpMode=RCIM_Cubic, tangentMode=RCTM_User, time=0.05, value=0}
         {arriveTangentWeight=0.093, interpMode=RCIM_Cubic, leaveTangentWeight=0.4435, tangentMode=RCTM_User, time=0.18, value=1}
         {arriveTangent=-1.2195, interpMode=RCIM_Cubic, leaveTangent=-1.2195, time=0.8, value=0.35}
         <at 1 0>
        ]}}})
    ScaleA=true
    ScaleRGB=false
   )
  ]
  rendererClasses [&NiagaraLightRendererProperties]
  renderers [{rendererClass=&NiagaraLightRendererProperties, rendererIndex=0}]
 }
 NE_FN_Sparks {
  simTarget=GPUComputeSim
  emitterUpdateScript [
   &EmitterState(Enable Distance Culling=true, MaxDistance=8000, Module.bUseMaxDistance=true, Module.bUseMinDistance=false, Scalability Mode=::NewEnumerator1)
   &SpawnRate(SpawnGroup=1, SpawnRate=&Multiply_Float(A=$System.SpawnOn, B=50), Use Spawn Probability=false)
   &SpawnRate(SpawnRate=&Multiply_Float(A=$System.SpawnOn, B=50), Use Spawn Probability=false) {module=SpawnRate001}
  ]
  particleSpawnScript [
   /SetVariables(Particles.SparkIntensity:NiagaraFloat=&FloatFromCurve(
     FloatCurve=DataInterface() {propertyValues {Curve {keys [
         <at 0 1 RCIM_Cubic>
         {arriveTangent=-2.8017, interpMode=RCIM_Cubic, leaveTangent=-2.8017, tangentMode=RCTM_User, time=0.6, value=0.5}
         <at 1 0 RCIM_Cubic>
        ]}}}
     CurveIndex=&RandomRangeFloat(Minimum=0, Maximum=1)
    )) {isSetParameters=true}
   &InitializeParticle(
    Color=&MultiplyLinearColor(
     Linear Color=&MaskLinearColorBySpawnGroup(Masked Linear Color=<rgba 1 0.2737 0.0425 1>, Passthrough Linear Color=<rgba 0.8921 0.048 1 1>)
     Scale Factor=&MakeFloat4FromVectorAndFloat(
      "Vector (RGB)"=&VectorFromFloat(Value=&Lerp_Float(A=4, B=10, Alpha=$Particles.SparkIntensity))
      "Float (Alpha)"=1
     )
    )
    Lifetime Max=0.7
    Lifetime Min=0.12
    Lifetime Mode=::NewEnumerator1
    Mass=&Lerp_Float(A=0.3, B=1, Alpha=$Particles.SparkIntensity)
    Mass Mode=::NewEnumerator1
    Position=&MaskVectorBySpawnGroup(
     Masked Vector=Linked() {linkedVariable {name=Engine.Owner.Position, type=NiagaraPosition}}
     Passthrough Vector=Linked() {linkedVariable {name=User.EndPoint, type=NiagaraPosition}}
    )
    Position Mode=::NewEnumerator1
    Sprite Size Mode=::NewEnumerator3
    Uniform Sprite Size=&Lerp_Float(A=2, B=4, Alpha=$Particles.SparkIntensity)
    UsePositionOffset=false
    Write Material Random=true
   )
   /SetVariables(Particles.SpriteSize=&Multiply_Vector2(
     A=$Particles.SpriteSize
     B=&Vector2DFromFloat(Value=&FloatFromCurve(FloatCurve=DataInterface() {propertyValues {Curve {keys [<at 1000 1 RCIM_Cubic>,<at 7000 5>]}}}, CurveIndex=$Engine.Owner.LODDistance))
    )) {isSetParameters=true}
   &AddVelocityInCone(
    Cone Angle=70
    Cone Axis=&MaskVectorBySpawnGroup(Masked Vector=$Engine.Owner.SystemXAxis, Passthrough Vector=$User.EndNormal)
    Module.Use Velocity Falloff On Cone Axis=false
    Velocity Distribution Along Cone Axis=2
    Velocity Strength=&Lerp_Float(
     A=&RandomRangeFloat(Minimum=500, Maximum=1200)
     B=&RandomRangeFloat(Minimum=1000, Maximum=3000)
     Alpha=$Particles.SparkIntensity
    )
   )
   &DynamicMaterialParameters(
    Index 0 Param 1=2
    Index 0 Param 3=3
    Index 0 Param 4=1500
    Module.Param0WriteEnabled=true
    Module.Param1WriteEnabled=false
    Module.Param2WriteEnabled=true
    Module.Param3WriteEnabled=false
   )
  ]
  particleUpdateScript [
   &ParticleState()
   &GravityForce(Gravity=<0 0 -2200>)
   &Drag(Use Linear Drag=true, Use Rotational Drag=false)
   &VectorNoiseForce(
    Force Amount=&FloatFromCurve(
     FloatCurve=DataInterface() {propertyValues {Curve {keys [
         {arriveTangent=14.9491, interpMode=RCIM_Cubic, leaveTangent=14.9491, tangentMode=RCTM_User, time=0, value=0}
         {arriveTangent=-3.0044, interpMode=RCIM_Cubic, leaveTangent=-3.0044, tangentMode=RCTM_User, time=0.2, value=1}
         {tangentMode=RCTM_User, time=1, value=0.2}
        ]}}}
     Scale Curve=2500
    )
    Module.Use Random Noise Time Interval=false
   )
   &SolveForcesAndVelocity(Clamp Velocity=false, Limit Acceleration=false)
   &ScaleColor(
    Linear Color Curve=DataInterface() {propertyValues {
      AlphaCurve {keys [<at 0 1>,<at 1 1>]}
      BlueCurve {keys [
        <at 0 1.5 RCIM_Cubic>
        {arriveTangent=-1.452, interpMode=RCIM_Cubic, leaveTangent=-1.452, tangentMode=RCTM_User, time=0.1467, value=1}
        <at 1 0.25>
       ]}
      GreenCurve {keys [
        <at 0 1.5 RCIM_Cubic>
        {arriveTangent=-1.4408, interpMode=RCIM_Cubic, leaveTangent=-1.4408, tangentMode=RCTM_User, time=0.1467, value=1}
        <at 1 0.1698>
       ]}
      RedCurve {keys [
        <at 0 1.5 RCIM_Cubic>
        {arriveTangent=-0.8332, interpMode=RCIM_Cubic, leaveTangent=-0.8332, tangentMode=RCTM_User, time=0.1467, value=1}
        <at 1 0.1219>
       ]}
     }}
    Scale Mode=::NewEnumerator2
   )
   &ScaleSpriteSize(Scale Factor=&Vector2DFromFloat(Value=&FloatFromCurve(FloatCurve=DataInterface() {propertyValues {Curve {keys [<at 0.6 1 RCIM_Cubic>, {arriveTangent=-5.5202, interpMode=RCIM_Cubic, leaveTangent=-5.5202, tangentMode=RCTM_User, time=1, value=0}]}}})))
   &ScaleSpriteSizeBySpeed(
    Max Scale Factor=<1 6>
    Sample Scale Factor By Curve=true
    Scale Factor Curve=DataInterface() {propertyValues {Curve {keys [
        {arriveTangent=0.2757, time=0, value=0}
        {arriveTangent=0.5388, interpMode=RCIM_Cubic, leaveTangent=0.5388, tangentMode=RCTM_User, time=0.6, value=0.3}
        {arriveTangent=2.7622, interpMode=RCIM_Cubic, leaveTangent=2.7622, tangentMode=RCTM_User, time=1, value=1}
       ]}}}
    Velocity Threshold=3000
   )
  ]
  rendererClasses [&NiagaraSpriteRendererProperties]
  renderers [{rendererClass=&NiagaraSpriteRendererProperties}]
 }
 RunesEnd {
  emitterSpawnScript [/SetVariables(Emitter.RuneSize=64) {isSetParameters=true, module=SetVariables_#3}]
  emitterUpdateScript [
   &EmitterState(
    Delay First Loop Only=true
    Enable Distance Culling=true
    Life Cycle Mode=::NewEnumerator1
    Loop Delay=0.25
    Loop Duration=3
    MaxDistance=7000
    Module.bUseMaxDistance=true
    Module.bUseMinDistance=false
    Scalability Mode=::NewEnumerator1
    UseLoopDelay=true
   )
   &SpawnRate(
    SpawnRate=&FloatFromCurve(
     FloatCurve=DataInterface() {propertyValues {Curve {keys [
         <at 0 0 RCIM_Cubic>
         {
          arriveTangent=-5.0015
          arriveTangentWeight=0.0471
          interpMode=RCIM_Cubic
          leaveTangent=-5.0015
          leaveTangentWeight=0.0707
          tangentMode=RCTM_User
          time=0.2
          value=1
         }
         {interpMode=RCIM_Cubic, leaveTangent=-5.8264, time=0.5, value=0}
        ]}}}
     CurveIndex=$Emitter.NormalizedLoopAge
     Scale Curve=100
    )
    Use Spawn Probability=false
   )
  ]
  particleSpawnScript [
   &InitializeParticle(
    Color Maximum=<rgba 1 20 12 0.9>
    Color Minimum=<rgba 0 5 1 0.7>
    Color Mode=::NewEnumerator2
    Lifetime Max=1
    Lifetime Min=0.05
    Lifetime Mode=::NewEnumerator1
    Position=$User.EndPoint
    Position Mode=::NewEnumerator1
    Sprite Rotation Angle=&SelectFloatFromArray(
     Array Sampling Mode=::NewEnumerator0
     Direct Array Index=?
     Float Selection Array=DataInterface() {propertyValues {FloatData [0,0.25,0.5,0.75]}}
     Random Seed=0
    )
    Sprite Rotation Mode=::NewEnumerator1
    Sprite Size Mode=::NewEnumerator3
    Uniform Sprite Size=$Emitter.RuneSize
    UsePositionOffset=false
    Write Material Random=true
   )
   &CylinderLocation(
    Cylinder Axis:Vector3f=$User.EndNormal
    Cylinder Height=4
    Cylinder Radius=&FloatFromCurve(
     FloatCurve=DataInterface() {propertyValues {Curve {keys [
         {arriveTangent=3.2464, interpMode=RCIM_Cubic, time=0, value=0.1}
         {arriveTangent=0.7271, interpMode=RCIM_Cubic, leaveTangent=0.7271, tangentMode=RCTM_User, time=0.5, value=1}
        ]}}}
     CurveIndex=$Emitter.NormalizedLoopAge
     Scale Curve=800
    )
    MaskSpawnGroup=false
    Override Seed=false
    Rotation Coordinate Space:ENiagaraCoordinateSpace=::Simulation
    Rotation Mode:ENiagara_RotationMode=::NewEnumerator1
    Surface Only=true
    Use Endcaps In Surface Only Mode=false
    UseOwnerScale=true
   )
   &CameraOffset(Camera Offset Amount=12)
   &SpriteFacingAndAlignment(Sprite Alignment=$User.EndUp, Sprite Facing=$User.EndNormal, WriteAlignment=true, WriteFacing=true)
  ]
  particleUpdateScript [
   &ParticleState()
   &SubUVAnimation(
    End Frame=4
    Sub UV Play Rate:NiagaraFloat=&RandomRangeFloat(Minimum=0.5, Maximum=2)
    SubUV Animation Mode:ENiagara_SubUVLookupMode=::NewEnumerator3
    SubUV Start Time Offset:NiagaraFloat=&RandomRangeFloat(Minimum=0, Maximum=1)
   )
   &ScaleColor(
    Scale Alpha=&FloatFromCurve(FloatCurve=DataInterface() {propertyValues {Curve {keys [
         {arriveTangent=-3.8362, interpMode=RCIM_Cubic, leaveTangent=-3.8362, tangentMode=RCTM_User, time=0, value=1}
         <at 0.1866 0.4>
         {arriveTangent=-4.1865, interpMode=RCIM_Cubic, leaveTangent=-4.1865, tangentMode=RCTM_User, time=0.2015, value=1}
         <at 0.4552 0.3>
         {arriveTangent=-8.2931, interpMode=RCIM_Cubic, leaveTangent=-8.2931, tangentMode=RCTM_User, time=0.4652, value=0.75}
         <at 0.7089 0>
         {arriveTangent=-5.5334, interpMode=RCIM_Cubic, leaveTangent=-5.5334, tangentMode=RCTM_User, time=0.7338, value=0.5}
         <at 1 0>
        ]}}})
    Scale RGB=&VectorFromCurve(VectorCurve=DataInterface() {propertyValues {
       XCurve {keys [<at 0 1>,<at 1 1>]}
       YCurve {keys [<at 0 1>,<at 1 1>]}
       ZCurve {keys [
         {arriveTangent=1.0471, interpMode=RCIM_Cubic, leaveTangent=1.0471, tangentMode=RCTM_User, time=0, value=0.8}
         {interpMode=RCIM_Cubic, tangentMode=RCTM_User, time=0.25, value=1}
         {arriveTangent=-2.0404, interpMode=RCIM_Cubic, leaveTangent=-2.0404, tangentMode=RCTM_User, time=1, value=0}
        ]}
      }})
    ScaleA=true
    ScaleRGB=true
   )
  ]
  rendererClasses [&NiagaraSpriteRendererProperties]
  renderers [{rendererClass=&NiagaraSpriteRendererProperties}]
 }
 RunesStart {
  emitterSpawnScript [/SetVariables(Emitter.RuneSize=64) {isSetParameters=true, module=SetVariables_#3}]
  emitterUpdateScript [
   &EmitterState(Enable Distance Culling=true, MaxDistance=6000, Module.bUseMaxDistance=true, Module.bUseMinDistance=false, Scalability Mode=::NewEnumerator1)
   &SpawnRate(
    SpawnRate=&FloatFromCurve(
     FloatCurve=DataInterface() {propertyValues {Curve {keys [
         <at 0 0 RCIM_Cubic>
         {arriveTangentWeight=0.0471, interpMode=RCIM_Cubic, leaveTangentWeight=0.0707, tangentMode=RCTM_User, time=0.15, value=1}
         {arriveTangent=-5.8264, interpMode=RCIM_Cubic, leaveTangent=-5.8264, tangentMode=RCTM_User, time=0.3, value=0}
        ]}}}
     CurveIndex=$Emitter.NormalizedLoopAge
     Scale Curve=150
    )
    Use Spawn Probability=false
   )
  ]
  particleSpawnScript [
   &InitializeParticle(
    Color Maximum=<rgba 1 25 12 0.9>
    Color Minimum=<rgba 0 10 0.5 0.6>
    Color Mode=::NewEnumerator2
    Lifetime Max=0.75
    Lifetime Min=0.05
    Lifetime Mode=::NewEnumerator1
    Sprite Rotation Angle=&SelectFloatFromArray(
     Array Sampling Mode=::NewEnumerator0
     Direct Array Index=?
     Float Selection Array=DataInterface() {propertyValues {FloatData [0,0.25,0.5,0.75]}}
     Random Seed=0
    )
    Sprite Rotation Mode=::NewEnumerator1
    Sprite Size Mode=::NewEnumerator3
    Uniform Sprite Size=$Emitter.RuneSize
    UsePositionOffset=false
    Write Material Random=true
   )
   &CylinderLocation(
    Cylinder Height=4
    Cylinder Radius=&FloatFromCurve(
     FloatCurve=DataInterface() {propertyValues {Curve {keys [<at 0 1 RCIM_Cubic>, {arriveTangent=-6.4476, interpMode=RCIM_Cubic, leaveTangent=-6.4476, tangentMode=RCTM_User, time=0.25, value=0}]}}}
     CurveIndex=$Emitter.NormalizedLoopAge
     Scale Curve=500
    )
    MaskSpawnGroup=false
    Override Seed=false
    Rotation Mode:ENiagara_RotationMode=::NewEnumerator1
    Surface Only=true
    Use Endcaps In Surface Only Mode=false
    UseOwnerScale=true
   )
   /SetVariables(Particles.Position=&self:GridSnap(Size:NiagaraFloat=$Emitter.RuneSize, StartPosition:Vector3f=Linked() {linkedVariable {name=Particles.Position, type=NiagaraPosition}})) {isSetParameters=true}
   &CameraOffset(Camera Offset Amount=8)
   &SpriteFacingAndAlignment(WriteAlignment=true, WriteFacing=true)
  ]
  particleUpdateScript [
   &ParticleState()
   &SubUVAnimation(
    End Frame=4
    Sub UV Play Rate:NiagaraFloat=&RandomRangeFloat(Minimum=0.5, Maximum=2)
    SubUV Animation Mode:ENiagara_SubUVLookupMode=::NewEnumerator3
    SubUV Start Time Offset:NiagaraFloat=&RandomRangeFloat(Minimum=0, Maximum=1)
   )
   &ScaleColor(
    Scale Alpha=&FloatFromCurve(FloatCurve=DataInterface() {propertyValues {Curve {keys [
         {arriveTangent=-3.8362, interpMode=RCIM_Cubic, leaveTangent=-3.8362, tangentMode=RCTM_User, time=0, value=1}
         <at 0.1866 0.4>
         {arriveTangent=-4.1865, interpMode=RCIM_Cubic, leaveTangent=-4.1865, tangentMode=RCTM_User, time=0.2015, value=1}
         <at 0.4552 0.3>
         {arriveTangent=-8.2931, interpMode=RCIM_Cubic, leaveTangent=-8.2931, tangentMode=RCTM_User, time=0.4652, value=0.75}
         <at 0.7089 0>
         {arriveTangent=-5.5334, interpMode=RCIM_Cubic, leaveTangent=-5.5334, tangentMode=RCTM_User, time=0.7338, value=0.5}
         <at 1 0>
        ]}}})
    Scale RGB=&VectorFromCurve(VectorCurve=DataInterface() {propertyValues {
       XCurve {keys [<at 0 1>,<at 1 1>]}
       YCurve {keys [<at 0 1>,<at 1 1>]}
       ZCurve {keys [
         {arriveTangent=1.0471, interpMode=RCIM_Cubic, leaveTangent=1.0471, tangentMode=RCTM_User, time=0, value=0.8}
         {interpMode=RCIM_Cubic, tangentMode=RCTM_User, time=0.25, value=1}
         {arriveTangent=-2.0404, interpMode=RCIM_Cubic, leaveTangent=-2.0404, tangentMode=RCTM_User, time=1, value=0}
        ]}
      }})
    ScaleA=true
    ScaleRGB=true
   )
  ]
  rendererClasses [&NiagaraSpriteRendererProperties]
  renderers [{rendererClass=&NiagaraSpriteRendererProperties}]
 }
 Sparkles {
  emitterUpdateScript [
   &EmitterState(Enable Distance Culling=true, MaxDistance=7000, Module.bUseMaxDistance=true, Module.bUseMinDistance=false, Scalability Mode=::NewEnumerator1)
   &SpawnBurst_Instantaneous(Spawn Count=1, Use Spawn Probability=false)
   &SpawnRate(SpawnRate=10, Use Spawn Probability=false)
  ]
  particleSpawnScript [
   &InitializeParticle(
    Color Maximum=<rgba 4 2.6494 0.0628 1>
    Color Minimum=<rgba 0.9788 0.0864 1 1>
    Color Mode=::NewEnumerator2
    Lifetime Max=8
    Lifetime Min=4
    Lifetime Mode=::NewEnumerator1
    Mesh Scale Mode=::NewEnumerator4
    Sprite Rotation Mode=::NewEnumerator2
    Sprite Size Mode=::NewEnumerator1
    Uniform Sprite Size Max=35
    Uniform Sprite Size Min=10
    UsePositionOffset=false
    Write Material Random=true
   )
   &CylinderLocation(
    Cylinder Height=4
    Cylinder Radius=400
    MaskSpawnGroup=false
    Override Seed=false
    Rotation Mode:ENiagara_RotationMode=::NewEnumerator1
    UseOwnerScale=true
   )
   /SetVariables(Particles.RandomPosition:NiagaraFloat=&FloatFromCurve(
     FloatCurve=DataInterface() {propertyValues {Curve {keys [
         {arriveTangent=-3.0656, interpMode=RCIM_Cubic, leaveTangent=-3.0656, tangentMode=RCTM_User, time=0, value=1}
         {arriveTangent=-0.5514, interpMode=RCIM_Cubic, leaveTangent=-0.5514, tangentMode=RCTM_User, time=0.5594, value=0.4766}
         {arriveTangent=-2.3442, interpMode=RCIM_Cubic, leaveTangent=-2.3442, tangentMode=RCTM_User, time=1, value=0}
        ]}}}
     CurveIndex=&RandomRangeFloat(Minimum=0, Maximum=1)
    )) {isSetParameters=true, module=SetVariables_#4}
  ]
  particleUpdateScript [
   &ParticleState()
   &DynamicMaterialParameters(
    Index 0 Param 1=1
    Index 0 Param 2=1
    Index 0 Param 3=0.5
    Module.Param0WriteEnabled=true
    Module.Param1WriteEnabled=true
    Module.Param2WriteEnabled=true
    Module.Param3WriteEnabled=true
   )
   &VectorNoiseForce(
    Force Amount=&Lerp_Float(A=100, B=1000, Alpha=$Particles.NormalizedAge)
    Module.Use Random Noise Time Interval=true
    Random Noise Time Interval=0.1
   )
   /Niagara/Modules/Update/Forces/LineAttractionForce(
    Attraction Strength:NiagaraFloat=&FloatFromCurve(
     FloatCurve=DataInterface() {propertyValues {Curve {keys [{arriveTangent=-0.728, interpMode=RCIM_Cubic, leaveTangent=-0.728, tangentMode=RCTM_User, time=0, value=1}, <at 1 0.25>]}}}
     Scale Curve=200
    )
    Line End:NiagaraPosition=$User.EndPoint
    Line Start:NiagaraPosition=$Engine.Owner.Position
    Use Attraction Falloff=false
   )
   &WindForce(
    Air Resistance=0.5
    Wind Speed=$System.DirectionToEndPoint
    Wind Speed Scale=&FloatFromCurve(
     FloatCurve=DataInterface() {propertyValues {Curve {keys [
         {arriveTangent=0.5837, interpMode=RCIM_Cubic, time=0, value=0}
         {arriveTangent=1.9345, interpMode=RCIM_Cubic, leaveTangent=1.9345, tangentMode=RCTM_User, time=0.1605, value=0.1637}
         <at 1 1 RCIM_Cubic>
        ]}}}
     Scale Curve=1400
    )
   )
   &Drag(Drag=&RandomRangeFloat(Minimum=1, Maximum=1.5), Use Linear Drag=true, Use Rotational Drag=false)
   &ScaleColor(
    Scale Alpha=&FloatFromCurve(FloatCurve=DataInterface() {propertyValues {Curve {keys [
         <at 0 0 RCIM_Cubic>
         {interpMode=RCIM_Cubic, tangentMode=RCTM_User, time=0.0856, value=0.9914}
         {arriveTangent=-2.3503, interpMode=RCIM_Cubic, leaveTangent=-2.3503, tangentMode=RCTM_User, time=0.8081, value=0.5934}
         <at 1 0>
        ]}}})
    Scale RGB=&VectorFromFloat(Value=1)
    ScaleA=true
    ScaleRGB=true
   )
   &ScaleSpriteSize(Scale Factor=&Vector2DFromFloat(Value=&FloatFromCurve(FloatCurve=DataInterface() {propertyValues {Curve {keys [
          <at 0 0 RCIM_Cubic>
          {arriveTangent=0.836, interpMode=RCIM_Cubic, leaveTangent=0.836, time=0.2, value=1}
          {arriveTangent=-1.8455, interpMode=RCIM_Cubic, leaveTangent=-1.8455, tangentMode=RCTM_User, time=0.6507, value=0.544}
          <at 1 0>
         ]}}})))
   &SolveForcesAndVelocity(Clamp Velocity=false, Limit Acceleration=false)
  ]
  rendererClasses [&NiagaraSpriteRendererProperties]
  renderers [{rendererClass=&NiagaraSpriteRendererProperties}]
 }
 SparklesEnd {
  emitterUpdateScript [
   &EmitterState(Enable Distance Culling=true, MaxDistance=7000, Module.bUseMaxDistance=true, Module.bUseMinDistance=false, Scalability Mode=::NewEnumerator1)
   &SpawnBurst_Instantaneous(Spawn Count=1, Use Spawn Probability=false)
   &SpawnRate(SpawnRate=10, Use Spawn Probability=false)
  ]
  particleSpawnScript [
   &InitializeParticle(
    Color Maximum=<rgba 2 1.3247 0.0314 1>
    Color Minimum=<rgba 0.9788 0.0864 1 1>
    Color Mode=::NewEnumerator2
    Lifetime Max=4.7
    Lifetime Min=4
    Lifetime Mode=::NewEnumerator1
    Mesh Scale Mode=::NewEnumerator4
    Sprite Rotation Mode=::NewEnumerator2
    Sprite Size Mode=::NewEnumerator1
    Uniform Sprite Size Max=35
    Uniform Sprite Size Min=10
    UsePositionOffset=false
    Write Material Random=true
   )
   &CylinderLocation(
    Cylinder Height=4
    Cylinder Origin:NiagaraPosition=&Add_Vector(
     A=Linked() {linkedVariable {name=User.EndPoint, type=NiagaraPosition}}
     B=&Multiply_VectorByFloat(Vector=$System.DirectionToEndPoint, Float=&RandomRangeFloat(Minimum=-450, Maximum=-400))
    )
    Cylinder Radius=40
    MaskSpawnGroup=false
    Override Seed=false
    Rotation Mode:ENiagara_RotationMode=::NewEnumerator1
    UseOwnerScale=true
   )
   /SetVariables(Particles.RandomPosition:NiagaraFloat=&FloatFromCurve(
     FloatCurve=DataInterface() {propertyValues {Curve {keys [
         {arriveTangent=-3.0656, interpMode=RCIM_Cubic, leaveTangent=-3.0656, tangentMode=RCTM_User, time=0, value=1}
         {arriveTangent=-0.5514, interpMode=RCIM_Cubic, leaveTangent=-0.5514, tangentMode=RCTM_User, time=0.5594, value=0.4766}
         {arriveTangent=-2.3442, interpMode=RCIM_Cubic, leaveTangent=-2.3442, tangentMode=RCTM_User, time=1, value=0}
        ]}}}
     CurveIndex=&RandomRangeFloat(Minimum=0, Maximum=1)
    )) {isSetParameters=true, module=SetVariables_#4}
   &AddVelocity(
    Coordinate Space=::World
    Velocity=&Multiply_VectorByFloat(Vector=$System.DirectionToEndPoint, Float=&RandomRangeFloat(Minimum=750, Maximum=850))
   )
  ]
  particleUpdateScript [
   &ParticleState()
   &DynamicMaterialParameters(
    Index 0 Param 1=1
    Index 0 Param 2=1
    Index 0 Param 3=0.5
    Module.Param0WriteEnabled=true
    Module.Param1WriteEnabled=true
    Module.Param2WriteEnabled=true
    Module.Param3WriteEnabled=true
   )
   &VectorNoiseForce(Force Amount=800, Module.Use Random Noise Time Interval=true, Random Noise Time Interval=0.1)
   &LineAttractionForce(
    Attraction Strength:NiagaraFloat=&FloatFromCurve(
     FloatCurve=DataInterface() {propertyValues {Curve {keys [
         {arriveTangent=-0.728, interpMode=RCIM_Cubic, leaveTangent=-0.728, tangentMode=RCTM_User, time=0, value=0}
         {arriveTangent=0.25, interpMode=RCIM_Cubic, leaveTangent=0.25, time=0.3, value=1}
         <at 0.4 0.1>
         <at 0.7 0>
        ]}}}
     Scale Curve=-500
    )
    Line End:NiagaraPosition=$Engine.Owner.Position
    Line Start:NiagaraPosition=$User.EndPoint
    Use Attraction Falloff=false
   )
   &Drag(Drag=&RandomRangeFloat(Minimum=2, Maximum=2.25), Use Linear Drag=true, Use Rotational Drag=false)
   &ScaleColor(
    Scale Alpha=&FloatFromCurve(FloatCurve=DataInterface() {propertyValues {Curve {keys [
         <at 0 0 RCIM_Cubic>
         {interpMode=RCIM_Cubic, tangentMode=RCTM_User, time=0.0856, value=0.9914}
         {arriveTangent=-2.3503, interpMode=RCIM_Cubic, leaveTangent=-2.3503, tangentMode=RCTM_User, time=0.8081, value=0.5934}
         <at 1 0>
        ]}}})
    Scale RGB=&VectorFromFloat(Value=1)
    ScaleA=true
    ScaleRGB=true
   )
   &ScaleSpriteSize(Scale Factor=&Vector2DFromFloat(Value=&FloatFromCurve(FloatCurve=DataInterface() {propertyValues {Curve {keys [
          {arriveTangent=11.4729, interpMode=RCIM_Cubic, leaveTangent=11.4729, tangentMode=RCTM_User, time=0, value=0}
          <at 0.2 1 RCIM_Cubic>
          <at 1 0>
         ]}}})))
   &SolveForcesAndVelocity(Clamp Velocity=false, Limit Acceleration=false)
  ]
  rendererClasses [&NiagaraSpriteRendererProperties]
  renderers [{rendererClass=&NiagaraSpriteRendererProperties}]
 }
}
summary {
 systemName=NS_Corruption_Awakening_Beam
 emitters [
  {emitterName=BeamMeshes, rendererClasses [&NiagaraSpriteRendererProperties, &NiagaraMeshRendererProperties]}
  {emitterName=Sparkles, rendererClasses [&NiagaraSpriteRendererProperties]}
  {emitterName=RunesStart, rendererClasses [&NiagaraSpriteRendererProperties]}
  {emitterName=SparklesEnd, rendererClasses [&NiagaraSpriteRendererProperties]}
  {emitterName=RunesEnd, rendererClasses [&NiagaraSpriteRendererProperties]}
  {emitterName=GlowSurface, rendererClasses [&NiagaraSpriteRendererProperties]}
  {emitterName=Lights, rendererClasses [&NiagaraLightRendererProperties]}
  {
   emitterName=Lightning
   rendererClasses [
    &NiagaraMeshRendererProperties
    &NiagaraSpriteRendererProperties
    &NiagaraLightRendererProperties
   ]
  }
  {emitterName=NE_FN_Sparks, simTarget=GPUComputeSim, rendererClasses [&NiagaraSpriteRendererProperties]}
 ]
 userVariables [
  User.EndNormal {type=Vector3f, description="", defaultValue {struct=Vector3f, value=<-1 0 0>}}
  User.EndPoint {type=NiagaraPosition, description="", defaultValue {}}
  User.EndUp {type=Vector3f, description="", defaultValue {struct=Vector3f, value=<0 0 1>}}
 ]
}
