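# Sound cue for randomized zap sounds of the Cube's awaken hover in the Corruption gameplay content.
# A quality-level node routes the first two levels to a random pick of eight waves with broader pitch and volume modulation.
# The last two levels use a three-wave subset with narrower modulation, excluded from branch culling.
# It uses the awaken loop attenuation, an important mission sound class and a small concurrency cap.
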SoundCue /CorruptionGameplay/Gameplay/Environmental/Cube/Audio/Cube_Awaken_Hover_Zaps_Cue
root=QualityLevel_0
nodes [
 {id=Modulator_0, children [Random_0], properties {pitchMax=1.2, pitchMin=0.9, volumeMax=1, volumeMin=0.6}}
 {id=Modulator_1, children [Random_1], properties {pitchMax=1.1, pitchMin=0.9}}
 {id=QualityLevel_0, children [Modulator_0,Modulator_0,Modulator_1,Modulator_1]}
 {
  id=Random_0
  children [WavePlayer_2,WavePlayer_3,WavePlayer_4,WavePlayer_5,WavePlayer_6,WavePlayer_7,WavePlayer_8,WavePlayer_9]
  properties {weights [1,1,1,1,1,1,1,1]}
 }
 {id=Random_1, children [WavePlayer_10, WavePlayer_11, WavePlayer_12], properties {bShouldExcludeFromBranchCulling=true, weights [1,1,1]}}
 {id=WavePlayer_10, properties {soundWaveAssetPtr=&/Cube_Awaken_Hover_Zaps_01}}
 {id=WavePlayer_11, properties {soundWaveAssetPtr=&/Cube_Awaken_Hover_Zaps_05}}
 {id=WavePlayer_12, properties {soundWaveAssetPtr=&/Cube_Awaken_Hover_Zaps_07}}
 {id=WavePlayer_2, properties {soundWaveAssetPtr=&Cube_Awaken_Hover_Zaps_01}}
 {id=WavePlayer_3, properties {soundWaveAssetPtr=&/Cube_Awaken_Hover_Zaps_02}}
 {id=WavePlayer_4, properties {soundWaveAssetPtr=&/Cube_Awaken_Hover_Zaps_03}}
 {id=WavePlayer_5, properties {soundWaveAssetPtr=&/Cube_Awaken_Hover_Zaps_04}}
 {id=WavePlayer_6, properties {soundWaveAssetPtr=&Cube_Awaken_Hover_Zaps_05}}
 {id=WavePlayer_7, properties {soundWaveAssetPtr=&/Cube_Awaken_Hover_Zaps_06}}
 {id=WavePlayer_8, properties {soundWaveAssetPtr=&Cube_Awaken_Hover_Zaps_07}}
 {id=WavePlayer_9, properties {soundWaveAssetPtr=&/Cube_Awaken_Hover_Zaps_08}}
]
properties {
 attenuationSettings=&/Cube_Awaken_Loop_Att
 bOverrideConcurrency=true
 maxDistance=12000
 soundClassObject=&Fort_Mission_Ingame_Important_PSM
 volumeMultiplier=0.7
 concurrencyOverrides {maxCount=3}
}
