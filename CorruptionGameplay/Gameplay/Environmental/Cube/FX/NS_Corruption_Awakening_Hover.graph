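# Niagara effect for the hover stage of the Corruption cube's awakening, with placement scaled by User.CubeRadius.
# A glow mesh with light, random lightning bursts with lights and zap audio, and rising, noise-jittered dirt chunks and sparkles surround the cube.
# Flickering SubUV rune sprites cluster along bottom-face edges and climb grid-snapped across the sides, using spawn groups to choose placement.
# Loop-age curves surge glow, spawn rates and noise force early, then settle.
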NiagaraSystem /CorruptionGameplay/Gameplay/Environmental/Cube/FX/NS_Corruption_Awakening_Hover
dependencies {
 usedDataInterfaces [
  &NiagaraDataInterfaceCurve
  &NiagaraDataInterfaceArrayFloat
  &NiagaraDataInterfaceVectorCurve
  &NiagaraDataInterfaceAudioPlayer
 ]
 usedDynamicInputs []
 usedModules [&SystemState]
}
emitters {
 Glow {
  emitterUpdateScript [
   &EmitterState(
    Inactive Response=::NewEnumerator1
    Life Cycle Mode=::NewEnumerator1
    Loop Behavior=::NewEnumerator1
    Loop Duration Mode=::NewEnumerator1
    UseLoopDelay=false
   )
   &SpawnBurst_Instantaneous(Use Spawn Probability=false, UseLoopCountLimit=false)
  ]
  particleSpawnScript [
   &InitializeParticle(Color=<rgba 1 1 1 0.5>, Lifetime=$System.CurrentLoopDuration, UsePositionOffset=false, Write Material Random=true)
   &InitialMeshOrientation(Mesh Orientation Mode=::NewEnumerator1, Rotation=<0 -0.25 0>, Use Rotation Vector=true)
   &CameraOffset(Camera Offset Amount=2)
   /SetVariables(Particles.LightPos:Vector3f=<100 0 0>) {isSetParameters=true}
  ]
  particleUpdateScript [
   &ParticleState(Kill Particles When Lifetime Has Elapsed=false)
   &ScaleColor(
    Scale Alpha=&FloatFromCurve(FloatCurve=DataInterface() {propertyValues {Curve {keys [
         {interpMode=RCIM_Cubic, tangentMode=RCTM_User, time=0, value=0}
         {arriveTangentWeight=0.0833, interpMode=RCIM_Cubic, leaveTangentWeight=0.15, tangentMode=RCTM_User, time=0.5, value=1}
        ]}}})
    Scale RGB=&VectorFromFloat(Value=&Add_Float(
      A=1
      B=&FloatFromCurve(
       FloatCurve=DataInterface() {propertyValues {Curve {keys [
           {arriveTangent=10.4941, interpMode=RCIM_Cubic, time=0, value=0}
           {
            arriveTangent=-2.7982
            arriveTangentWeight=0.0471
            interpMode=RCIM_Cubic
            leaveTangent=-2.7982
            leaveTangentWeight=0.0707
            tangentMode=RCTM_User
            time=0.3
            value=1
           }
           <at 1 0>
          ]}}}
       CurveIndex=$System.NormalizedLoopAge
      )
     ))
    ScaleA=true
    ScaleRGB=true
   )
   /SetVariables(Particles.LightColor=&MultiplyLinearColor(
     Linear Color=<rgba 50 11.4896 1.0417 1>
     Scale Factor=&MakeVector4_FromLinearColor(LinearColor=$Particles.Color)
    )) {isSetParameters=true}
  ]
  rendererClasses [&NiagaraMeshRendererProperties, &NiagaraLightRendererProperties]
  renderers [
   {rendererClass=&NiagaraMeshRendererProperties, rendererIndex=0}
   {rendererClass=&NiagaraLightRendererProperties, rendererIndex=1}
  ]
 }
 Lightning {
  emitterUpdateScript [
   &EmitterState(
    Enable Distance Culling=true
    Life Cycle Mode=::NewEnumerator1
    Loop Duration=&RandomRangeFloat(Minimum=0.3, Maximum=2)
    MaxDistance=20000
    Module.bUseMaxDistance=true
    Module.bUseMinDistance=false
    Recalculate Duration Each Loop=true
    Scalability Mode=::NewEnumerator1
    UseLoopDelay=false
   )
   &SpawnBurst_Instantaneous(Spawn Probability=0.3, Use Spawn Probability=true, UseLoopCountLimit=false)
   &SpawnBurst_Instantaneous(Spawn Count=2, Spawn Probability=0.25, Spawn Time=0.45, Use Spawn Probability=true, UseLoopCountLimit=false) {module=SpawnBurst_Instantaneous001}
   &SpawnBurst_Instantaneous(Spawn Count=2, Spawn Probability=0.4, Spawn Time=0.75, Use Spawn Probability=true, UseLoopCountLimit=false) {module=SpawnBurst_Instantaneous002}
   &SpawnRate(
    SpawnRate=&FloatFromCurve(
     FloatCurve=DataInterface() {propertyValues {Curve {keys [
         {arriveTangent=10.4941, interpMode=RCIM_Cubic, time=0, value=0}
         {arriveTangent=5.4555, interpMode=RCIM_Cubic, leaveTangent=5.4555, tangentMode=RCTM_User, time=0.2, value=0.35}
         {
          arriveTangent=-8.7693
          arriveTangentWeight=0.0471
          interpMode=RCIM_Cubic
          leaveTangent=-8.7693
          leaveTangentWeight=0.0707
          tangentMode=RCTM_User
          time=0.3
          value=1
         }
         {arriveTangentWeight=0.3941, interpMode=RCIM_Cubic, leaveTangentWeight=0.9853, tangentMode=RCTM_User, time=0.4, value=0.1}
         <at 1 0.1>
        ]}}}
     CurveIndex=$System.NormalizedLoopAge
     Scale Curve=15
    )
    Use Spawn Probability=false
   )
  ]
  particleSpawnScript [
   /SetVariables(Particles.BiasedRandom=&FloatFromCurve(
     FloatCurve=DataInterface() {propertyValues {Curve {keys [
         {interpMode=RCIM_Cubic, tangentMode=RCTM_User, time=0, value=0}
         {arriveTangent=2.5197, interpMode=RCIM_Cubic, leaveTangent=2.5197, tangentMode=RCTM_User, time=0.6, value=0.6}
         {interpMode=RCIM_Cubic, tangentMode=RCTM_User, time=1, value=1}
        ]}}}
     CurveIndex=&RandomRangeFloat(Minimum=0, Maximum=1)
    )) {isSetParameters=true}
   /Niagara/Modules/Spawn/Initialization/InitializeParticle(
    Color=<rgba 0.2156 0.5 0.0812 1>
    Lifetime=&Lerp_Float(A=0.07, B=0.4, Alpha=$Particles.BiasedRandom)
    Mesh Scale=&MakeVector(
     X=&RandomRangeFloat(Minimum=0.7, Maximum=2)
     Y=&RandomRangeFloat(Minimum=0.25, Maximum=1.5)
     Z=&RandomRangeFloat(Minimum=0.25, Maximum=1.5)
    )
    Module.Write Color=true
    Module.Write Lifetime=true
    Module.Write Mass=true
    Module.Write Material Random=true
    Module.Write Position=false
    Module.Write Scale=true
    Module.Write SpriteRotation=false
    Module.Write SpriteSize=true
    Position=&SimulationPosition()
    Sprite Size=&Vector2DFromFloat(Value=&RandomRangeFloat(Minimum=300, Maximum=700))
   )
   /SetVariables(Particles.Position=&MakeVector(
     X=5
     Y=&Multiply_Float(A=$User.CubeRadius, B=&RandomRangeFloat(Minimum=-0.9, Maximum=0.9))
     Z=&Multiply_Float(A=$User.CubeRadius, B=&RandomRangeFloat(Minimum=-0.9, Maximum=0.9))
    )) {isSetParameters=true}
   /SetVariables(Particles.LightPosition:Vector3f=&Add_Vector(
     A=Linked() {linkedVariable {name=Particles.Position, type=NiagaraPosition}}
     B=&MakeVector(X=&RandomRangeFloat(Minimum=150, Maximum=400), Y=0, Z=0)
    )) {isSetParameters=true}
   &InitialMeshOrientation(
    Mesh Orientation Mode=::NewEnumerator2
    Rotation=&RandomRangeVector(Minimum=<0 0 0>, Maximum=<1 0 0>)
    Use Rotation Vector=true
   ) {module=InitialMeshOrientation001}
   /SetVariables(Particles.MeshIndex=&RandomRangeInt(Minimum=0, Maximum=1)) {isSetParameters=true}
   /SetVariables(Particles.DynParamsSprite:Vector4f=<0.25 2 1 1>) {isSetParameters=true}
   /SetVariables(Particles.CamOffsetSprite=50) {isSetParameters=true}
   /SetVariables(Particles.LightRadius=&RandomRangeFloat(Minimum=80, Maximum=150)) {isSetParameters=true}
   /Niagara/Modules/Audio/PlayAudio(Audio Player=DataInterface() {propertyValues {
      Attenuation=null
      Concurrency=null
      MaxPlaysPerTick=10
      SoundToPlay=/CorruptionGameplay/Gameplay/Environmental/Cube/Audio/Cube_Awaken_Hover_Zaps_Cue
      bAllowLoopingOneShotSounds=false
      bLimitPlaysPerTick=true
      bOnlyActiveDuringGameplay=false
      bStopWhenComponentIsDestroyed=true
      ConfigurationUserParameter {parameter {name=None, typeDefHandle {registeredTypeIndex=1}, varData []}}
      ParameterNames []
     }})
  ]
  particleUpdateScript [
   &ParticleState()
   &ScaleColor(
    Scale Alpha=&FloatFromCurve(FloatCurve=DataInterface() {propertyValues {Curve {keys [
         <at 0 0>
         {interpMode=RCIM_Cubic, tangentMode=RCTM_User, time=0.1258, value=0.1952}
         <at 0.341 0.0902>
         {interpMode=RCIM_Cubic, tangentMode=RCTM_User, time=0.496, value=0.2532}
         <at 1 0>
        ]}}})
    Scale RGB=&VectorFromFloat(Value=&FloatFromCurve(FloatCurve=DataInterface() {propertyValues {Curve {keys [
          {arriveTangent=3.5209, interpMode=RCIM_Cubic, leaveTangent=3.5209, tangentMode=RCTM_User, time=0, value=0.5}
          {arriveTangent=-20.2701, interpMode=RCIM_Cubic, leaveTangent=-20.2701, tangentMode=RCTM_User, time=0.45, value=5}
          <at 1 1>
         ]}}}))
    ScaleA=true
    ScaleRGB=true
   )
   /SetVariables(Particles.SpriteColor=&MultiplyLinearColor(
     Linear Color=$Particles.Color
     Scale Factor=&MakeVector4_FromLinearColor(LinearColor=&MultiplyLinearColor(
       Linear Color=<rgba 25 15.0635 2.969 1>
       Scale Factor=&MakeVector4(
        X=1
        Y=1
        Z=1
        W=&FloatFromCurve(FloatCurve=DataInterface() {propertyValues {Curve {keys [
             {interpMode=RCIM_Cubic, tangentMode=RCTM_User, time=0, value=1}
             {arriveTangentWeight=0.031, interpMode=RCIM_Cubic, leaveTangentWeight=0.0366, tangentMode=RCTM_User, time=0.0918, value=0.05}
             {arriveTangentWeight=0.16, interpMode=RCIM_Cubic, leaveTangent=18.7153, leaveTangentWeight=0.1479, tangentMode=RCTM_Break, time=0.18, value=0}
             {arriveTangent=-0.2815, interpMode=RCIM_Cubic, leaveTangent=-0.2815, tangentMode=RCTM_User, time=0.3, value=0.95}
             {arriveTangentWeight=0.0345, interpMode=RCIM_Cubic, leaveTangentWeight=0.0566, tangentMode=RCTM_User, time=0.4, value=0.1}
             {
              arriveTangentWeight=0.1589
              interpMode=RCIM_Cubic
              leaveTangent=29.6956
              leaveTangentWeight=0.1018
              tangentMode=RCTM_Break
              time=0.5
              value=0.05
             }
             {arriveTangent=0.512, interpMode=RCIM_Cubic, leaveTangent=0.512, tangentMode=RCTM_User, time=0.5641, value=0.8648}
             <at 1 0>
            ]}}})
       )
      ))
    )) {isSetParameters=true}
   /SetVariables(Particles.LightColor=&MultiplyLinearColor(Linear Color=$Particles.SpriteColor, Scale Factor=<1 1 1 5>)) {isSetParameters=true}
   &DynamicMaterialParameters(
    Index 0 Param 1=&FloatFromCurve(FloatCurve=DataInterface() {propertyValues {Curve {keys [{tangentMode=RCTM_User, time=0, value=1}, {tangentMode=RCTM_User, time=1, value=0}]}}})
    Index 0 Param 2=&Add_Float(
     A=&FloatFromCurve(FloatCurve=DataInterface() {propertyValues {Curve {keys [
          {interpMode=RCIM_Cubic, leaveTangent=8.981, tangentMode=RCTM_Break, time=0, value=-0.3}
          {arriveTangent=0.0136, interpMode=RCIM_Cubic, leaveTangent=-5.3647, tangentMode=RCTM_Break, time=0.4833, value=1}
          <at 1 -0.3>
         ]}}})
     B=&Multiply_Float(A=$Particles.BiasedRandom, B=0.5)
    )
    Index 0 Param 3=&RandomRangeFloat(Minimum=0, Maximum=3)
    Index 1 Param 1=&Add_Float(A=$Particles.NormalizedAge, B=&RandomRangeFloat(Minimum=0, Maximum=1))
    Index 1 Param 2=&RandomRangeFloat(Minimum=-50, Maximum=50)
    Index 1 Param 3=3
    Module.Param0WriteEnabled=true
    Module.Param0WriteEnabled1=true
    Module.Param1WriteEnabled=true
    Module.Param1WriteEnabled1=true
    Module.Param2WriteEnabled=true
    Module.Param2WriteEnabled1=true
    Module.Param3WriteEnabled=false
    Module.Param3WriteEnabled1=false
    Write Parameter Index 1=true
   ) {module=DynamicMaterialParameters001}
   &ScaleSpriteSize(Uniform Curve Sprite Scale=DataInterface() {propertyValues {Curve {keys [
        {arriveTangent=21.5526, interpMode=RCIM_Cubic, time=0, value=0}
        <at 0.1 1 RCIM_Cubic>
        {arriveTangent=-1.8758, interpMode=RCIM_Cubic, leaveTangent=-1.8758, tangentMode=RCTM_User, time=1, value=0}
       ]}}})
  ]
  rendererClasses [
   &NiagaraMeshRendererProperties
   &NiagaraSpriteRendererProperties
   &NiagaraLightRendererProperties
  ]
  renderers [
   {rendererClass=&NiagaraMeshRendererProperties, rendererIndex=0}
   {rendererClass=&NiagaraSpriteRendererProperties, rendererIndex=1}
   {rendererClass=&NiagaraLightRendererProperties, rendererIndex=2}
  ]
 }
 NE_FN_DirtChunks {
  emitterUpdateScript [
   &EmitterState(Enable Distance Culling=true, MaxDistance=7500, Module.bUseMaxDistance=true, Module.bUseMinDistance=false, Scalability Mode=::NewEnumerator1)
   &SpawnRate(SpawnRate=7, Use Spawn Probability=false)
   &SpawnRate(SpawnGroup=1, SpawnRate=2.5, Use Spawn Probability=false) {module=SpawnRate001}
  ]
  particleSpawnScript [
   /Niagara/Modules/Spawn/Initialization/V2/InitializeParticle(
    Color=<rgba 2 0.665 0.0366 0.25>
    Lifetime Max=6
    Lifetime Min=4
    Lifetime Mode=::NewEnumerator1
    Mesh Scale=&Multiply_VectorByFloat(
     Vector=&UniformRangedVector(Minimum=<0.2 0.4 0.4>, Maximum=<0.8 1 1.1>)
     Float=&UniformRangedFloat(Minimum=0.12, Maximum=0.32)
    )
    Mesh Scale Mode=::NewEnumerator4
    Sprite Size Max=<25 25>
    Sprite Size Min=<10 10>
    Sprite Size Mode=::NewEnumerator2
    UsePositionOffset=false
    Write Material Random=true
   )
   /SetVariables(Particles.VisibilityTag=&SpawnGroup()) {isSetParameters=true}
   /SetVariables(
    Particles.RandomTimeOffset:NiagaraFloat=&RandomRangeFloat(Minimum=-0.1, Maximum=0.25)
    Particles.RockColor:LinearColor=<rgba 0 0 0 1>
   ) {isSetParameters=true}
   &InitialMeshOrientation(Orientation Vector=&RandomVector(Vector Scale=1), Use Orientation Vector=true, Use Rotation Vector=false)
   /SetVariables(Particles.Position=/Niagara/DynamicInputs/Transforms/TransformPosition(
     Position=&MakeVector(
      X=&FloatFromCurve(
       FloatCurve=DataInterface() {propertyValues {Curve {keys [<at 0 0.05 RCIM_Cubic>, {arriveTangent=2.0792, interpMode=RCIM_Cubic, leaveTangent=2.0792, tangentMode=RCTM_User, time=1, value=1}]}}}
       CurveIndex=&RandomRangeFloat(Minimum=0, Maximum=1)
       Scale Curve=900
      )
      Y=&Multiply_Float(A=$User.CubeRadius, B=&RandomRangeFloat(Minimum=-0.9, Maximum=0.9))
      Z=&Multiply_Float(A=$User.CubeRadius, B=&RandomRangeFloat(Minimum=-0.9, Maximum=0.9))
     )
     Source Space=::Local
     Destination Space=::Simulation
    )) {isSetParameters=true}
   &AddRotationalVelocity(Rotation Rate=&RandomVector(Vector Scale=&RandomRangeFloat(Minimum=0.15, Maximum=0.3)))
   &ApplyInitialForces()
   &DynamicMaterialParameters(Module.Param0WriteEnabled=true, Module.Param1WriteEnabled=false, Module.Param2WriteEnabled=false, Module.Param3WriteEnabled=false)
   &SubUVAnimation(End Frame=3, SubUV Animation Mode:ENiagara_SubUVLookupMode=::NewEnumerator2) {module=SubUVAnimation001}
  ]
  particleUpdateScript [
   &ParticleState()
   /SetVariables(Particles.GlowDynParams:Vector4f=&MakeVector4(
     X=&FloatFromCurve(FloatCurve=DataInterface() {propertyValues {Curve {keys [
          {arriveTangent=10.4941, interpMode=RCIM_Cubic, time=0, value=0.7}
          {
           arriveTangent=1.4377
           arriveTangentWeight=0.0471
           interpMode=RCIM_Cubic
           leaveTangent=1.4377
           leaveTangentWeight=0.0707
           tangentMode=RCTM_User
           time=0.3
           value=0.1
          }
          <at 0.7 0.7>
         ]}}})
     Y=&RandomRangeFloat(Minimum=1, Maximum=2)
     Z=0
     W=0
    )) {isSetParameters=true}
   &ScaleColor(
    Scale Alpha=&FloatFromCurve(FloatCurve=DataInterface() {propertyValues {Curve {keys [
         {arriveTangent=10.4941, interpMode=RCIM_Cubic, time=0, value=0.05}
         {arriveTangent=0.9, arriveTangentWeight=0.0471, interpMode=RCIM_Cubic, leaveTangent=0.9, leaveTangentWeight=0.0707, time=0.3, value=1}
         {
          arriveTangent=-1.3571
          arriveTangentWeight=0.3941
          interpMode=RCIM_Cubic
          leaveTangent=-1.3571
          leaveTangentWeight=0.9853
          time=0.5
          value=0.5
         }
         <at 1 0.05>
        ]}}})
    ScaleA=true
    ScaleRGB=false
   )
   &GravityForce(Gravity=&RandomRangeVector(Minimum=<-1 -1 500>, Maximum=<1 1 1500>))
   &VectorNoiseForce(
    Force Amount=&FloatFromCurve(
     FloatCurve=DataInterface() {propertyValues {Curve {keys [
         {arriveTangent=10.4941, interpMode=RCIM_Cubic, time=0, value=0.05}
         {arriveTangent=0.4, arriveTangentWeight=0.0471, interpMode=RCIM_Cubic, leaveTangent=0.4, leaveTangentWeight=0.0707, time=0.3, value=1}
         {arriveTangentWeight=0.3941, leaveTangentWeight=0.9853, time=0.5, value=0.25}
         <at 1 0.05>
        ]}}}
     CurveIndex=&Add_Float(A=$Emitter.NormalizedLoopAge, B=&RandomRangeFloat(Minimum=-0.1, Maximum=0.25))
     Scale Curve=10000
    )
    Module.Use Random Noise Time Interval=true
    Random Noise Time Interval=0.05
   )
   &Drag(
    Drag=&FloatFromCurve(FloatCurve=DataInterface() {propertyValues {Curve {keys [<at 4 0.1 RCIM_Cubic>,<at 12 1>]}}}, CurveIndex=$System.Age, Scale Curve=50)
    Use Linear Drag=true
    Use Rotational Drag=false
   )
   &SolveForcesAndVelocity(Clamp Velocity=false, Limit Acceleration=false)
   /SetVariables(Particles.RockBlurParams:Vector4f=&MakeFloat4FromVectorAndFloat("Vector (RGB)"=$Particles.Velocity, "Float (Alpha)"=-1)) {isSetParameters=true}
   &SolveRotationalForcesAndVelocity()
   &ScaleMeshSize(Scale Factor=&VectorFromFloat(Value=&FloatFromCurve(FloatCurve=DataInterface() {propertyValues {Curve {keys [
          {interpMode=RCIM_Cubic, tangentMode=RCTM_User, time=0, value=0}
          {interpMode=RCIM_Cubic, tangentMode=RCTM_User, time=0.2, value=1}
          {arriveTangentWeight=0.3202, interpMode=RCIM_Cubic, leaveTangentWeight=0.1067, tangentMode=RCTM_User, time=0.8, value=1}
          {tangentMode=RCTM_User, time=1, value=0}
         ]}}})))
   &ScaleSpriteSize(Scale Factor=&Vector2DFromFloat(Value=&FloatFromCurve(FloatCurve=DataInterface() {propertyValues {Curve {keys [
          {interpMode=RCIM_Cubic, tangentMode=RCTM_User, time=0, value=0}
          {interpMode=RCIM_Cubic, tangentMode=RCTM_User, time=0.2, value=1}
          {arriveTangentWeight=0.3202, interpMode=RCIM_Cubic, leaveTangentWeight=0.1067, tangentMode=RCTM_User, time=0.8, value=1}
          {tangentMode=RCTM_User, time=1, value=0}
         ]}}})))
   /SetVariables(Particles.GlowSize=&Vector2DFromFloat(Value=&Multiply_Float(
      A=/Niagara/DynamicInputs/Math/CalculateParticleRadius(
       Radius Calculation Type:ENiagaraCollisionRadiusOptions=::NewEnumerator0
       Particle Radius:NiagaraFloat=1
       Mesh Dimensions:Vector3f=<1 1 1>
       Method for Calculating Particle Radius:ENiagaraCalculateRadiusOptions=::NewEnumerator2
      )
      B=&RandomRangeFloat(Minimum=10, Maximum=20)
     ))) {isSetParameters=true}
   &SpriteRotationRate(Rotation Rate=&Lerp_Float(A=-70, B=70, Alpha=$Particles.Initial.MaterialRandom))
  ]
  rendererClasses [&NiagaraMeshRendererProperties, &NiagaraSpriteRendererProperties]
  renderers [
   {rendererClass=&NiagaraMeshRendererProperties, rendererIndex=0}
   {rendererClass=&NiagaraSpriteRendererProperties, rendererIndex=1}
   {rendererClass=&NiagaraSpriteRendererProperties, rendererIndex=2}
  ]
 }
 RunesBottom {
  emitterSpawnScript [/SetVariables(Emitter.RuneSize=48) {isSetParameters=true, module=SetVariables_#1}]
  emitterUpdateScript [
   &EmitterState(Enable Distance Culling=true, MaxDistance=9000, Module.bUseMaxDistance=true, Module.bUseMinDistance=false, Scalability Mode=::NewEnumerator1)
   &SpawnRate(
    SpawnRate=&FloatFromCurve(
     FloatCurve=DataInterface() {propertyValues {Curve {keys [
         {arriveTangent=10.4941, interpMode=RCIM_Cubic, time=0, value=0}
         {arriveTangent=0.2, arriveTangentWeight=0.0471, interpMode=RCIM_Cubic, leaveTangent=0.2, leaveTangentWeight=0.0707, time=0.3, value=1}
         {arriveTangentWeight=0.3941, interpMode=RCIM_Cubic, leaveTangentWeight=0.9853, tangentMode=RCTM_User, time=0.5, value=0.2}
         <at 1 0.1>
        ]}}}
     CurveIndex=$Emitter.NormalizedLoopAge
     Scale Curve=600
    )
    Use Spawn Probability=false
   )
  ]
  particleSpawnScript [
   &FN_SetSpawnGroup(Number of Spawn Groups=2)
   &InitializeParticle(
    Color Maximum=<rgba 1 16 15 0.75>
    Color Minimum=<rgba 0 5 2 0.4>
    Color Mode=::NewEnumerator2
    Lifetime Max=0.75
    Lifetime Min=0.05
    Lifetime Mode=::NewEnumerator1
    Sprite Rotation Angle=&SelectFloatFromArray(
     Array Sampling Mode=::NewEnumerator0
     Direct Array Index=?
     Float Selection Array=DataInterface() {propertyValues {FloatData [0,0.25,0.5,0.75]}}
     Random Seed=0
    )
    Sprite Rotation Mode=::NewEnumerator1
    Sprite Size Mode=::NewEnumerator3
    Uniform Sprite Size=$Emitter.RuneSize
    UsePositionOffset=false
    Write Material Random=true
   )
   /SetVariables(Particles.CubeRad-ParticleRad:NiagaraFloat=&Subtract_Float(
     A=$User.CubeRadius
     B=&CalculateParticleRadius(
      Radius Calculation Type:ENiagaraCollisionRadiusOptions=::NewEnumerator0
      Particle Radius:NiagaraFloat=1
      Mesh Dimensions:Vector3f=<1 1 1>
      Method for Calculating Particle Radius:ENiagaraCalculateRadiusOptions=::NewEnumerator2
     )
    )) {isSetParameters=true, module=SetVariables_#2}
   /SetVariables(Particles.RandomPosition:NiagaraFloat=&FloatFromCurve(
     FloatCurve=DataInterface() {propertyValues {Curve {keys [
         <at 0 -1 RCIM_Cubic>
         {arriveTangent=1.1416, interpMode=RCIM_Cubic, leaveTangent=1.1416, tangentMode=RCTM_User, time=0.49, value=-0.8}
         {arriveTangent=1.4558, interpMode=RCIM_Cubic, leaveTangent=1.4558, tangentMode=RCTM_User, time=0.5, value=0.75}
         {tangentMode=RCTM_User, time=1, value=1}
        ]}}}
     CurveIndex=&RandomRangeFloat(Minimum=0, Maximum=1)
     Scale Curve=Linked() {linkedVariable {name=Particles.CubeRad-ParticleRad, type=NiagaraFloat}}
    )) {isSetParameters=true}
   /SetVariables(Particles.Position=&MaskVectorBySpawnGroup(
     Masked Vector=&MakeVector(
      X=0
      Y=&Multiply_Float(
       A=Linked() {linkedVariable {name=Particles.CubeRad-ParticleRad, type=NiagaraFloat}}
       B=&RandomRangeFloat(Minimum=-1, Maximum=1)
      )
      Z=$Particles.RandomPosition
     )
     Passthrough Vector=&MakeVector(
      X=0
      Y=$Particles.RandomPosition
      Z=&Multiply_Float(
       A=Linked() {linkedVariable {name=Particles.CubeRad-ParticleRad, type=NiagaraFloat}}
       B=&RandomRangeFloat(Minimum=-1, Maximum=1)
      )
     )
    )) {isSetParameters=true, module=SetVariables_#3}
   &SpriteFacingAndAlignment(WriteAlignment=true, WriteFacing=true)
   /SetVariables(Particles.Position=&self:GridSnap(Size:NiagaraFloat=$Emitter.RuneSize, StartPosition:Vector3f=Linked() {linkedVariable {name=Particles.Position, type=NiagaraPosition}})) {enabled=false, isSetParameters=true, module=SetVariables_#4}
   &CameraOffset(Camera Offset Amount=8)
  ]
  particleUpdateScript [
   &ParticleState()
   &SubUVAnimation(
    End Frame=4
    Sub UV Play Rate:NiagaraFloat=&RandomRangeFloat(Minimum=0.5, Maximum=2)
    SubUV Animation Mode:ENiagara_SubUVLookupMode=::NewEnumerator3
    SubUV Start Time Offset:NiagaraFloat=&RandomRangeFloat(Minimum=0, Maximum=1)
   )
   &ScaleColor(
    Scale Alpha=&FloatFromCurve(FloatCurve=DataInterface() {propertyValues {Curve {keys [
         {arriveTangent=-3.8362, interpMode=RCIM_Cubic, leaveTangent=-3.8362, tangentMode=RCTM_User, time=0, value=1}
         <at 0.1866 0.4>
         {arriveTangent=-4.1865, interpMode=RCIM_Cubic, leaveTangent=-4.1865, tangentMode=RCTM_User, time=0.2015, value=1}
         <at 0.4552 0.3>
         {arriveTangent=-8.2931, interpMode=RCIM_Cubic, leaveTangent=-8.2931, tangentMode=RCTM_User, time=0.4652, value=0.75}
         <at 0.7089 0>
         {arriveTangent=-5.5334, interpMode=RCIM_Cubic, leaveTangent=-5.5334, tangentMode=RCTM_User, time=0.7338, value=0.5}
         <at 1 0>
        ]}}})
    Scale RGB=&VectorFromCurve(VectorCurve=DataInterface() {propertyValues {
       XCurve {keys [<at 0 1>,<at 1 1>]}
       YCurve {keys [<at 0 1>,<at 1 1>]}
       ZCurve {keys [
         {arriveTangent=1.0471, interpMode=RCIM_Cubic, leaveTangent=1.0471, tangentMode=RCTM_User, time=0, value=0.8}
         {interpMode=RCIM_Cubic, tangentMode=RCTM_User, time=0.25, value=1}
         {arriveTangent=-2.0404, interpMode=RCIM_Cubic, leaveTangent=-2.0404, tangentMode=RCTM_User, time=1, value=0}
        ]}
      }})
    ScaleA=true
    ScaleRGB=true
   )
  ]
  rendererClasses [&NiagaraSpriteRendererProperties]
  renderers [{rendererClass=&NiagaraSpriteRendererProperties}]
 }
 RunesSides {
  emitterSpawnScript [/SetVariables(Emitter.RuneSize=48) {isSetParameters=true, module=SetVariables_#1}]
  emitterUpdateScript [
   &EmitterState(Enable Distance Culling=true, MaxDistance=9000, Module.bUseMaxDistance=true, Module.bUseMinDistance=false, Scalability Mode=::NewEnumerator1)
   &SpawnRate(
    SpawnRate=&FloatFromCurve(
     FloatCurve=DataInterface() {propertyValues {Curve {keys [
         {arriveTangent=10.4941, interpMode=RCIM_Cubic, time=0, value=0}
         {arriveTangent=0.2, arriveTangentWeight=0.0471, interpMode=RCIM_Cubic, leaveTangent=0.2, leaveTangentWeight=0.0707, time=0.3, value=1}
         {arriveTangentWeight=0.3941, interpMode=RCIM_Cubic, leaveTangentWeight=0.9853, tangentMode=RCTM_User, time=0.5, value=0.2}
         <at 1 0.1>
        ]}}}
     CurveIndex=$System.NormalizedLoopAge
     Scale Curve=600
    )
    Use Spawn Probability=false
   )
  ]
  particleSpawnScript [
   &FN_SetSpawnGroup()
   &InitializeParticle(
    Color Maximum=<rgba 1 12 15 0.75>
    Color Minimum=<rgba 0 5 2 0.5>
    Color Mode=::NewEnumerator2
    Lifetime Max=0.75
    Lifetime Min=0.05
    Lifetime Mode=::NewEnumerator1
    Sprite Rotation Angle=&SelectFloatFromArray(
     Array Sampling Mode=::NewEnumerator0
     Direct Array Index=?
     Float Selection Array=DataInterface() {propertyValues {FloatData [0,0.25,0.5,0.75]}}
     Random Seed=0
    )
    Sprite Rotation Mode=::NewEnumerator1
    Sprite Size Mode=::NewEnumerator3
    Uniform Sprite Size=$Emitter.RuneSize
    UsePositionOffset=false
    Write Material Random=true
   )
   /SetVariables(Particles.CubeRad-ParticleRad:NiagaraFloat=&Subtract_Float(
     A=$User.CubeRadius
     B=&CalculateParticleRadius(
      Radius Calculation Type:ENiagaraCollisionRadiusOptions=::NewEnumerator0
      Particle Radius:NiagaraFloat=1
      Mesh Dimensions:Vector3f=<1 1 1>
      Method for Calculating Particle Radius:ENiagaraCalculateRadiusOptions=::NewEnumerator2
     )
    )) {isSetParameters=true, module=SetVariables_#2}
   /SetVariables(Particles.Position=&RotateVectorByQuaternion(
     Vector To Rotate=&MakeVector(
      X=&Subtract_Float(
       A=&FloatFromCurve(
        FloatCurve=DataInterface() {propertyValues {Curve {keys [{arriveTangent=3.4123, interpMode=RCIM_Cubic, leaveTangent=3.4123, tangentMode=RCTM_User, time=0, value=0.25}, <at 0.5 0.99>]}}}
        CurveIndex=$Emitter.NormalizedLoopAge
        Scale Curve=Linked() {linkedVariable {name=Particles.CubeRad-ParticleRad, type=NiagaraFloat}}
       )
       B=$User.CubeRadius
      )
      Y=&Multiply_Float(
       A=Linked() {linkedVariable {name=Particles.CubeRad-ParticleRad, type=NiagaraFloat}}
       B=&RandomRangeFloat(Minimum=-0.99, Maximum=0.99)
      )
      Z=$User.CubeRadius
     )
     Quaternion=&MakeQuatFromAxisAngle(
      Axis=<1 0 0>
      AngleInDegrees=&Multiply_Float(A=&MakeFloatFromInt(Int=$Particles.SpawnGroup), B=90)
     )
    )) {isSetParameters=true, module=SetVariables_#3}
   /SetVariables() {isSetParameters=true}
   /SetVariables(Particles.SpriteFacing=&RotateVectorByQuaternion(
     Vector To Rotate=<0 0 1>
     Quaternion=&MakeQuatFromAxisAngle(
      Axis=<1 0 0>
      AngleInDegrees=&Multiply_Float(A=&MakeFloatFromInt(Int=$Particles.SpawnGroup), B=90)
     )
    )) {isSetParameters=true}
   /SetVariables(Particles.Position=&self:GridSnap(Size:NiagaraFloat=$Emitter.RuneSize, StartPosition:Vector3f=Linked() {linkedVariable {name=Particles.Position, type=NiagaraPosition}})) {isSetParameters=true, module=SetVariables_#4}
   &CameraOffset(Camera Offset Amount=8)
  ]
  particleUpdateScript [
   &ParticleState()
   &SubUVAnimation(
    End Frame=4
    Sub UV Play Rate:NiagaraFloat=&RandomRangeFloat(Minimum=0.5, Maximum=2)
    SubUV Animation Mode:ENiagara_SubUVLookupMode=::NewEnumerator3
    SubUV Start Time Offset:NiagaraFloat=&RandomRangeFloat(Minimum=0, Maximum=1)
   )
   &ScaleColor(
    Scale Alpha=&FloatFromCurve(FloatCurve=DataInterface() {propertyValues {Curve {keys [
         {arriveTangent=-3.8362, interpMode=RCIM_Cubic, leaveTangent=-3.8362, tangentMode=RCTM_User, time=0, value=1}
         <at 0.1866 0.4>
         {arriveTangent=-4.1865, interpMode=RCIM_Cubic, leaveTangent=-4.1865, tangentMode=RCTM_User, time=0.2015, value=1}
         <at 0.4552 0.3>
         {arriveTangent=-8.2931, interpMode=RCIM_Cubic, leaveTangent=-8.2931, tangentMode=RCTM_User, time=0.4652, value=0.75}
         <at 0.7089 0>
         {arriveTangent=-5.5334, interpMode=RCIM_Cubic, leaveTangent=-5.5334, tangentMode=RCTM_User, time=0.7338, value=0.5}
         <at 1 0>
        ]}}})
    Scale RGB=&VectorFromCurve(VectorCurve=DataInterface() {propertyValues {
       XCurve {keys [<at 0 1>,<at 1 1>]}
       YCurve {keys [<at 0 1>,<at 1 1>]}
       ZCurve {keys [
         {arriveTangent=1.0471, interpMode=RCIM_Cubic, leaveTangent=1.0471, tangentMode=RCTM_User, time=0, value=0.8}
         {interpMode=RCIM_Cubic, tangentMode=RCTM_User, time=0.25, value=1}
         {arriveTangent=-2.0404, interpMode=RCIM_Cubic, leaveTangent=-2.0404, tangentMode=RCTM_User, time=1, value=0}
        ]}
      }})
    ScaleA=true
    ScaleRGB=true
   )
  ]
  rendererClasses [&NiagaraSpriteRendererProperties]
  renderers [{rendererClass=&NiagaraSpriteRendererProperties}]
 }
 Sparkles {
  emitterUpdateScript [
   &EmitterState(Enable Distance Culling=true, MaxDistance=6000, Module.bUseMaxDistance=true, Module.bUseMinDistance=false, Scalability Mode=::NewEnumerator1)
   &SpawnRate(SpawnRate=50, Use Spawn Probability=false)
  ]
  particleSpawnScript [
   &InitializeParticle(
    Color=<rgba 0.5 0 0.0812 1>
    Lifetime Max=5
    Lifetime Mode=::NewEnumerator1
    Mesh Scale Mode=::NewEnumerator4
    Sprite Size Mode=::NewEnumerator1
    Uniform Sprite Size Max=25
    Uniform Sprite Size Min=7
    UsePositionOffset=false
    Write Material Random=true
   )
   /SetVariables(Particles.Position=&TransformPosition(
     Position=&MakeVector(
      X=&FloatFromCurve(
       FloatCurve=DataInterface() {propertyValues {Curve {keys [<at 0 0.05 RCIM_Cubic>, {arriveTangent=2.0792, interpMode=RCIM_Cubic, leaveTangent=2.0792, tangentMode=RCTM_User, time=1, value=1}]}}}
       CurveIndex=&RandomRangeFloat(Minimum=0, Maximum=1)
       Scale Curve=900
      )
      Y=&Multiply_Float(A=$User.CubeRadius, B=&RandomRangeFloat(Minimum=-0.9, Maximum=0.9))
      Z=&Multiply_Float(A=$User.CubeRadius, B=&RandomRangeFloat(Minimum=-0.9, Maximum=0.9))
     )
     Source Space=::Local
     Destination Space=::Simulation
    )) {isSetParameters=true}
  ]
  particleUpdateScript [
   &ParticleState()
   &DynamicMaterialParameters(
    Index 0 Param 1=1
    Index 0 Param 2=1
    Index 0 Param 3=0.5
    Module.Param0WriteEnabled=true
    Module.Param1WriteEnabled=true
    Module.Param2WriteEnabled=true
    Module.Param3WriteEnabled=true
   )
   &ScaleColor(
    Scale Alpha=&FloatFromCurve(FloatCurve=DataInterface() {propertyValues {Curve {keys [
         <at 0 0 RCIM_Cubic>
         {interpMode=RCIM_Cubic, tangentMode=RCTM_User, time=0.0856, value=0.9914}
         {arriveTangent=-2.3503, interpMode=RCIM_Cubic, leaveTangent=-2.3503, tangentMode=RCTM_User, time=0.8081, value=0.5934}
         <at 1 0>
        ]}}})
    Scale RGB=&VectorFromFloat(Value=&Add_Float(
      A=1
      B=&FloatFromCurve(
       FloatCurve=DataInterface() {propertyValues {Curve {keys [
           {arriveTangent=10.4941, interpMode=RCIM_Cubic, time=0, value=0}
           {
            arriveTangent=-2.7982
            arriveTangentWeight=0.0471
            interpMode=RCIM_Cubic
            leaveTangent=-2.7982
            leaveTangentWeight=0.0707
            tangentMode=RCTM_User
            time=0.3
            value=1
           }
           <at 1 0>
          ]}}}
       CurveIndex=$System.NormalizedLoopAge
       Scale Curve=5
      )
     ))
    ScaleA=true
    ScaleRGB=true
   )
   /SetVariables(Particles.Color2=&MakeLinearColorFromVectorAndFloat(
     "Vector (RGB)"=&MakeVectorFromLinearColorRGB(Color=<rgba 0 1 0.8193 1>)
     "Float (Alpha)"=&MakeFloatFromLinearColor(LinearColor=$Particles.Color, Channel:ENiagaraLinearColor_Channels=::NewEnumerator3)
    )) {isSetParameters=true}
   &GravityForce(Gravity=&RandomRangeVector(Minimum=<-1 -1 250>, Maximum=<1 1 500>))
   &VectorNoiseForce(
    Force Amount=&FloatFromCurve(
     FloatCurve=DataInterface() {propertyValues {Curve {keys [
         {arriveTangent=10.4941, interpMode=RCIM_Cubic, time=0, value=0.1}
         {arriveTangent=0.4, arriveTangentWeight=0.0471, interpMode=RCIM_Cubic, leaveTangent=0.4, leaveTangentWeight=0.0707, time=0.3, value=1}
         {arriveTangentWeight=0.3941, leaveTangentWeight=0.9853, time=0.5, value=0.35}
         <at 1 0.1>
        ]}}}
     CurveIndex=&Add_Float(A=$Emitter.NormalizedLoopAge, B=&RandomRangeFloat(Minimum=-0.1, Maximum=0.25))
     Scale Curve=20000
    )
    Module.Use Random Noise Time Interval=true
    Random Noise Time Interval=0.05
   )
   &Drag(Drag=50, Use Linear Drag=true, Use Rotational Drag=false)
   &ScaleSpriteSize(Scale Factor=&Vector2DFromFloat(Value=&FloatFromCurve(FloatCurve=DataInterface() {propertyValues {Curve {keys [
          <at 0 0 RCIM_Cubic>
          {arriveTangent=0.836, interpMode=RCIM_Cubic, leaveTangent=0.836, time=0.2, value=1}
          {arriveTangent=-1.8455, interpMode=RCIM_Cubic, leaveTangent=-1.8455, tangentMode=RCTM_User, time=0.6507, value=0.544}
          <at 1 0>
         ]}}})))
   &SolveForcesAndVelocity(Clamp Velocity=false, Limit Acceleration=false)
  ]
  rendererClasses [&NiagaraSpriteRendererProperties]
  renderers [
   {rendererClass=&NiagaraSpriteRendererProperties}
   {rendererClass=&NiagaraSpriteRendererProperties, rendererIndex=1}
  ]
 }
}
summary {
 systemName=NS_Corruption_Awakening_Hover
 emitters [
  {emitterName=Sparkles, rendererClasses [&NiagaraSpriteRendererProperties]}
  {emitterName=RunesSides, rendererClasses [&NiagaraSpriteRendererProperties]}
  {
   emitterName=NE_FN_DirtChunks
   rendererClasses [&NiagaraMeshRendererProperties, &NiagaraSpriteRendererProperties]
  }
  {emitterName=RunesBottom, rendererClasses [&NiagaraSpriteRendererProperties]}
  {emitterName=Glow, rendererClasses [&NiagaraMeshRendererProperties, &NiagaraLightRendererProperties]}
  {
   emitterName=Lightning
   rendererClasses [
    &NiagaraMeshRendererProperties
    &NiagaraSpriteRendererProperties
    &NiagaraLightRendererProperties
   ]
  }
 ]
 userVariables [
  User.CubeRadius {type=NiagaraFloat, defaultValue=768, description=""}
  User.EndPoint {type=Vector3f, description="", defaultValue {struct=Vector3f, value=<0 0 0>}}
 ]
}
