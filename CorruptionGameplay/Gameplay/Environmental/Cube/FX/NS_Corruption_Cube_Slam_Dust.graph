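# Niagara dust and debris effect for a Corruption cube slam, sized by User.CubeRadius and a derived CubeRadiusScale.
# FastSmoke and DirtChunks map MaterialRandom through vector curves to spawn along a square outline and push particles outward.
# LingerSmoke spawns longer-lived, rising smoke on a cylinder surface, with spawning gated by radius thresholds.
# DirtChunks adds tumbling mesh and sprite debris under gravity, with delayed collision limited to one spawn group.
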NiagaraSystem /CorruptionGameplay/Gameplay/Environmental/Cube/FX/NS_Corruption_Cube_Slam_Dust
dependencies {
 usedDataInterfaces [&NiagaraDataInterfaceVectorCurve, &NiagaraDataInterfaceCurve]
 usedDynamicInputs [&RemapRange]
 usedModules [
  &self:SystemSpawnScript.SystemScriptSource.SystemScriptGraph.NiagaraNodeAssignment_0.SetVariables_#1
  &SystemState
 ]
}
emitters {
 FastSmoke {
  emitterUpdateScript [
   &EmitterState()
   &SpawnBurst_Instantaneous(Spawn Count=50, Spawn Time=0.04, Use Spawn Probability=false, UseLoopCountLimit=false) {module=SpawnBurst_Instantaneous004}
   &SpawnRate(SpawnRate=&Multiply_Float(A=100, B=$System.CubeRadiusScale), Use Spawn Probability=false)
  ]
  particleSpawnScript [
   /Niagara/Modules/Spawn/Initialization/InitializeParticle(
    Color=&UniformRangedLinearColor(Minimum=<rgba 0.125 0.1036 0.117 0.5>, Maximum=<rgba 0.2344 0.2074 0.2097 0.75>)
    Lifetime=&RandomRangeFloat(Minimum=0.8, Maximum=1.5)
    Module.Write Color=true
    Module.Write Lifetime=true
    Module.Write Mass=false
    Module.Write Material Random=true
    Module.Write Position=false
    Module.Write Scale=false
    Module.Write SpriteRotation=true
    Module.Write SpriteSize=true
    Sprite Rotation=&UniformRangedFloat(Minimum=0, Maximum=360)
    Sprite Size=&Vector2DFromFloat(Value=&Multiply_Float(A=&RandomRangeFloat(Minimum=500, Maximum=700), B=$System.CubeRadiusScale))
   )
   /SetVariables(Particles.Position=/Niagara/DynamicInputs/Transforms/TransformPosition(
     Position=&VectorFromCurve(
      VectorCurve=DataInterface() {propertyValues {
        XCurve {keys [{interpMode=RCIM_Cubic, leaveTangentWeight=0.0833, tangentMode=RCTM_User, time=0, value=0}]}
        YCurve {keys [
          {interpMode=RCIM_Cubic, tangentMode=RCTM_User, time=0, value=1}
          {interpMode=RCIM_Cubic, leaveTangentWeight=0.6719, tangentMode=RCTM_User, time=0, value=1}
          {arriveTangentWeight=0.3436, interpMode=RCIM_Cubic, leaveTangentWeight=0.3436, tangentMode=RCTM_User, time=0.25, value=-1}
          {arriveTangentWeight=0.3436, interpMode=RCIM_Cubic, leaveTangentWeight=0.3436, tangentMode=RCTM_User, time=0.5, value=-1}
          {arriveTangentWeight=0.0833, interpMode=RCIM_Cubic, tangentMode=RCTM_User, time=0.75, value=1}
         ]}
        ZCurve {keys [
          {interpMode=RCIM_Cubic, leaveTangentWeight=0.0833, tangentMode=RCTM_User, time=0.25, value=1}
          {arriveTangentWeight=0.3436, interpMode=RCIM_Cubic, leaveTangentWeight=0.3436, tangentMode=RCTM_User, time=0.5, value=-1}
          {arriveTangentWeight=0.3436, interpMode=RCIM_Cubic, leaveTangentWeight=0.3436, tangentMode=RCTM_User, time=0.75, value=-1}
          {arriveTangentWeight=0.0833, interpMode=RCIM_Cubic, tangentMode=RCTM_User, time=1, value=1}
         ]}
       }}
      CurveIndex=$Particles.MaterialRandom
      Scale Curve=&VectorFromFloat(Value=&Multiply_Float(A=$User.CubeRadius, B=0.7))
     )
     Source Space=::Local
     Destination Space=::Simulation
    )) {isSetParameters=true}
   &AddVelocity(Velocity=&VectorFromCurve(
     VectorCurve=DataInterface() {propertyValues {
       XCurve {keys [{interpMode=RCIM_Cubic, leaveTangentWeight=0.0833, tangentMode=RCTM_User, time=0, value=-0.5}]}
       YCurve {keys [
         {interpMode=RCIM_Cubic, tangentMode=RCTM_User, time=0, value=1}
         {interpMode=RCIM_Cubic, leaveTangentWeight=0.6719, tangentMode=RCTM_User, time=0, value=1}
         {arriveTangentWeight=0.3436, interpMode=RCIM_Cubic, leaveTangentWeight=0.3436, tangentMode=RCTM_User, time=0.25, value=-1}
         {arriveTangentWeight=0.3436, interpMode=RCIM_Cubic, leaveTangentWeight=0.3436, tangentMode=RCTM_User, time=0.5, value=-1}
         {arriveTangentWeight=0.0833, interpMode=RCIM_Cubic, tangentMode=RCTM_User, time=0.75, value=1}
        ]}
       ZCurve {keys [
         {interpMode=RCIM_Cubic, leaveTangentWeight=0.0833, tangentMode=RCTM_User, time=0.25, value=1}
         {arriveTangentWeight=0.3436, interpMode=RCIM_Cubic, leaveTangentWeight=0.3436, tangentMode=RCTM_User, time=0.5, value=-1}
         {arriveTangentWeight=0.3436, interpMode=RCIM_Cubic, leaveTangentWeight=0.3436, tangentMode=RCTM_User, time=0.75, value=-1}
         {arriveTangentWeight=0.0833, interpMode=RCIM_Cubic, tangentMode=RCTM_User, time=1, value=1}
        ]}
      }}
     CurveIndex=$Particles.MaterialRandom
     Scale Curve=&VectorFromFloat(Value=&Multiply_Float(A=$User.CubeRadius, B=&RandomRangeFloat(Minimum=2, Maximum=3)))
    ))
   &AddVelocityFromPoint(Module.Use Falloff Distance=false, Velocity Strength=&RandomRangeFloat(Minimum=1200, Maximum=3000))
  ]
  particleUpdateScript [
   &ParticleState()
   &Drag(Drag=&RandomRangeFloat(Minimum=5, Maximum=7), Use Linear Drag=true, Use Rotational Drag=false)
   &SpriteRotationRate(Rotation Rate=&Multiply_Float(
     A=&Lerp_Float(A=-300, B=300, Alpha=$Particles.MaterialRandom)
     B=&FloatFromCurve(FloatCurve=DataInterface() {propertyValues {Curve {keys [
          {arriveTangent=-4.4572, interpMode=RCIM_Cubic, leaveTangent=-4.4572, tangentMode=RCTM_User, time=0, value=1}
          {arriveTangent=-1.0628, interpMode=RCIM_Cubic, leaveTangent=-1.0628, tangentMode=RCTM_User, time=0.2, value=0.4}
          <at 1.0011 0.01>
         ]}}})
    ))
   &AccelerationForce(Acceleration=<0 0 40>)
   &SolveForcesAndVelocity(Clamp Velocity=false, Limit Acceleration=false)
   &CameraOffset(Camera Offset Amount=&Multiply_Float(
     A=0.5
     B=/Niagara/DynamicInputs/Math/CalculateParticleRadius(
      Radius Calculation Type:ENiagaraCollisionRadiusOptions=::NewEnumerator0
      Particle Radius:NiagaraFloat=1
      Mesh Dimensions:Vector3f=<1 1 1>
      Method for Calculating Particle Radius:ENiagaraCalculateRadiusOptions=::NewEnumerator2
     )
    ))
   &ScaleSpriteSize(Scale Factor=&Vector2DFromFloat(Value=&FloatFromCurve(FloatCurve=DataInterface() {propertyValues {Curve {keys [
          {arriveTangent=2.7535, interpMode=RCIM_Cubic, leaveTangent=2.7535, tangentMode=RCTM_User, time=0, value=0.25}
          {arriveTangent=0.5339, interpMode=RCIM_Cubic, leaveTangent=-0.3011, tangentMode=RCTM_User, time=1, value=1.25}
         ]}}})))
   &ScaleColor(
    Scale Alpha=&FloatFromCurve(FloatCurve=DataInterface() {propertyValues {Curve {keys [
         {arriveTangent=8.5887, interpMode=RCIM_Cubic, leaveTangent=8.5887, tangentMode=RCTM_User, time=0, value=0}
         {arriveTangent=-1.4353, interpMode=RCIM_Cubic, leaveTangent=-1.4353, tangentMode=RCTM_User, time=0.3, value=1}
         {interpMode=RCIM_Cubic, leaveTangent=-3.4295, time=1, value=0}
        ]}}})
    Scale RGB=&VectorFromFloat(Value=&FloatFromCurve(
      FloatCurve=DataInterface() {propertyValues {Curve {keys [{arriveTangent=-11.0936, interpMode=RCIM_Cubic, leaveTangent=-11.0936, tangentMode=RCTM_User, time=0, value=3}, <at 0.4 1>]}}}
      CurveIndex=$Particles.Age
     ))
    ScaleA=true
    ScaleRGB=true
   )
   /SetVariables(Particles.SubImageIndex=&Lerp_Float(A=20, B=63, Alpha=$Particles.NormalizedAge)) {isSetParameters=true, module=SetVariables_#2}
  ]
  rendererClasses [&NiagaraSpriteRendererProperties]
  renderers [{rendererClass=&NiagaraSpriteRendererProperties}]
 }
 LingerSmoke001 {
  emitterUpdateScript [
   &EmitterState()
   &SpawnBurst_Instantaneous(
    Spawn Count=&MakeCustomIntFromBool(Bool=&SetBoolByFloatComparison(A=$User.CubeRadius, B=300), True INT=15, False INT=0)
    Use Spawn Probability=false
    UseLoopCountLimit=false
   ) {module=SpawnBurst_Instantaneous004}
   &SpawnRate(
    SpawnRate=&MakeCustomFloatFromBool(Bool=&SetBoolByFloatComparison(A=$User.CubeRadius, B=600), True Float=150, False Float=0)
    Use Spawn Probability=false
   )
  ]
  particleSpawnScript [
   &InitializeParticle(
    Color=&UniformRangedLinearColor(Minimum=<rgba 0.12 0.0995 0.1123 0.3>, Maximum=<rgba 0.22 0.1947 0.1969 0.55>)
    Lifetime=&RandomRangeFloat(Minimum=5, Maximum=8)
    Module.Write Color=true
    Module.Write Lifetime=true
    Module.Write Mass=false
    Module.Write Material Random=true
    Module.Write Position=false
    Module.Write Scale=false
    Module.Write SpriteRotation=true
    Module.Write SpriteSize=true
    Sprite Rotation=&UniformRangedFloat(Minimum=0, Maximum=360)
    Sprite Size=&Vector2DFromFloat(Value=&Multiply_Float(A=&RandomRangeFloat(Minimum=1200, Maximum=1600), B=$System.CubeRadiusScale))
   )
   &CylinderLocation(
    Cylinder Radius=&Multiply_Float(A=$User.CubeRadius, B=0.5)
    MaskSpawnGroup=false
    Override Seed=false
    Rotation Mode:ENiagara_RotationMode=::NewEnumerator1
    Surface Only=true
    Use Endcaps In Surface Only Mode=false
    UseOwnerScale=true
   )
   &AddVelocityFromPoint(Module.Use Falloff Distance=false, Velocity Strength=&RandomRangeFloat(Minimum=3000, Maximum=3500))
  ]
  particleUpdateScript [
   &ParticleState()
   &Drag(
    Drag=&FloatFromCurve(
     FloatCurve=DataInterface() {propertyValues {Curve {keys [
         {arriveTangent=4.6712, interpMode=RCIM_Cubic, leaveTangent=4.6712, tangentMode=RCTM_User, time=0, value=0.2}
         {interpMode=RCIM_Cubic, tangentMode=RCTM_User, time=0.5, value=1}
        ]}}}
     Scale Curve=8
    )
    Use Linear Drag=true
    Use Rotational Drag=false
   ) {enabled=false}
   &SpriteRotationRate(Rotation Rate=&Multiply_Float(
     A=&Lerp_Float(A=-200, B=200, Alpha=$Particles.MaterialRandom)
     B=&FloatFromCurve(FloatCurve=DataInterface() {propertyValues {Curve {keys [
          {arriveTangent=-4.4572, interpMode=RCIM_Cubic, time=0, value=1}
          {arriveTangent=-2.0073, interpMode=RCIM_Cubic, leaveTangent=-2.0073, tangentMode=RCTM_User, time=0.25, value=0.25}
          <at 0.5 0.01>
         ]}}})
    ))
   &AccelerationForce(Acceleration=<0 0 3000>)
   &ScaleVelocity(Velocity Scale=&VectorFromFloat(Value=&FloatFromCurve(
      FloatCurve=DataInterface() {propertyValues {Curve {keys [
          {arriveTangent=-12.3491, interpMode=RCIM_Cubic, time=0.01, value=1}
          {arriveTangent=-27.4843, interpMode=RCIM_Cubic, leaveTangent=-27.4843, tangentMode=RCTM_User, time=0.03, value=0.6818}
          {arriveTangent=-2.0974, interpMode=RCIM_Cubic, leaveTangent=-2.0974, tangentMode=RCTM_User, time=0.07, value=0.15}
          {interpMode=RCIM_Cubic, tangentMode=RCTM_User, time=0.22, value=0.02}
          <at 0.5 0.01>
         ]}}}
      CurveIndex=&Multiply_Float(A=$Particles.NormalizedAge, B=0.5)
     )))
   &SolveForcesAndVelocity(Clamp Velocity=false, Limit Acceleration=false)
   &CameraOffset(Camera Offset Amount=&Multiply_Float(
     A=0.5
     B=&CalculateParticleRadius(
      Radius Calculation Type:ENiagaraCollisionRadiusOptions=::NewEnumerator0
      Particle Radius:NiagaraFloat=1
      Mesh Dimensions:Vector3f=<1 1 1>
      Method for Calculating Particle Radius:ENiagaraCalculateRadiusOptions=::NewEnumerator2
     )
    ))
   &ScaleSpriteSize(Scale Factor=&Vector2DFromFloat(Value=&FloatFromCurve(FloatCurve=DataInterface() {propertyValues {Curve {keys [
          {arriveTangent=1.8989, interpMode=RCIM_Cubic, leaveTangent=1.8989, tangentMode=RCTM_User, time=0, value=0.35}
          {arriveTangent=0.2115, interpMode=RCIM_Cubic, leaveTangent=0.2115, tangentMode=RCTM_User, time=1, value=1.25}
         ]}}})))
   &ScaleColor(
    Scale Alpha=&FloatFromCurve(FloatCurve=DataInterface() {propertyValues {Curve {keys [
         {arriveTangent=10.8332, interpMode=RCIM_Cubic, leaveTangent=10.8332, tangentMode=RCTM_User, time=0, value=0}
         {arriveTangent=-1.727, interpMode=RCIM_Cubic, leaveTangent=-1.727, tangentMode=RCTM_User, time=0.15, value=1}
         {interpMode=RCIM_Cubic, leaveTangent=-3.4295, time=1, value=0}
        ]}}})
    ScaleA=true
    ScaleRGB=false
   )
   /SetVariables(Particles.SubImageIndex=&Lerp_Float(A=10, B=63, Alpha=$Particles.NormalizedAge)) {isSetParameters=true, module=SetVariables_#2}
  ]
  rendererClasses [&NiagaraSpriteRendererProperties]
  renderers [{rendererClass=&NiagaraSpriteRendererProperties}]
 }
 NE_FN_DirtChunks {
  emitterUpdateScript [
   &EmitterState(Enable Distance Culling=true, MaxDistance=8000, Module.bUseMaxDistance=true, Module.bUseMinDistance=false, Scalability Mode=::NewEnumerator1)
   &SpawnRate(
    SpawnRate=&MakeCustomFloatFromBool(Bool=&SetBoolByFloatComparison(A=$User.CubeRadius, B=600), True Float=75, False Float=0)
    Use Spawn Probability=false
   )
   &SpawnBurst_Instantaneous(
    Spawn Count=&MakeIntFromFloat(Float=&Multiply_Float(A=75, B=$System.CubeRadiusScale))
    Use Spawn Probability=false
    UseLoopCountLimit=false
   )
   &SpawnBurst_Instantaneous(
    Spawn Count=&MakeIntFromFloat(Float=&Multiply_Float(A=30, B=$System.CubeRadiusScale))
    Spawn Group=1
    Spawn Time=0.02
    Use Spawn Probability=false
    UseLoopCountLimit=false
   ) {module=SpawnBurst_Instantaneous001}
  ]
  particleSpawnScript [
   /Niagara/Modules/Spawn/Initialization/V2/InitializeParticle(
    Color=<rgba 2 0.665 0.0366 0.25>
    Lifetime Mode=::NewEnumerator1
    Mass Mode=::NewEnumerator3
    Mesh Scale=&Multiply_VectorByFloat(
     Vector=&UniformRangedVector(Minimum=<0.2 0.4 0.4>, Maximum=<0.8 1 1.1>)
     Float=&Multiply_Float(A=&UniformRangedFloat(Minimum=0.1, Maximum=0.2), B=$System.CubeRadiusScale)
    )
    Mesh Scale Mode=::NewEnumerator4
    Sprite Size Max=<25 25>
    Sprite Size Min=<10 10>
    Sprite Size Mode=::NewEnumerator2
    UsePositionOffset=false
    Write Material Random=true
   )
   /SetVariables(Particles.VisibilityTag=&SpawnGroup()) {isSetParameters=true}
   /SetVariables(Particles.RockColor:LinearColor=<rgba 0 0 0 1>) {isSetParameters=true}
   &InitialMeshOrientation(Orientation Vector=&RandomVector(Vector Scale=1), Use Orientation Vector=true, Use Rotation Vector=false)
   /SetVariables(Particles.Position=&TransformPosition(
     Position=&VectorFromCurve(
      VectorCurve=DataInterface() {propertyValues {
        XCurve {keys [{interpMode=RCIM_Cubic, leaveTangentWeight=0.0833, tangentMode=RCTM_User, time=0, value=0}]}
        YCurve {keys [
          {interpMode=RCIM_Cubic, tangentMode=RCTM_User, time=0, value=1}
          {interpMode=RCIM_Cubic, leaveTangentWeight=0.6719, tangentMode=RCTM_User, time=0, value=1}
          {arriveTangentWeight=0.3436, interpMode=RCIM_Cubic, leaveTangentWeight=0.3436, tangentMode=RCTM_User, time=0.25, value=-1}
          {arriveTangentWeight=0.3436, interpMode=RCIM_Cubic, leaveTangentWeight=0.3436, tangentMode=RCTM_User, time=0.5, value=-1}
          {arriveTangentWeight=0.0833, interpMode=RCIM_Cubic, tangentMode=RCTM_User, time=0.75, value=1}
         ]}
        ZCurve {keys [
          {interpMode=RCIM_Cubic, leaveTangentWeight=0.0833, tangentMode=RCTM_User, time=0.25, value=1}
          {arriveTangentWeight=0.3436, interpMode=RCIM_Cubic, leaveTangentWeight=0.3436, tangentMode=RCTM_User, time=0.5, value=-1}
          {arriveTangentWeight=0.3436, interpMode=RCIM_Cubic, leaveTangentWeight=0.3436, tangentMode=RCTM_User, time=0.75, value=-1}
          {arriveTangentWeight=0.0833, interpMode=RCIM_Cubic, tangentMode=RCTM_User, time=1, value=1}
         ]}
       }}
      CurveIndex=$Particles.MaterialRandom
      Scale Curve=&VectorFromFloat(Value=$User.CubeRadius)
     )
     Source Space=::Local
     Destination Space=::Simulation
    )) {isSetParameters=true}
   &AddVelocity(Velocity=&Multiply_VectorByFloat(
     Vector=&Add_Vector(
      A=&VectorFromCurve(
       VectorCurve=DataInterface() {propertyValues {
         XCurve {keys [{interpMode=RCIM_Cubic, leaveTangentWeight=0.0833, tangentMode=RCTM_User, time=0, value=-0.5}]}
         YCurve {keys [
           {interpMode=RCIM_Cubic, tangentMode=RCTM_User, time=0, value=1}
           {interpMode=RCIM_Cubic, leaveTangentWeight=0.6719, tangentMode=RCTM_User, time=0, value=1}
           {arriveTangentWeight=0.3436, interpMode=RCIM_Cubic, leaveTangentWeight=0.3436, tangentMode=RCTM_User, time=0.25, value=-1}
           {arriveTangentWeight=0.3436, interpMode=RCIM_Cubic, leaveTangentWeight=0.3436, tangentMode=RCTM_User, time=0.5, value=-1}
           {arriveTangentWeight=0.0833, interpMode=RCIM_Cubic, tangentMode=RCTM_User, time=0.75, value=1}
          ]}
         ZCurve {keys [
           {interpMode=RCIM_Cubic, leaveTangentWeight=0.0833, tangentMode=RCTM_User, time=0.25, value=1}
           {arriveTangentWeight=0.3436, interpMode=RCIM_Cubic, leaveTangentWeight=0.3436, tangentMode=RCTM_User, time=0.5, value=-1}
           {arriveTangentWeight=0.3436, interpMode=RCIM_Cubic, leaveTangentWeight=0.3436, tangentMode=RCTM_User, time=0.75, value=-1}
           {arriveTangentWeight=0.0833, interpMode=RCIM_Cubic, tangentMode=RCTM_User, time=1, value=1}
          ]}
        }}
       CurveIndex=$Particles.MaterialRandom
       Scale Curve=&VectorFromFloat(Value=&RandomRangeFloat(Minimum=200, Maximum=700))
      )
      B=&RandomRangeVector(Minimum=<-1500 -10 -10>, Maximum=<-500 10 10>)
     )
     Float=$System.CubeRadiusScale
    )) {module=AddVelocity001}
   &AddVelocityFromPoint(
    Module.Use Falloff Distance=false
    Velocity Strength=&Multiply_Float(A=&RandomRangeFloat(Minimum=0.4, Maximum=0.7), B=$User.CubeRadius)
   )
   &AddRotationalVelocity(Rotation Rate=&RandomVector(Vector Scale=&RandomRangeFloat(Minimum=-2, Maximum=2)))
   &ApplyInitialForces()
   &DynamicMaterialParameters(Module.Param0WriteEnabled=true, Module.Param1WriteEnabled=false, Module.Param2WriteEnabled=false, Module.Param3WriteEnabled=false)
   &SubUVAnimation(End Frame=3, SubUV Animation Mode:ENiagara_SubUVLookupMode=::NewEnumerator2) {module=SubUVAnimation001}
  ]
  particleUpdateScript [
   &ParticleState()
   /SetVariables(Particles.GlowDynParams:Vector4f=&MakeVector4(
     X=&FloatFromCurve(FloatCurve=DataInterface() {propertyValues {Curve {keys [
          {arriveTangent=10.4941, interpMode=RCIM_Cubic, time=0, value=0.7}
          {
           arriveTangent=1.4377
           arriveTangentWeight=0.0471
           interpMode=RCIM_Cubic
           leaveTangent=1.4377
           leaveTangentWeight=0.0707
           tangentMode=RCTM_User
           time=0.3
           value=0.1
          }
          <at 0.7 0.7>
         ]}}})
     Y=&RandomRangeFloat(Minimum=1, Maximum=2)
     Z=0
     W=0
    )) {isSetParameters=true}
   &ScaleColor(
    Scale Alpha=&FloatFromCurve(FloatCurve=DataInterface() {propertyValues {Curve {keys [
         {arriveTangent=10.4941, interpMode=RCIM_Cubic, time=0, value=0.05}
         {arriveTangent=0.9, arriveTangentWeight=0.0471, interpMode=RCIM_Cubic, leaveTangent=0.9, leaveTangentWeight=0.0707, time=0.3, value=1}
         {
          arriveTangent=-1.3571
          arriveTangentWeight=0.3941
          interpMode=RCIM_Cubic
          leaveTangent=-1.3571
          leaveTangentWeight=0.9853
          time=0.5
          value=0.5
         }
         <at 1 0.05>
        ]}}})
    ScaleA=true
    ScaleRGB=false
   )
   &Collision(
    Age Colliding Particles=true
    CPU Collision Trace Channel=ECC_WorldStatic
    Collision Enabled=&Boolean_AndOperation(
     Bool1=&SetBoolByIntComparison(A=&SpawnGroup(), B=0, Comparison Type=::NewEnumerator0)
     Bool2=&SetBoolByFloatComparison(A=$Particles.Age, B=0.12)
    )
    Friction=1
    Friction During a Bounce=0.25
    Max CPU Trace Length=100
    Module.Control Rest State Via Surface Normal=false
    Module.EnableMaxCollisionCount=false
    Module.Randomize Collision Normal Vector=true
    Randomize Collision Normal=0.25
    Restitution=0.75
   )
   &GravityForce(Gravity=<0 0 -2200>)
   &Drag(Drag=&RandomRangeFloat(Minimum=0.4, Maximum=0.75), Use Linear Drag=true, Use Rotational Drag=false)
   &SolveForcesAndVelocity(Clamp Velocity=false, Limit Acceleration=false)
   &SolveRotationalForcesAndVelocity()
   &ScaleMeshSize(Scale Factor=&VectorFromFloat(Value=&FloatFromCurve(FloatCurve=DataInterface() {propertyValues {Curve {keys [
          {interpMode=RCIM_Cubic, tangentMode=RCTM_User, time=0, value=0}
          {interpMode=RCIM_Cubic, tangentMode=RCTM_User, time=0.2, value=1}
          {arriveTangentWeight=0.3202, interpMode=RCIM_Cubic, leaveTangentWeight=0.1067, tangentMode=RCTM_User, time=0.8, value=1}
          {tangentMode=RCTM_User, time=1, value=0}
         ]}}})))
   &ScaleSpriteSize(Scale Factor=&Vector2DFromFloat(Value=&FloatFromCurve(FloatCurve=DataInterface() {propertyValues {Curve {keys [
          {interpMode=RCIM_Cubic, tangentMode=RCTM_User, time=0, value=0}
          {interpMode=RCIM_Cubic, tangentMode=RCTM_User, time=0.2, value=1}
          {arriveTangentWeight=0.3202, interpMode=RCIM_Cubic, leaveTangentWeight=0.1067, tangentMode=RCTM_User, time=0.8, value=1}
          {tangentMode=RCTM_User, time=1, value=0}
         ]}}})))
   &SpriteRotationRate(Rotation Rate=&Lerp_Float(A=-70, B=70, Alpha=$Particles.Initial.MaterialRandom))
  ]
  rendererClasses [&NiagaraMeshRendererProperties, &NiagaraSpriteRendererProperties]
  renderers [
   {rendererClass=&NiagaraMeshRendererProperties, rendererIndex=0}
   {rendererClass=&NiagaraSpriteRendererProperties, rendererIndex=1}
  ]
 }
}
summary {
 systemName=NS_Corruption_Cube_Slam_Dust
 emitters [
  {emitterName=FastSmoke, rendererClasses [&NiagaraSpriteRendererProperties]}
  {
   emitterName=NE_FN_DirtChunks
   rendererClasses [&NiagaraMeshRendererProperties, &NiagaraSpriteRendererProperties]
  }
  {emitterName=LingerSmoke001, rendererClasses [&NiagaraSpriteRendererProperties]}
 ]
 userVariables [User.CubeRadius {type=NiagaraFloat, defaultValue=768, description=""}]
}
