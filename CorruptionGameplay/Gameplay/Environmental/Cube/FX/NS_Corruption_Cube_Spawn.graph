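# Niagara system for the spawn effect of a Corruption Cube, scaled by User.CubeRadius.
# Center lightning, GPU sparks and smoke open the effect, then delayed bursts add the cube mesh and a glow with a light.
# Curve-driven spawn rates time corner lightning, shimmers, sparkles and runes, with runes rotated onto cube faces by spawn group.
# Several emitters use distance culling, while center lightning and sparks enlarge with LOD distance.
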NiagaraSystem /CorruptionGameplay/Gameplay/Environmental/Cube/FX/NS_Corruption_Cube_Spawn
dependencies {
 usedDataInterfaces [&NiagaraDataInterfaceCurve, &NiagaraDataInterfaceVectorCurve, &NiagaraDataInterfaceColorCurve]
 usedDynamicInputs []
 usedModules [
  &self:SystemSpawnScript.SystemScriptSource.SystemScriptGraph.NiagaraNodeAssignment_0.SetVariables_#1
  &self:SystemSpawnScript.SystemScriptSource.SystemScriptGraph.NiagaraNodeAssignment_1.SetVariables_#2
  &SystemState
 ]
}
emitters {
 CornerLightning {
  emitterUpdateScript [
   &EmitterState(
    Enable Distance Culling=true
    MaxDistance=12000
    Module.bUseMaxDistance=true
    Module.bUseMinDistance=false
    Scalability Mode=::NewEnumerator1
   )
   &SpawnBurst_Instantaneous(Spawn Time=2, Use Spawn Probability=false, UseLoopCountLimit=false)
   &SpawnBurst_Instantaneous(Spawn Count=2, Spawn Time=3.25, Use Spawn Probability=false, UseLoopCountLimit=false) {module=SpawnBurst_Instantaneous001}
   &SpawnBurst_Instantaneous(Spawn Count=2, Spawn Time=4, Use Spawn Probability=false, UseLoopCountLimit=false) {module=SpawnBurst_Instantaneous002}
   &SpawnRate(
    SpawnRate=&FloatFromCurve(
     FloatCurve=DataInterface() {propertyValues {Curve {keys [
         {interpMode=RCIM_Cubic, tangentMode=RCTM_User, time=0.1, value=0}
         {arriveTangent=2.5217, interpMode=RCIM_Cubic, leaveTangent=2.5217, tangentMode=RCTM_User, time=0.7, value=1}
         {arriveTangent=-2.354, interpMode=RCIM_Cubic, leaveTangent=-2.354, tangentMode=RCTM_User, time=0.7506, value=0.3698}
         {interpMode=RCIM_Cubic, tangentMode=RCTM_User, time=1, value=0}
        ]}}}
     CurveIndex=$System.NormalizedLoopAge
     Scale Curve=40
    )
    Use Spawn Probability=false
   )
  ]
  particleSpawnScript [
   /SetVariables(Particles.BiasedRandom=&FloatFromCurve(
     FloatCurve=DataInterface() {propertyValues {Curve {keys [
         {interpMode=RCIM_Cubic, tangentMode=RCTM_User, time=0, value=0}
         {arriveTangent=2.5197, interpMode=RCIM_Cubic, leaveTangent=2.5197, tangentMode=RCTM_User, time=0.6, value=0.6}
         {interpMode=RCIM_Cubic, tangentMode=RCTM_User, time=1, value=1}
        ]}}}
     CurveIndex=&RandomRangeFloat(Minimum=0, Maximum=1)
    )) {isSetParameters=true}
   /Niagara/Modules/Spawn/Initialization/InitializeParticle(
    Color=<rgba 0.5 0.5 0.5 1>
    Lifetime=&Lerp_Float(A=0.075, B=0.3, Alpha=$Particles.BiasedRandom)
    Mesh Scale=&MakeVector(
     X=&Multiply_Float(
      A=&Divide_Float(A=$User.CubeRadius, B=1000)
      B=&RandomRangeFloat(Minimum=1, Maximum=2)
     )
     Y=&RandomRangeFloat(Minimum=0.2, Maximum=1)
     Z=&RandomRangeFloat(Minimum=0.2, Maximum=1)
    )
    Module.Write Color=true
    Module.Write Lifetime=true
    Module.Write Mass=true
    Module.Write Material Random=true
    Module.Write Position=false
    Module.Write Scale=true
    Module.Write SpriteRotation=false
    Module.Write SpriteSize=true
    Position=&SimulationPosition()
    Sprite Size=&Vector2DFromFloat(Value=&RandomRangeFloat(Minimum=300, Maximum=700))
   )
   /SetVariables(Particles.Position=&MakeVector(
     X=0
     Y=&Multiply_Float(A=$User.CubeRadius, B=&UniformAOrBFloat(A=-1))
     Z=&Multiply_Float(A=$User.CubeRadius, B=&UniformAOrBFloat(A=-1))
    )) {isSetParameters=true}
   &InitialMeshOrientation(
    Mesh Orientation Mode=::NewEnumerator2
    Orientation Vector=<-1 0 0>
    Rotation=&RandomRangeVector(Minimum=<0 0 0>, Maximum=<1 0 0>)
    Use Rotation Vector=true
   ) {module=InitialMeshOrientation001}
   /SetVariables(Particles.MeshIndex=&RandomRangeInt(Minimum=0, Maximum=1)) {isSetParameters=true}
   /SetVariables(Particles.DynParamsSprite:Vector4f=<0.25 2 1 1>) {isSetParameters=true}
   &CameraOffset(Camera Offset Amount=55)
  ]
  particleUpdateScript [
   &ParticleState()
   &ScaleColor(
    Scale Alpha=&FloatFromCurve(FloatCurve=DataInterface() {propertyValues {Curve {keys [
         <at 0 0>
         {interpMode=RCIM_Cubic, tangentMode=RCTM_User, time=0.1258, value=0.1952}
         <at 0.341 0.0902>
         {interpMode=RCIM_Cubic, tangentMode=RCTM_User, time=0.496, value=0.2532}
         <at 1 0>
        ]}}})
    Scale RGB=&VectorFromFloat(Value=&FloatFromCurve(FloatCurve=DataInterface() {propertyValues {Curve {keys [
          {arriveTangent=3.5209, interpMode=RCIM_Cubic, leaveTangent=3.5209, tangentMode=RCTM_User, time=0, value=0.5}
          {arriveTangent=-20.2701, interpMode=RCIM_Cubic, leaveTangent=-20.2701, tangentMode=RCTM_User, time=0.45, value=5}
          <at 1 1>
         ]}}}))
    ScaleA=true
    ScaleRGB=true
   )
   /SetVariables(Particles.SpriteColor=&MultiplyLinearColor(
     Linear Color=$Particles.Color
     Scale Factor=&MakeVector4_FromLinearColor(LinearColor=&MultiplyLinearColor(
       Linear Color=<rgba 22.2767 2.7043 25 1>
       Scale Factor=&MakeVector4(
        X=1
        Y=1
        Z=1
        W=&FloatFromCurve(FloatCurve=DataInterface() {propertyValues {Curve {keys [
             {interpMode=RCIM_Cubic, tangentMode=RCTM_User, time=0, value=1}
             {arriveTangentWeight=0.031, interpMode=RCIM_Cubic, leaveTangentWeight=0.0366, tangentMode=RCTM_User, time=0.0918, value=0.05}
             {arriveTangentWeight=0.16, interpMode=RCIM_Cubic, leaveTangent=18.7153, leaveTangentWeight=0.1479, tangentMode=RCTM_Break, time=0.18, value=0}
             {arriveTangent=-0.2815, interpMode=RCIM_Cubic, leaveTangent=-0.2815, tangentMode=RCTM_User, time=0.3, value=0.95}
             {arriveTangentWeight=0.0345, interpMode=RCIM_Cubic, leaveTangentWeight=0.0566, tangentMode=RCTM_User, time=0.4, value=0.1}
             {
              arriveTangentWeight=0.1589
              interpMode=RCIM_Cubic
              leaveTangent=29.6956
              leaveTangentWeight=0.1018
              tangentMode=RCTM_Break
              time=0.5
              value=0.05
             }
             {arriveTangent=0.512, interpMode=RCIM_Cubic, leaveTangent=0.512, tangentMode=RCTM_User, time=0.5641, value=0.8648}
             <at 1 0>
            ]}}})
       )
      ))
    )) {isSetParameters=true}
   &DynamicMaterialParameters(
    Index 0 Param 1=&FloatFromCurve(FloatCurve=DataInterface() {propertyValues {Curve {keys [{tangentMode=RCTM_User, time=0, value=1}, {tangentMode=RCTM_User, time=1, value=0}]}}})
    Index 0 Param 2=&Add_Float(
     A=&FloatFromCurve(FloatCurve=DataInterface() {propertyValues {Curve {keys [
          {interpMode=RCIM_Cubic, leaveTangent=5.6571, tangentMode=RCTM_Break, time=0, value=0}
          {arriveTangent=0.0136, interpMode=RCIM_Cubic, leaveTangent=-5.3647, tangentMode=RCTM_Break, time=0.4833, value=1}
          <at 1 -0.25>
         ]}}})
     B=&Multiply_Float(A=$Particles.BiasedRandom, B=0.5)
    )
    Index 0 Param 3=&RandomRangeFloat(Minimum=0, Maximum=2.5)
    Index 1 Param 1=&Add_Float(A=$Particles.NormalizedAge, B=&RandomRangeFloat(Minimum=0, Maximum=1))
    Index 1 Param 2=&RandomRangeFloat(Minimum=-50, Maximum=50)
    Index 1 Param 3=3
    Module.Param0WriteEnabled=true
    Module.Param0WriteEnabled1=true
    Module.Param1WriteEnabled=true
    Module.Param1WriteEnabled1=true
    Module.Param2WriteEnabled=true
    Module.Param2WriteEnabled1=true
    Module.Param3WriteEnabled=false
    Module.Param3WriteEnabled1=false
    Write Parameter Index 1=true
   ) {module=DynamicMaterialParameters001}
   &ScaleSpriteSize(Uniform Curve Sprite Scale=DataInterface() {propertyValues {Curve {keys [
        {arriveTangent=21.5526, interpMode=RCIM_Cubic, time=0, value=0}
        <at 0.1 1 RCIM_Cubic>
        {arriveTangent=-1.8758, interpMode=RCIM_Cubic, leaveTangent=-1.8758, tangentMode=RCTM_User, time=1, value=0}
       ]}}})
  ]
  rendererClasses [&NiagaraMeshRendererProperties, &NiagaraSpriteRendererProperties]
  renderers [
   {rendererClass=&NiagaraMeshRendererProperties, rendererIndex=0}
   {rendererClass=&NiagaraSpriteRendererProperties, rendererIndex=1}
  ]
 }
 CubeMesh {
  emitterUpdateScript [
   &EmitterState()
   &SpawnBurst_Instantaneous(Spawn Time=2, Use Spawn Probability=false, UseLoopCountLimit=false)
  ]
  particleSpawnScript [
   /Niagara/Modules/Spawn/Initialization/V2/InitializeParticle(
    Color=<rgba 3.4598 0.479 4 1>
    Lifetime=8
    Mesh Scale Mode=::NewEnumerator3
    Mesh Uniform Scale=&Divide_Float(A=$User.CubeRadius, B=768)
    Position Offset=$System.CenterOffset
    UsePositionOffset=true
    Write Material Random=true
   )
   &CameraOffset(Camera Offset Amount=1.5)
  ]
  particleUpdateScript [
   &ParticleState()
   &DynamicMaterialParameters(
    Index 0 Param 1=&FloatFromCurve(FloatCurve=DataInterface() {propertyValues {Curve {keys [
         {arriveTangent=0.8673, interpMode=RCIM_Cubic, leaveTangent=0.8673, tangentMode=RCTM_User, time=0, value=0.07}
         {arriveTangent=1.0964, interpMode=RCIM_Cubic, leaveTangent=1.0964, tangentMode=RCTM_User, time=0.3, value=0.3}
         {arriveTangent=2.0958, interpMode=RCIM_Cubic, leaveTangent=2.0958, tangentMode=RCTM_User, time=1, value=1}
        ]}}})
    Index 0 Param 2=768
    Index 0 Param 3=&FloatFromCurve(FloatCurve=DataInterface() {propertyValues {Curve {keys [
         {arriveTangent=2.1493, interpMode=RCIM_Cubic, time=0.6, value=0}
         {arriveTangent=4.9354, interpMode=RCIM_Cubic, leaveTangent=4.9354, tangentMode=RCTM_User, time=1, value=1}
        ]}}})
    Module.Param0WriteEnabled=true
    Module.Param1WriteEnabled=true
    Module.Param2WriteEnabled=true
    Module.Param3WriteEnabled=false
   )
  ]
  rendererClasses [&NiagaraMeshRendererProperties]
  renderers [{rendererClass=&NiagaraMeshRendererProperties, rendererIndex=0}]
 }
 Glow {
  emitterUpdateScript [
   &EmitterState()
   &SpawnBurst_Instantaneous(Spawn Time=2, Use Spawn Probability=false, UseLoopCountLimit=false)
  ]
  particleSpawnScript [
   &InitializeParticle(
    Color=<rgba 2.2886 0.4473 10 1>
    Lifetime=10
    Mesh Scale Mode=::NewEnumerator3
    Mesh Uniform Scale=&Divide_Float(A=$User.CubeRadius, B=768)
    UsePositionOffset=false
    Write Material Random=true
   )
   &InitialMeshOrientation(Mesh Orientation Mode=::NewEnumerator2, Orientation Vector=<0 0 1>, Rotation=<0 -0.25 0>, Use Rotation Vector=false)
   &CameraOffset(Camera Offset Amount=2)
   /SetVariables() {isSetParameters=true}
   &DynamicMaterialParameters(
    Index 0 Param 1=&Multiply_Float(A=0.75, B=$System.CubeRadiusScale)
    Module.Param0WriteEnabled=true
    Module.Param1WriteEnabled=false
    Module.Param2WriteEnabled=false
    Module.Param3WriteEnabled=false
   )
  ]
  particleUpdateScript [
   &ParticleState()
   &ScaleColor(
    Scale Alpha=&FloatFromCurve(FloatCurve=DataInterface() {propertyValues {Curve {keys [
         {interpMode=RCIM_Cubic, tangentMode=RCTM_User, time=0, value=0}
         {
          arriveTangent=2.6317
          arriveTangentWeight=0.2609
          interpMode=RCIM_Cubic
          leaveTangent=2.6317
          leaveTangentWeight=0.1118
          tangentMode=RCTM_User
          time=0.85
          value=1
         }
         <at 1 0>
        ]}}})
    Scale RGB=&VectorFromFloat(Value=&Add_Float(
      A=1
      B=&FloatFromCurve(FloatCurve=DataInterface() {propertyValues {Curve {keys [{interpMode=RCIM_Cubic, tangentMode=RCTM_User, time=0, value=0}, {interpMode=RCIM_Cubic, tangentMode=RCTM_User, time=1, value=1}]}}})
     ))
    ScaleA=true
    ScaleRGB=true
   )
   /SetVariables(Particles.LightColor=&MultiplyLinearColor(
     Linear Color=<rgba 50 4.9186 42.2504 1>
     Scale Factor=&MakeVector4_FromFloat(Float=&FloatFromCurve(FloatCurve=DataInterface() {propertyValues {Curve {keys [
           {arriveTangent=9.0725, interpMode=RCIM_Cubic, time=0, value=0}
           {arriveTangent=0.3361, interpMode=RCIM_Cubic, leaveTangent=0.3361, tangentMode=RCTM_User, time=0.25, value=0.75}
           {arriveTangent=1, interpMode=RCIM_Cubic, leaveTangent=1, time=0.8, value=2}
           {arriveTangent=-0.0405, interpMode=RCIM_Cubic, leaveTangent=-0.0405, tangentMode=RCTM_User, time=1, value=0}
          ]}}}))
    )) {isSetParameters=true}
   /SetVariables(Particles.LightRadius=&FloatFromCurve(
     FloatCurve=DataInterface() {propertyValues {Curve {keys [
         {arriveTangent=9.0725, interpMode=RCIM_Cubic, time=0, value=0}
         {arriveTangent=0.3361, interpMode=RCIM_Cubic, leaveTangent=0.3361, tangentMode=RCTM_User, time=0.25, value=1}
         {arriveTangent=1, interpMode=RCIM_Cubic, leaveTangent=1, time=0.8, value=1.25}
         {arriveTangent=-0.0405, interpMode=RCIM_Cubic, leaveTangent=-0.0405, tangentMode=RCTM_User, time=1, value=0}
        ]}}}
     Scale Curve=&Multiply_Float(A=$User.CubeRadius, B=2)
    )) {isSetParameters=true}
  ]
  rendererClasses [&NiagaraMeshRendererProperties, &NiagaraLightRendererProperties]
  renderers [
   {rendererClass=&NiagaraMeshRendererProperties, rendererIndex=0}
   {rendererClass=&NiagaraLightRendererProperties, rendererIndex=1}
  ]
 }
 Lightning_center {
  emitterUpdateScript [
   &EmitterState()
   &SpawnBurst_Instantaneous(Use Spawn Probability=false, UseLoopCountLimit=false)
   &SpawnBurst_Instantaneous(Spawn Time=0.06, Use Spawn Probability=false, UseLoopCountLimit=false) {module=SpawnBurst_Instantaneous001}
  ]
  particleSpawnScript [
   /Niagara/Modules/Spawn/Initialization/InitializeParticle(
    Color=<rgba 3.6266 1.155 5 1>
    Lifetime=0.3
    Mesh Scale=&Multiply_VectorByFloat(
     Vector=&RandomRangeVector(Minimum=<5 5 25>, Maximum=<15 15 35>)
     Float=&RemapRange(Input Value To Remap=$Engine.Owner.LODDistance, Input Min=5000, Input Max=20000, Output Min=0.5, Output Max=3)
    )
    Module.Write Color=true
    Module.Write Lifetime=true
    Module.Write Mass=false
    Module.Write Material Random=true
    Module.Write Position=false
    Module.Write Scale=true
    Module.Write SpriteRotation=false
    Module.Write SpriteSize=true
    Sprite Size=&Vector2DFromFloat(Value=&Multiply_Float(
      A=1000
      B=&RemapRange(Input Value To Remap=$Engine.Owner.LODDistance, Input Min=5000, Input Max=20000, Output Min=1, Output Max=10)
     ))
   )
   &InitialMeshOrientation(
    Orientation Axis=<-1 0 0>
    Rotation=&RandomRangeVector(Minimum=<0 0 0>, Maximum=<1 0 0>)
    Use Orientation Vector=true
    Use Rotation Vector=true
   ) {module=InitialMeshOrientation001}
   /SetVariables(Particles.DynParamsSprite:Vector4f=<1 1 0 0>, Particles.DynParamsSprite2:Vector4f=<400 0 0 0>) {isSetParameters=true}
   &InitialMeshOrientation(
    Mesh Orientation Mode=::NewEnumerator2
    Rotation=&MakeVector(X=&RandomRangeFloat(Minimum=0, Maximum=1), Y=0, Z=0)
    Use Rotation Vector=true
   )
   /SetVariables(Particles.MeshIndex=&RandomRangeInt(Minimum=0, Maximum=1)) {isSetParameters=true}
   /SetVariables(Particles.CamOffsetSprite=150) {isSetParameters=true}
  ]
  particleUpdateScript [
   &ParticleState()
   &ScaleColor(
    Scale Alpha=&FloatFromCurve(FloatCurve=DataInterface() {propertyValues {Curve {keys [
         <at 0 0>
         {interpMode=RCIM_Cubic, tangentMode=RCTM_User, time=0.1258, value=0.1952}
         <at 0.341 0>
         {interpMode=RCIM_Cubic, tangentMode=RCTM_User, time=0.496, value=0.2532}
         <at 1 0>
        ]}}})
    Scale RGB=&VectorFromFloat(Value=&FloatFromCurve(FloatCurve=DataInterface() {propertyValues {Curve {keys [
          <at 0.4 0.5 RCIM_Cubic>
          {arriveTangent=-2.404, interpMode=RCIM_Cubic, leaveTangent=-2.404, tangentMode=RCTM_User, time=0.45, value=1}
          <at 1 0.5>
         ]}}}))
    ScaleA=true
    ScaleRGB=true
   )
   /SetVariables(Particles.SpriteColor=&MultiplyLinearColor(
     Linear Color=<rgba 4.6519 0.0848 5 1>
     Scale Factor=&MakeVector4(
      X=1
      Y=1
      Z=1
      W=&FloatFromCurve(FloatCurve=DataInterface() {propertyValues {Curve {keys [
           {arriveTangent=9.0725, interpMode=RCIM_Cubic, leaveTangent=9.0725, tangentMode=RCTM_User, time=0.27, value=0}
           {arriveTangent=-2.5249, interpMode=RCIM_Cubic, leaveTangent=-2.5249, tangentMode=RCTM_User, time=0.35, value=1}
           {arriveTangentWeight=0.3577, interpMode=RCIM_Cubic, leaveTangentWeight=0.2385, tangentMode=RCTM_User, time=0.5, value=0}
           {arriveTangent=-4.1273, interpMode=RCIM_Cubic, leaveTangent=-4.1273, tangentMode=RCTM_User, time=0.6, value=0.8}
           {arriveTangent=-0.0405, interpMode=RCIM_Cubic, leaveTangent=-0.0405, tangentMode=RCTM_User, time=1, value=0}
          ]}}})
     )
    )) {isSetParameters=true}
   &DynamicMaterialParameters(
    Index 0 Param 1=&FloatFromCurve(FloatCurve=DataInterface() {propertyValues {Curve {keys [<at 0 1>,<at 1 0>]}}})
    Index 0 Param 2=&FloatFromCurve(
     FloatCurve=DataInterface() {propertyValues {Curve {keys [
         {interpMode=RCIM_Cubic, leaveTangent=-1.1405, tangentMode=RCTM_Break, time=0, value=0.5}
         {arriveTangent=4.4936, interpMode=RCIM_Cubic, leaveTangent=-5.3647, tangentMode=RCTM_Break, time=0.4833, value=1}
         <at 1 -0.3>
        ]}}}
     Scale Curve=&RemapRange(Input Value To Remap=$Engine.Owner.LODDistance, Input Min=5000, Input Max=15000, Output Min=0.5, Output Max=1)
    )
    Index 0 Param 3=0.5
    Module.Param0WriteEnabled=true
    Module.Param1WriteEnabled=true
    Module.Param2WriteEnabled=true
    Module.Param3WriteEnabled=false
   ) {module=DynamicMaterialParameters001}
   /SetVariables(Particles.light color:LinearColor=&MakeLinearColorFromVectorAndFloat(
     "Vector (RGB)"=&Multiply_VectorByFloat(
      Vector=&MakeVectorFromLinearColorRGB(Color=<rgba 1.5806 0.2805 3 1>)
      Float=&FloatFromCurve(
       FloatCurve=DataInterface() {propertyValues {Curve {keys [
           {interpMode=RCIM_Cubic, tangentMode=RCTM_User, time=0, value=1}
           {arriveTangent=-0.1663, interpMode=RCIM_Cubic, leaveTangent=-0.1663, tangentMode=RCTM_User, time=0.0918, value=0.1}
           {arriveTangent=-0.2815, interpMode=RCIM_Cubic, leaveTangent=-0.2815, tangentMode=RCTM_User, time=0.1514, value=0.9748}
           {arriveTangent=-0.2665, interpMode=RCIM_Cubic, leaveTangent=-0.2665, tangentMode=RCTM_User, time=0.2547, value=0}
           {arriveTangent=0.512, interpMode=RCIM_Cubic, leaveTangent=0.512, tangentMode=RCTM_User, time=0.5641, value=0.8648}
           {arriveTangent=-1.9839, interpMode=RCIM_Cubic, leaveTangent=-1.9839, tangentMode=RCTM_User, time=0.8, value=0.05}
           {arriveTangent=-0.25, interpMode=RCIM_Cubic, leaveTangent=-0.25, time=0.9, value=0.45}
           <at 1 0>
          ]}}}
       Scale Curve=25
      )
     )
     "Float (Alpha)"=1
    )) {isSetParameters=true}
   /SetVariables(Particles.LightPosition:Vector3f=<-700 0 0>) {isSetParameters=true}
  ]
  rendererClasses [
   &NiagaraMeshRendererProperties
   &NiagaraLightRendererProperties
   &NiagaraSpriteRendererProperties
  ]
  renderers [
   {rendererClass=&NiagaraMeshRendererProperties, rendererIndex=0}
   {rendererClass=&NiagaraLightRendererProperties, rendererIndex=1}
   {rendererClass=&NiagaraSpriteRendererProperties, rendererIndex=2}
  ]
 }
 LingerSmoke {
  emitterUpdateScript [
   &EmitterState(Life Cycle Mode=::NewEnumerator1, Loop Behavior=::NewEnumerator1, Loop Delay=0.1, Loop Duration=1, UseLoopDelay=true)
   &SpawnRate(
    SpawnRate=&FloatFromCurve(
     FloatCurve=DataInterface() {propertyValues {Curve {keys [{arriveTangent=-1.7213, interpMode=RCIM_Cubic, leaveTangent=-1.7213, tangentMode=RCTM_User, time=0, value=1}, <at 1 0 RCIM_Cubic>]}}}
     CurveIndex=$Emitter.NormalizedLoopAge
     Scale Curve=16
    )
    Use Spawn Probability=false
   )
   &SpawnBurst_Instantaneous(Spawn Count=12, Use Spawn Probability=false)
  ]
  particleSpawnScript [
   &InitializeParticle(
    Color=&UniformRangedLinearColor(Minimum=<rgba 0.1449 0.0392 0.16 0.4>, Maximum=<rgba 0.2447 0.0638 0.25 0.6>)
    Lifetime=&FloatFromCurve(
     FloatCurve=DataInterface() {propertyValues {Curve {keys [<at 0.1 0.25 RCIM_Cubic>,<at 0.2 1>]}}}
     CurveIndex=$Emitter.Age
     Scale Curve=&RandomRangeFloat(Minimum=3, Maximum=7)
    )
    Module.Write Color=true
    Module.Write Lifetime=true
    Module.Write Mass=false
    Module.Write Material Random=true
    Module.Write Position=true
    Module.Write Scale=false
    Module.Write SpriteRotation=true
    Module.Write SpriteSize=true
    Sprite Rotation=&UniformRangedFloat(Minimum=0, Maximum=360)
    Sprite Size=&Vector2DFromFloat(Value=&FloatFromCurve(
      FloatCurve=DataInterface() {propertyValues {Curve {keys [<at 0.1 1.7 RCIM_Cubic>,<at 0.2 1>]}}}
      CurveIndex=$Emitter.Age
      Scale Curve=&RandomRangeFloat(Minimum=400, Maximum=700)
     ))
   )
   &SphereLocation(
    MaskSpawnGroup=false
    Non Uniform Scale=<0.1 1 1>
    Offset=<50 0 0>
    Override Seed=false
    Sphere Radius=50
    Surface Only=false
    UseNonUniformScale=true
    UseOffset=true
    UseSphereOrientation=false
   ) {module=SphereLocation001}
   &AddVelocityFromPoint(
    Module.Use Falloff Distance=false
    Origin Offset=<0 0 -55>
    Origin Offset Coordinate Space=::World
    Velocity Strength=&FloatFromCurve(
     FloatCurve=DataInterface() {propertyValues {Curve {keys [<at 0.1 8 RCIM_Cubic>,<at 0.2 1>]}}}
     CurveIndex=$Emitter.Age
     Scale Curve=&RandomRangeFloat(Minimum=250, Maximum=400)
    )
   )
  ]
  particleUpdateScript [
   &ParticleState()
   &Drag(Drag=&RandomRangeFloat(Minimum=2, Maximum=3), Use Linear Drag=true, Use Rotational Drag=false)
   &SpriteRotationRate(Rotation Rate=&Multiply_Float(
     A=&Lerp_Float(A=-300, B=300, Alpha=$Particles.MaterialRandom)
     B=&FloatFromCurve(FloatCurve=DataInterface() {propertyValues {Curve {keys [
          {arriveTangent=-4.4572, interpMode=RCIM_Cubic, leaveTangent=-4.4572, tangentMode=RCTM_User, time=0, value=1}
          {arriveTangent=-1.0628, interpMode=RCIM_Cubic, leaveTangent=-1.0628, tangentMode=RCTM_User, time=0.2, value=0.4}
          <at 1.0011 0.01>
         ]}}})
    ))
   &AccelerationForce(Acceleration=<0 0 40>)
   &SolveForcesAndVelocity(Clamp Velocity=false, Limit Acceleration=false)
   &CameraOffset(Camera Offset Amount=&Multiply_Float(
     A=0.5
     B=/Niagara/DynamicInputs/Math/CalculateParticleRadius(
      Radius Calculation Type:ENiagaraCollisionRadiusOptions=::NewEnumerator0
      Particle Radius:NiagaraFloat=1
      Mesh Dimensions:Vector3f=<1 1 1>
      Method for Calculating Particle Radius:ENiagaraCalculateRadiusOptions=::NewEnumerator2
     )
    ))
   &ScaleSpriteSize(Scale Factor=&Vector2DFromFloat(Value=&FloatFromCurve(FloatCurve=DataInterface() {propertyValues {Curve {keys [
          {arriveTangent=2.7007, interpMode=RCIM_Cubic, leaveTangent=2.7007, tangentMode=RCTM_User, time=0, value=0.25}
          {arriveTangent=1, interpMode=RCIM_Cubic, leaveTangent=1, time=0.3314, value=0.8}
          {arriveTangent=0.5339, interpMode=RCIM_Cubic, leaveTangent=-0.3011, tangentMode=RCTM_User, time=1, value=1.25}
         ]}}})))
   &ScaleColor(
    Scale Alpha=&FloatFromCurve(FloatCurve=DataInterface() {propertyValues {Curve {keys [<at 0 0 RCIM_Cubic>, <at 0.3 1 RCIM_Cubic>, {interpMode=RCIM_Cubic, leaveTangent=-3.4295, time=1, value=0}]}}})
    Scale RGB=&VectorFromFloat(Value=&FloatFromCurve(FloatCurve=DataInterface() {propertyValues {Curve {keys [<at 0.05 40 RCIM_Cubic>,<at 0.5 1>]}}}, CurveIndex=$Emitter.Age))
    ScaleA=true
    ScaleRGB=true
   )
   /SetVariables(Particles.SubImageIndex=&Lerp_Float(A=20, B=63, Alpha=$Particles.NormalizedAge)) {isSetParameters=true}
  ]
  rendererClasses [&NiagaraSpriteRendererProperties]
  renderers [{rendererClass=&NiagaraSpriteRendererProperties}]
 }
 NE_FN_Sparks {
  simTarget=GPUComputeSim
  emitterUpdateScript [
   &EmitterState(Enable Distance Culling=true, MaxDistance=8000, Module.bUseMaxDistance=true, Module.bUseMinDistance=false, Scalability Mode=::NewEnumerator1)
   &SpawnBurst_Instantaneous(Spawn Count=55, Spawn Time=0.1, Use Spawn Probability=false)
   &SpawnBurst_Instantaneous(Spawn Count=45, Spawn Time=0.18, Use Spawn Probability=false) {module=SpawnBurst_Instantaneous001}
  ]
  particleSpawnScript [
   /SetVariables(Particles.SparkIntensity:NiagaraFloat=&FloatFromCurve(
     FloatCurve=DataInterface() {propertyValues {Curve {keys [
         <at 0 1 RCIM_Cubic>
         {arriveTangent=-2.5317, interpMode=RCIM_Cubic, leaveTangent=-2.5317, tangentMode=RCTM_User, time=0.4, value=0.5}
         <at 1 0 RCIM_Cubic>
        ]}}}
     CurveIndex=&RandomRangeFloat(Minimum=0, Maximum=1)
    )) {isSetParameters=true}
   /Niagara/Modules/Spawn/Initialization/V2/InitializeParticle(
    Color=&LerpLinearColors(StartColor=<rgba 15 0.3906 14.871 1>, EndColor=<rgba 25 2.7344 18.1064 1>, LerpFactor=$Particles.SparkIntensity)
    Lifetime Max=1.5
    Lifetime Min=0.15
    Lifetime Mode=::NewEnumerator1
    Mass=&Lerp_Float(A=0.3, B=1, Alpha=$Particles.SparkIntensity)
    Mass Mode=::NewEnumerator1
    Sprite Size Mode=::NewEnumerator3
    Uniform Sprite Size=&Lerp_Float(A=5, B=8, Alpha=$Particles.SparkIntensity)
    UsePositionOffset=false
    Write Material Random=true
   )
   /SetVariables(Particles.SpriteSize=&Multiply_Vector2(
     A=$Particles.SpriteSize
     B=&Vector2DFromFloat(Value=&FloatFromCurve(FloatCurve=DataInterface() {propertyValues {Curve {keys [<at 500 1 RCIM_Cubic>,<at 7000 3>]}}}, CurveIndex=$Engine.Owner.LODDistance))
    )) {isSetParameters=true}
   &SphereLocation(
    MaskSpawnGroup=false
    Override Seed=false
    Sphere Radius=15
    Surface Only=false
    UseNonUniformScale=false
    UseOffset=false
    UseSphereOrientation=false
   )
   &AddVelocityInCone(
    Cone Angle=75
    Cone Axis=<-1 0 0>
    Module.Use Velocity Falloff On Cone Axis=false
    Velocity Distribution Along Cone Axis=0.8
    Velocity Strength=&RandomRangeFloat(Minimum=500, Maximum=2400)
   )
  ]
  particleUpdateScript [
   &ParticleState()
   &GravityForce(Gravity=<0 0 -2200>)
   &Drag(Drag=0.15, Use Linear Drag=true, Use Rotational Drag=false)
   &VectorNoiseForce(
    Force Amount=&FloatFromCurve(
     FloatCurve=DataInterface() {propertyValues {Curve {keys [
         {arriveTangent=-9.4515, interpMode=RCIM_Cubic, leaveTangent=-4.1687, tangentMode=RCTM_User, time=0, value=1}
         {arriveTangent=-3.4372, interpMode=RCIM_Cubic, leaveTangent=-3.4372, tangentMode=RCTM_User, time=0.6, value=0}
        ]}}}
     Scale Curve=3000
    )
    Module.Use Random Noise Time Interval=false
   )
   &SolveForcesAndVelocity(Clamp Velocity=false, Limit Acceleration=false)
   &ScaleColor(
    Linear Color Curve=DataInterface() {propertyValues {
      AlphaCurve {keys [<at 0 1>,<at 1 1>]}
      BlueCurve {keys [
        <at 0 2 RCIM_Cubic>
        {arriveTangent=-1.452, interpMode=RCIM_Cubic, leaveTangent=-1.452, tangentMode=RCTM_User, time=0.1467, value=1}
        <at 1 0.0755>
       ]}
      GreenCurve {keys [
        <at 0 2 RCIM_Cubic>
        {arriveTangent=-1.4408, interpMode=RCIM_Cubic, leaveTangent=-1.4408, tangentMode=RCTM_User, time=0.1467, value=1}
        <at 1 0.078>
       ]}
      RedCurve {keys [
        <at 0 2 RCIM_Cubic>
        {arriveTangent=-0.8332, interpMode=RCIM_Cubic, leaveTangent=-0.8332, tangentMode=RCTM_User, time=0.1467, value=1}
        <at 1 0.25>
       ]}
     }}
    Scale Mode=::NewEnumerator2
   )
   &ScaleSpriteSize(Scale Factor=&Vector2DFromFloat(Value=&FloatFromCurve(FloatCurve=DataInterface() {propertyValues {Curve {keys [<at 0.4 1 RCIM_Cubic>, {arriveTangent=-3.8662, interpMode=RCIM_Cubic, leaveTangent=-3.8662, tangentMode=RCTM_User, time=1, value=0}]}}})))
   &ScaleSpriteSizeBySpeed(
    Max Scale Factor=<1.25 10>
    Sample Scale Factor By Curve=true
    Scale Factor Curve=DataInterface() {propertyValues {Curve {keys [
        {arriveTangent=0.2757, time=0, value=0}
        {arriveTangent=0.3844, interpMode=RCIM_Cubic, leaveTangent=0.3844, tangentMode=RCTM_User, time=0.3, value=0.15}
        {arriveTangent=2.2765, interpMode=RCIM_Cubic, leaveTangent=2.2765, tangentMode=RCTM_User, time=1, value=1}
       ]}}}
    Velocity Threshold=2000
   )
  ]
  rendererClasses [&NiagaraSpriteRendererProperties]
  renderers [{rendererClass=&NiagaraSpriteRendererProperties}]
 }
 Runes {
  emitterUpdateScript [
   &EmitterState(
    Enable Distance Culling=true
    Life Cycle Mode=::NewEnumerator1
    Loop Behavior=::NewEnumerator1
    Loop Delay=2.5
    Loop Duration=8
    MaxDistance=10000
    Module.bUseMaxDistance=true
    Module.bUseMinDistance=false
    Scalability Mode=::NewEnumerator1
    UseLoopDelay=true
   )
   &SpawnRate(
    SpawnRate=&FloatFromCurve(
     FloatCurve=DataInterface() {propertyValues {Curve {keys [
         <at 0 0.05 RCIM_Cubic>
         {arriveTangent=3, interpMode=RCIM_Cubic, time=0.4, value=1}
         {arriveTangent=4, interpMode=RCIM_Cubic, leaveTangent=4, time=0.55, value=0.3}
         {
          arriveTangent=-19.2208
          arriveTangentWeight=0.0943
          interpMode=RCIM_Cubic
          leaveTangent=-19.2208
          leaveTangentWeight=0.0471
          tangentMode=RCTM_User
          time=0.65
          value=2
         }
         {tangentMode=RCTM_User, time=1, value=0}
        ]}}}
     CurveIndex=$Emitter.NormalizedLoopAge
     Scale Curve=&Multiply_Float(A=800, B=$System.CubeRadiusScale)
    )
    Use Spawn Probability=false
   )
  ]
  particleSpawnScript [
   &FN_SetSpawnGroup()
   &InitializeParticle(
    Color Maximum=<rgba 1.5 25 15 0.8>
    Color Minimum=<rgba 0 15 5 0.5>
    Color Mode=::NewEnumerator2
    Lifetime Max=0.5
    Lifetime Min=0.1
    Lifetime Mode=::NewEnumerator1
    Sprite Rotation Angle=&Divide_Float(A=&MakeFloatFromInt(Int=$Particles.SpawnGroup), B=4)
    Sprite Rotation Mode=::NewEnumerator1
    Sprite Size Mode=::NewEnumerator3
    Uniform Sprite Size=48
    UsePositionOffset=false
    Write Material Random=true
   )
   /SetVariables(Particles.CubeRad-ParticleRad:NiagaraFloat=&Subtract_Float(
     A=$User.CubeRadius
     B=&CalculateParticleRadius(
      Radius Calculation Type:ENiagaraCollisionRadiusOptions=::NewEnumerator0
      Particle Radius:NiagaraFloat=1
      Mesh Dimensions:Vector3f=<1 1 1>
      Method for Calculating Particle Radius:ENiagaraCalculateRadiusOptions=::NewEnumerator2
     )
    )) {isSetParameters=true}
   /SetVariables(Particles.Position=&RotateVectorByQuaternion(
     Vector To Rotate=&MakeVector(
      X=&Multiply_Float(
       A=&Lerp_Float(
        A=-1
        B=&FloatFromCurve(
         FloatCurve=DataInterface() {propertyValues {Curve {keys [{arriveTangent=0.4798, interpMode=RCIM_Cubic, time=0.04, value=0}, {interpMode=RCIM_Cubic, leaveTangent=2.1857, time=0.48, value=-2}]}}}
         CurveIndex=$Emitter.NormalizedLoopAge
        )
        Alpha=&FloatFromCurve(
         FloatCurve=DataInterface() {propertyValues {Curve {keys [
             {arriveTangent=4.4415, interpMode=RCIM_Cubic, leaveTangent=4.4415, tangentMode=RCTM_User, time=0, value=0}
             {arriveTangent=0.8587, interpMode=RCIM_Cubic, leaveTangent=0.8587, tangentMode=RCTM_User, time=0.3, value=0.7}
             {tangentMode=RCTM_User, time=1, value=1}
            ]}}}
         CurveIndex=&RandomRangeFloat(Minimum=0, Maximum=1)
        )
       )
       B=Linked() {linkedVariable {name=Particles.CubeRad-ParticleRad, type=NiagaraFloat}}
      )
      Y=&Multiply_Float(
       A=Linked() {linkedVariable {name=Particles.CubeRad-ParticleRad, type=NiagaraFloat}}
       B=&RandomRangeFloat(Minimum=-1, Maximum=1)
      )
      Z=$User.CubeRadius
     )
     Quaternion=&MakeQuatFromAxisAngle(
      Axis=<1 0 0>
      AngleInDegrees=&Multiply_Float(A=&MakeFloatFromInt(Int=$Particles.SpawnGroup), B=90)
     )
    )) {isSetParameters=true}
   /SetVariables() {isSetParameters=true}
   /SetVariables(Particles.SpriteFacing=&RotateVectorByQuaternion(
     Vector To Rotate=<0 0 1>
     Quaternion=&MakeQuatFromAxisAngle(
      Axis=<1 0 0>
      AngleInDegrees=&Multiply_Float(A=&MakeFloatFromInt(Int=$Particles.SpawnGroup), B=90)
     )
    )) {isSetParameters=true}
   &CameraOffset(Camera Offset Amount=5)
  ]
  particleUpdateScript [
   &ParticleState()
   &SubUVAnimation(
    End Frame=4
    Sub UV Play Rate:NiagaraFloat=&RandomRangeFloat(Minimum=0.5, Maximum=2)
    SubUV Animation Mode:ENiagara_SubUVLookupMode=::NewEnumerator3
    SubUV Start Time Offset:NiagaraFloat=&RandomRangeFloat(Minimum=0, Maximum=1)
   )
   &ScaleColor(
    Scale Alpha=&FloatFromCurve(FloatCurve=DataInterface() {propertyValues {Curve {keys [
         {arriveTangent=-3.8362, interpMode=RCIM_Cubic, leaveTangent=-3.8362, tangentMode=RCTM_User, time=0, value=1}
         <at 0.1866 0.4>
         {arriveTangent=-4.1865, interpMode=RCIM_Cubic, leaveTangent=-4.1865, tangentMode=RCTM_User, time=0.2015, value=1}
         <at 0.4552 0.3>
         {arriveTangent=-8.2931, interpMode=RCIM_Cubic, leaveTangent=-8.2931, tangentMode=RCTM_User, time=0.4652, value=0.75}
         <at 0.7089 0>
         {arriveTangent=-5.5334, interpMode=RCIM_Cubic, leaveTangent=-5.5334, tangentMode=RCTM_User, time=0.7338, value=0.5}
         <at 1 0>
        ]}}})
    Scale RGB=&VectorFromCurve(VectorCurve=DataInterface() {propertyValues {
       XCurve {keys [<at 0 1>,<at 1 1>]}
       YCurve {keys [<at 0 1>,<at 1 1>]}
       ZCurve {keys [
         {arriveTangent=1.0471, interpMode=RCIM_Cubic, leaveTangent=1.0471, tangentMode=RCTM_User, time=0, value=0.8}
         {interpMode=RCIM_Cubic, tangentMode=RCTM_User, time=0.25, value=1}
         {arriveTangent=-2.0404, interpMode=RCIM_Cubic, leaveTangent=-2.0404, tangentMode=RCTM_User, time=1, value=0}
        ]}
      }})
    ScaleA=true
    ScaleRGB=true
   )
  ]
  rendererClasses [&NiagaraSpriteRendererProperties]
  renderers [{rendererClass=&NiagaraSpriteRendererProperties}]
 }
 Shimmers {
  emitterUpdateScript [
   &EmitterState(Enable Distance Culling=true, MaxDistance=8000, Module.bUseMaxDistance=true, Module.bUseMinDistance=false, Scalability Mode=::NewEnumerator1)
   &SpawnRate(
    SpawnRate=&FloatFromCurve(
     FloatCurve=DataInterface() {propertyValues {Curve {keys [<at 0.1 1 RCIM_Cubic>, {arriveTangent=-3.4561, interpMode=RCIM_Cubic, leaveTangent=-3.4561, tangentMode=RCTM_User, time=0.75, value=0}]}}}
     CurveIndex=$Emitter.NormalizedLoopAge
     Scale Curve=&Multiply_Float(A=35, B=$System.CubeRadiusScale)
    )
    Use Spawn Probability=false
   )
  ]
  particleSpawnScript [
   /SetVariables(
    Particles.ColorEnd:LinearColor=&RandomRangeLinearColor(Minimum=<rgba 0.25 0 0.138 1>, Maximum=<rgba 0.0466 0 0.25 1>)
    Particles.ColorStart:LinearColor=&RandomRangeLinearColor(Minimum=<rgba 0.36 0.3 0.45 1>, Maximum=<rgba 0.6 0.68 0.74 1>)
   ) {isSetParameters=true}
   &InitializeParticle(
    Color=$Particles.ColorStart
    Lifetime Mode=::NewEnumerator1
    Sprite Rotation Mode=::NewEnumerator2
    Sprite Size=&RandomRangeVector2D(Minimum=<150 150>, Maximum=<300 300>, Evaluation Type=::NewEnumerator1)
    Sprite Size Mode=::NewEnumerator4
    UsePositionOffset=false
    Write Material Random=true
   )
   /SetVariables(Particles.Position=&Add_Vector(
     A=&VectorFromCurve(
      VectorCurve=DataInterface() {propertyValues {
        XCurve {keys [{interpMode=RCIM_Cubic, leaveTangentWeight=0.0833, tangentMode=RCTM_User, time=0, value=0}]}
        YCurve {keys [
          {interpMode=RCIM_Cubic, tangentMode=RCTM_User, time=0, value=1}
          {interpMode=RCIM_Cubic, leaveTangentWeight=0.6719, tangentMode=RCTM_User, time=0, value=1}
          {arriveTangentWeight=0.3436, interpMode=RCIM_Cubic, leaveTangentWeight=0.3436, tangentMode=RCTM_User, time=0.25, value=-1}
          {arriveTangentWeight=0.3436, interpMode=RCIM_Cubic, leaveTangentWeight=0.3436, tangentMode=RCTM_User, time=0.5, value=-1}
          {arriveTangentWeight=0.0833, interpMode=RCIM_Cubic, tangentMode=RCTM_User, time=0.75, value=1}
         ]}
        ZCurve {keys [
          <at 0.25 1 RCIM_Cubic>
          <at 0.5 -1>
          {arriveTangentWeight=0.3436, interpMode=RCIM_Cubic, leaveTangentWeight=0.3436, tangentMode=RCTM_User, time=0.75, value=-1}
          <at 1 1>
         ]}
       }}
      CurveIndex=&RandomRangeFloat(Minimum=0, Maximum=1)
      Scale Curve=&VectorFromFloat(Value=&FloatFromCurve(
        FloatCurve=DataInterface() {propertyValues {Curve {keys [
            {arriveTangent=7.4173, interpMode=RCIM_Cubic, leaveTangent=7.4173, tangentMode=RCTM_User, time=0, value=0.1}
            {interpMode=RCIM_Cubic, tangentMode=RCTM_User, time=0.2, value=1}
           ]}}}
        CurveIndex=$Emitter.NormalizedLoopAge
        Scale Curve=$User.CubeRadius
       ))
     )
     B=&MakeVector(X=&RandomRangeFloat(Minimum=-25, Maximum=0), Y=0, Z=0)
    )) {isSetParameters=true}
   &AddVelocity(Coordinate Space=::World, Velocity=&RandomRangeVector(Minimum=<-2 -2 50>, Maximum=<2 2 16>))
   &DynamicMaterialParameters(
    Index 0 Param 1=&RandomRangeFloat(Minimum=0, Maximum=1, Evaluation Type=::NewEnumerator1)
    Index 0 Param 2=&RandomRangeFloat(Minimum=0.4, Maximum=0.6, Evaluation Type=::NewEnumerator1)
    Index 0 Param 3=0.2
    Module.Param0WriteEnabled=true
    Module.Param1WriteEnabled=true
    Module.Param2WriteEnabled=true
    Module.Param3WriteEnabled=true
   )
   &CameraOffset(Camera Offset Amount=&RandomRangeFloat(Minimum=100, Maximum=170, Evaluation Type=::NewEnumerator1))
  ]
  particleUpdateScript [
   &ParticleState()
   &AccelerationForce(Acceleration=<0 0 100>)
   &SolveForcesAndVelocity(Clamp Velocity=false, Limit Acceleration=false)
   &SpriteRotationRate(Rotation Rate=&RandomRangeFloat(Minimum=-40, Maximum=40))
   &ScaleColor(
    Scale Alpha=&FloatFromCurve(
     FloatCurve=DataInterface() {propertyValues {Curve {keys [<at 0 0 RCIM_Cubic>,<at 0.5 1 RCIM_Cubic>,<at 1 0>]}}}
     Scale Curve=&RandomRangeFloat(Minimum=0.8, Maximum=1)
    )
    Scale RGB=&MakeVectorFromLinearColorRGB(Color=&LerpLinearColors(
      StartColor=$Particles.ColorStart
      EndColor=$Particles.ColorEnd
      LerpFactor=&FloatFromCurve(FloatCurve=DataInterface() {propertyValues {Curve {keys [<at 0 0 RCIM_Cubic>, {interpMode=RCIM_Cubic, tangentMode=RCTM_Break, time=0.5, value=1}]}}})
     ))
    ScaleA=true
    ScaleRGB=true
   ) {module=ScaleColor001}
   &ScaleSpriteSize(Uniform Curve Sprite Scale=DataInterface() {propertyValues {Curve {keys [<at 0 0.7 RCIM_Cubic>, {interpMode=RCIM_Cubic, tangentMode=RCTM_Break, time=1, value=1}]}}})
  ]
  rendererClasses [&NiagaraSpriteRendererProperties]
  renderers [{rendererClass=&NiagaraSpriteRendererProperties}]
 }
 Sparkles {
  simTarget=GPUComputeSim
  emitterUpdateScript [
   &EmitterState(
    Enable Distance Culling=true
    Life Cycle Mode=::NewEnumerator1
    Loop Behavior=::NewEnumerator1
    Loop Delay=2.5
    Loop Duration=8
    MaxDistance=7000
    Module.bUseMaxDistance=true
    Module.bUseMinDistance=false
    Scalability Mode=::NewEnumerator1
    UseLoopDelay=true
   )
   &SpawnRate(
    SpawnRate=&FloatFromCurve(
     FloatCurve=DataInterface() {propertyValues {Curve {keys [
         <at 0 0 RCIM_Cubic>
         {arriveTangent=3, interpMode=RCIM_Cubic, time=0.4, value=1}
         {arriveTangent=4, interpMode=RCIM_Cubic, leaveTangent=4, time=0.55, value=0.3}
         {
          arriveTangent=-19.2208
          arriveTangentWeight=0.0943
          interpMode=RCIM_Cubic
          leaveTangent=-19.2208
          leaveTangentWeight=0.0471
          tangentMode=RCTM_User
          time=0.65
          value=2
         }
         {tangentMode=RCTM_User, time=1, value=0}
        ]}}}
     CurveIndex=$Emitter.NormalizedLoopAge
     Scale Curve=&Multiply_Float(A=900, B=$System.CubeRadiusScale)
    )
    Use Spawn Probability=false
   )
  ]
  particleSpawnScript [
   &FN_SetSpawnGroup()
   &InitializeParticle(
    Color=<rgba 10 0.8416 9.4092 1>
    Lifetime Max=0.6
    Lifetime Min=0.25
    Lifetime Mode=::NewEnumerator1
    Mesh Scale Mode=::NewEnumerator4
    Sprite Size Mode=::NewEnumerator1
    Uniform Sprite Size Max=35
    Uniform Sprite Size Min=12
    UsePositionOffset=false
    Write Material Random=true
   )
   /SetVariables(Particles.Position=&MakeVector(
     X=&FloatFromCurve(
      FloatCurve=DataInterface() {propertyValues {Curve {keys [<at 0 0.05 RCIM_Cubic>, {arriveTangent=2.0792, interpMode=RCIM_Cubic, leaveTangent=2.0792, tangentMode=RCTM_User, time=1, value=1}]}}}
      CurveIndex=&RandomRangeFloat(Minimum=0, Maximum=1)
      Scale Curve=600
     )
     Y=&Multiply_Float(A=$User.CubeRadius, B=&RandomRangeFloat(Minimum=-0.9, Maximum=0.9))
     Z=&Multiply_Float(A=$User.CubeRadius, B=&RandomRangeFloat(Minimum=-0.9, Maximum=0.9))
    )) {enabled=false, isSetParameters=true}
   /SetVariables(Particles.Position=&RotateVectorByQuaternion(
     Vector To Rotate=&MakeVector(
      X=&Multiply_Float(
       A=&Lerp_Float(
        A=-1
        B=&FloatFromCurve(
         FloatCurve=DataInterface() {propertyValues {Curve {keys [{arriveTangent=0.4798, interpMode=RCIM_Cubic, time=0.08, value=0}, {interpMode=RCIM_Cubic, leaveTangent=2.1857, time=0.65, value=-2}]}}}
         CurveIndex=$Emitter.NormalizedLoopAge
        )
        Alpha=&FloatFromCurve(
         FloatCurve=DataInterface() {propertyValues {Curve {keys [
             {arriveTangent=4.4415, interpMode=RCIM_Cubic, leaveTangent=4.4415, tangentMode=RCTM_User, time=0, value=0}
             {arriveTangent=0.8587, interpMode=RCIM_Cubic, leaveTangent=0.8587, tangentMode=RCTM_User, time=0.3, value=0.75}
             {tangentMode=RCTM_User, time=1, value=1}
            ]}}}
         CurveIndex=&RandomRangeFloat(Minimum=0, Maximum=1)
        )
       )
       B=$User.CubeRadius
      )
      Y=&Multiply_Float(A=$User.CubeRadius, B=&RandomRangeFloat(Minimum=-1, Maximum=1))
      Z=$User.CubeRadius
     )
     Quaternion=&MakeQuatFromAxisAngle(
      Axis=<1 0 0>
      AngleInDegrees=&Multiply_Float(A=&MakeFloatFromInt(Int=$Particles.SpawnGroup), B=90)
     )
    )) {isSetParameters=true}
   &CameraOffset(Camera Offset Amount=100)
  ]
  particleUpdateScript [
   &ParticleState()
   &DynamicMaterialParameters(
    Index 0 Param 1=1
    Index 0 Param 2=1
    Index 0 Param 3=0.5
    Module.Param0WriteEnabled=true
    Module.Param1WriteEnabled=true
    Module.Param2WriteEnabled=true
    Module.Param3WriteEnabled=true
   )
   &ScaleColor(
    Scale Alpha=&FloatFromCurve(FloatCurve=DataInterface() {propertyValues {Curve {keys [
         <at 0 0 RCIM_Cubic>
         {interpMode=RCIM_Cubic, tangentMode=RCTM_User, time=0.25, value=0.9914}
         {arriveTangent=-2.3503, interpMode=RCIM_Cubic, leaveTangent=-2.3503, tangentMode=RCTM_User, time=0.7, value=0.6}
         <at 1 0>
        ]}}})
    ScaleA=true
    ScaleRGB=false
   )
   &GravityForce(Gravity=&RandomRangeVector(Minimum=<-5 -5 2500>, Maximum=<5 5 5000>))
   &VectorNoiseForce(Force Amount=20000, Module.Use Random Noise Time Interval=true, Random Noise Time Interval=0.05)
   &Drag(Drag=15, Use Linear Drag=true, Use Rotational Drag=false)
   &ScaleSpriteSize(Scale Factor=&Vector2DFromFloat(Value=&FloatFromCurve(FloatCurve=DataInterface() {propertyValues {Curve {keys [
          <at 0 0 RCIM_Cubic>
          {arriveTangent=0.836, interpMode=RCIM_Cubic, leaveTangent=0.836, time=0.2, value=1}
          {arriveTangent=-1.8455, interpMode=RCIM_Cubic, leaveTangent=-1.8455, tangentMode=RCTM_User, time=0.6507, value=0.544}
          <at 1 0>
         ]}}})))
   &SolveForcesAndVelocity(Clamp Velocity=false, Limit Acceleration=false)
  ]
  rendererClasses [&NiagaraSpriteRendererProperties]
  renderers [{rendererClass=&NiagaraSpriteRendererProperties}]
 }
}
summary {
 systemName=NS_Corruption_Cube_Spawn
 emitters [
  {emitterName=Sparkles, simTarget=GPUComputeSim, rendererClasses [&NiagaraSpriteRendererProperties]}
  {emitterName=Glow, rendererClasses [&NiagaraMeshRendererProperties, &NiagaraLightRendererProperties]}
  {
   emitterName=CornerLightning
   rendererClasses [&NiagaraMeshRendererProperties, &NiagaraSpriteRendererProperties]
  }
  {emitterName=CubeMesh, rendererClasses [&NiagaraMeshRendererProperties]}
  {emitterName=Runes, rendererClasses [&NiagaraSpriteRendererProperties]}
  {emitterName=Shimmers, rendererClasses [&NiagaraSpriteRendererProperties]}
  {
   emitterName=Lightning_center
   rendererClasses [
    &NiagaraMeshRendererProperties
    &NiagaraLightRendererProperties
    &NiagaraSpriteRendererProperties
   ]
  }
  {emitterName=LingerSmoke, rendererClasses [&NiagaraSpriteRendererProperties]}
  {emitterName=NE_FN_Sparks, simTarget=GPUComputeSim, rendererClasses [&NiagaraSpriteRendererProperties]}
 ]
 userVariables [
  User.CubeRadius {type=NiagaraFloat, defaultValue=384, description=""}
  User.GlobalScale {type=NiagaraFloat, defaultValue=0, description=""}
 ]
}
