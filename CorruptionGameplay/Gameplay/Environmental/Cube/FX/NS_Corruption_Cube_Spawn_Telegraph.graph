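# Niagara system that telegraphs a Corruption cube spawn with lightning, runes, glows and impact effects.
# Lightning meshes spawn at System.LightningEndPoint and stretch toward User.SourcePoint, where glows and contracting rune rings align to SourceNormal.
# Burns, smoke, GPU sparks and shimmers play at the endpoint, while electric sprites, sparkles and a light surround the cube.
# System.SpawnOn gates many spawn rates and fades the light; cube radius and beam length drive scaling.
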NiagaraSystem /CorruptionGameplay/Gameplay/Environmental/Cube/FX/NS_Corruption_Cube_Spawn_Telegraph
dependencies {
 usedDataInterfaces [
  &NiagaraDataInterfaceArrayFloat
  &NiagaraDataInterfaceCurve
  &NiagaraDataInterfaceColorCurve
  &NiagaraDataInterfaceMeshRendererInfo
  &NiagaraDataInterfaceVectorCurve
 ]
 usedDynamicInputs [
  &Multiply_Vector2D_ByFloat
  &RemapRange
  &Waveform
  &MakeCustomFloatFromBool
  &RandomBool
  &OneMinusFloat
  /Niagara/DynamicInputs/Integer/V2/Round_Float
 ]
 usedModules [
  &self:SystemSpawnScript.SystemScriptSource.SystemScriptGraph.NiagaraNodeAssignment_2.SetVariables_#1
  &self:SystemSpawnScript.SystemScriptSource.SystemScriptGraph.NiagaraNodeAssignment_3.SetVariables_#2
  &SystemState
  &self:SystemSpawnScript.SystemScriptSource.SystemScriptGraph.NiagaraNodeAssignment_0.SetVariables_#3
  &self:SystemSpawnScript.SystemScriptSource.SystemScriptGraph.NiagaraNodeAssignment_1.SetVariables_#4
 ]
}
emitters {
 Burns {
  emitterUpdateScript [
   &EmitterState(Enable Distance Culling=true, MaxDistance=3500, Module.bUseMaxDistance=true, Module.bUseMinDistance=false, Scalability Mode=::NewEnumerator1)
   &SpawnRate(SpawnRate=&Multiply_Float(A=50, B=$System.SpawnOn), Use Spawn Probability=false)
  ]
  particleSpawnScript [
   &InitializeParticle(
    Color=&RandomRangeLinearColor(Minimum=<rgba 2 0.3572 0.1778 0.7>, Maximum=<rgba 2 0.6821 0.0104 1>)
    Lifetime Max=4
    Lifetime Min=2
    Lifetime Mode=::NewEnumerator1
    Position:Vector3f=$System.LightningEndPoint
    Position Mode=::NewEnumerator1
    Sprite Rotation Mode=::NewEnumerator2
    Sprite Size Mode=::NewEnumerator1
    Uniform Sprite Size Max=50
    Uniform Sprite Size Min=25
    UsePositionOffset=false
    Write Material Random=true
   )
   &SpriteFacingAndAlignment(Sprite Facing=<-1 0 0>, WriteAlignment=false, WriteFacing=true)
   &CameraOffset(Camera Offset Amount=20)
  ]
  particleUpdateScript [
   &ParticleState()
   &ScaleColor(
    Scale Alpha=&FloatFromCurve(FloatCurve=DataInterface() {propertyValues {Curve {keys [{interpMode=RCIM_Cubic, tangentMode=RCTM_User, time=0, value=1}, {interpMode=RCIM_Cubic, tangentMode=RCTM_User, time=1, value=0}]}}})
    Scale RGB=&VectorFromCurve(VectorCurve=DataInterface() {propertyValues {
       XCurve {keys [
         {arriveTangent=-8.6461, interpMode=RCIM_Cubic, leaveTangent=-8.6461, tangentMode=RCTM_User, time=0, value=1}
         {interpMode=RCIM_Cubic, tangentMode=RCTM_User, time=0.2, value=0}
        ]}
       YCurve {keys [
         {arriveTangent=-14.7012, interpMode=RCIM_Cubic, leaveTangent=-14.7012, tangentMode=RCTM_User, time=0, value=1}
         {arriveTangent=-5.0341, interpMode=RCIM_Cubic, leaveTangent=-5.0341, tangentMode=RCTM_User, time=0.0881, value=0.2549}
         {interpMode=RCIM_Cubic, tangentMode=RCTM_User, time=0.2, value=0}
        ]}
       ZCurve {keys [
         {arriveTangent=-17.991, interpMode=RCIM_Cubic, leaveTangent=-17.991, tangentMode=RCTM_User, time=0, value=1}
         {arriveTangent=-4.7229, interpMode=RCIM_Cubic, leaveTangent=-4.7229, tangentMode=RCTM_User, time=0.081, value=0.1992}
         {interpMode=RCIM_Cubic, tangentMode=RCTM_User, time=0.2, value=0}
        ]}
      }})
    ScaleA=true
    ScaleRGB=true
   )
  ]
  rendererClasses [&NiagaraSpriteRendererProperties]
  renderers [{rendererClass=&NiagaraSpriteRendererProperties}]
 }
 ElectricSprites {
  emitterUpdateScript [
   &EmitterState(
    Enable Distance Culling=true
    Inactive Response=::NewEnumerator1
    Life Cycle Mode=::NewEnumerator1
    Loop Duration=&UniformRangedFloat(Minimum=0.25, Maximum=1)
    MaxDistance=6500
    Module.bUseMaxDistance=true
    Module.bUseMinDistance=false
    Recalculate Duration Each Loop=true
    Scalability Mode=::NewEnumerator1
    UseLoopDelay=false
   )
   &SpawnRate(
    Spawn Probability=0.6
    SpawnRate=&Multiply_Float(A=&RandomRangeFloat(Minimum=30, Maximum=50), B=$System.CubeRadiusScale)
    Use Spawn Probability=true
   )
   &SpawnRate(
    Spawn Probability=0.5
    SpawnGroup=1
    SpawnRate=&Multiply_Float(A=25, B=$System.SpawnOn)
    Use Spawn Probability=false
   ) {module=SpawnRate001}
   &SpawnBurst_Instantaneous(Spawn Count=&UniformRangedInt(Minimum=2, Maximum=7), Use Spawn Probability=false, UseLoopCountLimit=false)
  ]
  particleSpawnScript [
   /Niagara/Modules/Spawn/Initialization/InitializeParticle(
    Color=&RandomRangeLinearColor(Minimum=<rgba 5.1665 0.2572 8 1>, Maximum=<rgba 16 0.5359 15.8881 1>)
    Lifetime=&UniformRangedFloat(Minimum=0.1, Maximum=0.25)
    Module.Write Color=true
    Module.Write Lifetime=true
    Module.Write Mass=true
    Module.Write Material Random=true
    Module.Write Position=false
    Module.Write Scale=false
    Module.Write SpriteRotation=true
    Module.Write SpriteSize=true
    Position=&SimulationPosition()
    Sprite Rotation=&UniformRangedFloat(Minimum=0, Maximum=360)
    Sprite Size=&UniformRangedVector2D(Minimum=<50 50>, Maximum=<140 140>)
   )
   &BoxLocation(
    Box Size=&Multiply_VectorByFloat(Vector=&MakeVector(X=15, Y=$User.CubeRadius, Z=$User.CubeRadius), Float=1.5)
    MaskSpawnGroup=false
    Module.Override Seed=false
    Rotate Box=false
    Surface Only=false
   )
   /SetVariables(Particles.Position=&MaskVectorBySpawnGroup(
     Masked Vector=Linked() {linkedVariable {name=Particles.Position, type=NiagaraPosition}}
     Passthrough Vector=/Niagara/DynamicInputs/Transforms/TransformPosition(Position=$System.LightningEndPoint, Source Space=::Local, Destination Space=::Simulation)
    )) {isSetParameters=true}
   &DynamicMaterialParameters(
    Index 0 Param 1=$Particles.MaterialRandom
    Index 0 Param 2=&Lerp_Float(A=2, B=4, Alpha=$Particles.MaterialRandom)
    Module.Param0WriteEnabled=true
    Module.Param1WriteEnabled=true
    Module.Param2WriteEnabled=false
    Module.Param3WriteEnabled=false
   )
   &CameraOffset(Camera Offset Amount=70)
   &SubUVAnimation(End Frame=3, SubUV Animation Mode:ENiagara_SubUVLookupMode=::NewEnumerator2) {module=SubUVAnimation001}
  ]
  particleUpdateScript [
   &ParticleState()
   &DynamicMaterialParameters(
    Index 0 Param 3=&FloatFromCurve(FloatCurve=DataInterface() {propertyValues {Curve {keys [<at 0.2 0 RCIM_Cubic>, {interpMode=RCIM_Cubic, leaveTangent=1.5819, time=1, value=1}]}}})
    Module.Param0WriteEnabled=false
    Module.Param1WriteEnabled=false
    Module.Param2WriteEnabled=true
    Module.Param3WriteEnabled=false
   ) {module=DynamicMaterialParameters001}
   &ScaleSpriteSize(Scale Factor=&Vector2DFromFloat(Value=&FloatFromCurve(FloatCurve=DataInterface() {propertyValues {Curve {keys [{arriveTangent=0.3147, interpMode=RCIM_Cubic, leaveTangent=0.3147, tangentMode=RCTM_User, time=0, value=0.8}, <at 1 1>]}}})))
  ]
  rendererClasses [&NiagaraSpriteRendererProperties]
  renderers [{rendererClass=&NiagaraSpriteRendererProperties}]
 }
 GlowSurface {
  emitterUpdateScript [
   &EmitterState()
   &SpawnRate(SpawnRate=2.5, Use Spawn Probability=false)
  ]
  particleSpawnScript [
   /Niagara/Modules/Spawn/Initialization/V2/InitializeParticle(
    Color=<rgba 2.9034 0.5 3 0.7>
    Lifetime Max=2.8
    Lifetime Min=2.2
    Lifetime Mode=::NewEnumerator1
    Position=$User.SourcePoint
    Position Mode=::NewEnumerator1
    Sprite Rotation Mode=::NewEnumerator2
    Sprite Size Mode=::NewEnumerator3
    Uniform Sprite Size=&RandomRangeFloat(Minimum=700, Maximum=1000)
    UsePositionOffset=false
    Write Material Random=true
   )
   &SpriteFacingAndAlignment(Sprite Facing=$User.SourceNormal, WriteAlignment=true, WriteFacing=true)
   &CylinderLocation(
    Cylinder Axis:Vector3f=$User.SourceNormal
    Cylinder Height=10
    Cylinder Radius=50
    MaskSpawnGroup=false
    Override Seed=false
    Rotation Coordinate Space:ENiagaraCoordinateSpace=::Simulation
    Rotation Mode:ENiagara_RotationMode=::NewEnumerator1
    UseOwnerScale=true
   )
   &DynamicMaterialParameters(
    Index 0 Param 1=&RandomRangeFloat(Minimum=0, Maximum=1)
    Index 0 Param 2=&RandomRangeFloat(Minimum=0, Maximum=1)
    Module.Param0WriteEnabled=true
    Module.Param1WriteEnabled=true
    Module.Param2WriteEnabled=true
    Module.Param3WriteEnabled=false
   )
   &CameraOffset(Camera Offset Amount=5)
  ]
  particleUpdateScript [
   &ParticleState()
   &SpriteRotationRate(Rotation Rate=&Multiply_Float(
     A=&RandomRangeFloat(Minimum=4, Maximum=15)
     B=&MaskFloatBySpawnGroup(Masked Float=1, Passthrough Float=-1)
    ))
   &ScaleColor(
    Scale Alpha=&FloatFromCurve(FloatCurve=DataInterface() {propertyValues {Curve {keys [
         {interpMode=RCIM_Cubic, tangentMode=RCTM_User, time=0, value=0}
         <at 0.4 1>
         {interpMode=RCIM_Cubic, tangentMode=RCTM_User, time=0.7, value=1}
         {tangentMode=RCTM_User, time=1, value=0}
        ]}}})
    ScaleA=true
    ScaleRGB=false
   )
   &ScaleSpriteSize(
    Scale Sprite Size Mode=::NewEnumerator0
    Uniform Scale Factor=&FloatFromCurve(FloatCurve=DataInterface() {propertyValues {Curve {keys [
         {interpMode=RCIM_Cubic, tangentMode=RCTM_User, time=0, value=1}
         {arriveTangent=-1.1258, interpMode=RCIM_Cubic, leaveTangent=-1.1258, tangentMode=RCTM_User, time=1, value=0.1}
        ]}}})
   )
  ]
  rendererClasses [&NiagaraSpriteRendererProperties]
  renderers [{rendererClass=&NiagaraSpriteRendererProperties}]
 }
 ImpactSmoke {
  emitterUpdateScript [
   &EmitterState(Enable Distance Culling=true, MaxDistance=8500, Module.bUseMaxDistance=true, Module.bUseMinDistance=false, Scalability Mode=::NewEnumerator1)
   &SpawnRate(SpawnRate=&Multiply_Float(A=15, B=$System.SpawnOn), Use Spawn Probability=false)
  ]
  particleSpawnScript [
   /Niagara/Modules/Spawn/Initialization/InitializeParticle(
    Color=&UniformRangedLinearColor(Minimum=<rgba 0.091 0.0276 0.1 0.55>, Maximum=<rgba 0.1387 0.0414 0.15 0.75>)
    Lifetime=&RandomRangeFloat(Minimum=1, Maximum=1.5)
    Module.Write Color=true
    Module.Write Lifetime=true
    Module.Write Mass=false
    Module.Write Material Random=true
    Module.Write Position=true
    Module.Write Scale=false
    Module.Write SpriteRotation=true
    Module.Write SpriteSize=true
    Position=&TransformPosition(Position=$System.LightningEndPoint, Source Space=::Local, Destination Space=::Simulation)
    Sprite Rotation=&UniformRangedFloat(Minimum=0, Maximum=360)
    Sprite Size=&Vector2DFromFloat(Value=&RandomRangeFloat(Minimum=500, Maximum=700))
   )
   /Niagara/Modules/Spawn/Location/SphereLocation(
    Hemisphere Z=true
    MaskSpawnGroup=false
    Module.Override Randomness=false
    Non Uniform Scale=<1 1 0.1>
    Offset=<0 0 -50>
    Sphere Radius=50
    Surface Only=true
    Surface Only Band Thickness=1
   )
   &AddVelocityInCone(
    Cone Angle=55
    Cone Axis=&Lerp_Vector(
     A=<0 0 1>
     B=&NormalizeVector(
      Vector To Normalize=&Subtract_Vector(
       A=Linked() {linkedVariable {name=Particles.Position, type=NiagaraPosition}}
       B=Linked() {linkedVariable {name=User.SourcePoint, type=NiagaraPosition}}
      )
      Safe Normalize=true
      Safe Fallback Threshold=0
      Safe Fallback Vector=<0 0 0.001>
     )
     Alpha=0.4
    )
    Cone Axis Coordinate Space=::World
    Module.Use Velocity Falloff On Cone Axis=false
    Velocity Strength=&RandomRangeFloat(Minimum=400, Maximum=2000)
   )
  ]
  particleUpdateScript [
   &ParticleState()
   &Drag(Drag=&RandomRangeFloat(Minimum=2, Maximum=3), Use Linear Drag=true, Use Rotational Drag=false)
   &SpriteRotationRate(Rotation Rate=&Multiply_Float(
     A=&Lerp_Float(A=-300, B=300, Alpha=$Particles.MaterialRandom)
     B=&FloatFromCurve(FloatCurve=DataInterface() {propertyValues {Curve {keys [
          {arriveTangent=-4.4572, interpMode=RCIM_Cubic, leaveTangent=-4.4572, tangentMode=RCTM_User, time=0, value=1}
          {arriveTangent=-1.0628, interpMode=RCIM_Cubic, leaveTangent=-1.0628, tangentMode=RCTM_User, time=0.2, value=0.4}
          <at 1.0011 0.01>
         ]}}})
    ))
   &AccelerationForce(Acceleration=<0 0 40>)
   &SolveForcesAndVelocity(Clamp Velocity=false, Limit Acceleration=false)
   &CameraOffset(Camera Offset Amount=&Multiply_Float(
     A=0.7
     B=/Niagara/DynamicInputs/Math/CalculateParticleRadius(
      Radius Calculation Type:ENiagaraCollisionRadiusOptions=::NewEnumerator0
      Particle Radius:NiagaraFloat=1
      Mesh Dimensions:Vector3f=<1 1 1>
      Method for Calculating Particle Radius:ENiagaraCalculateRadiusOptions=::NewEnumerator2
     )
    ))
   &ScaleSpriteSize(Scale Factor=&Vector2DFromFloat(Value=&FloatFromCurve(FloatCurve=DataInterface() {propertyValues {Curve {keys [
          {arriveTangent=3.5786, interpMode=RCIM_Cubic, leaveTangent=3.5786, tangentMode=RCTM_User, time=0, value=0.1}
          {arriveTangent=1, interpMode=RCIM_Cubic, leaveTangent=1, time=0.3314, value=0.8}
          {arriveTangent=0.5339, interpMode=RCIM_Cubic, leaveTangent=-0.3011, tangentMode=RCTM_User, time=1, value=1.25}
         ]}}})))
   &ScaleColor(
    Scale Alpha=&FloatFromCurve(FloatCurve=DataInterface() {propertyValues {Curve {keys [<at 0 0 RCIM_Cubic>, <at 0.25 1 RCIM_Cubic>, {interpMode=RCIM_Cubic, leaveTangent=-3.4295, time=1, value=0}]}}})
    Scale RGB=&VectorFromFloat(Value=&FloatFromCurve(
      FloatCurve=DataInterface() {propertyValues {Curve {keys [{arriveTangent=-366.0958, interpMode=RCIM_Cubic, leaveTangent=-366.0958, tangentMode=RCTM_User, time=0, value=40}, <at 0.25 1>]}}}
      CurveIndex=$Particles.Age
     ))
    ScaleA=true
    ScaleRGB=true
   )
   /SetVariables(Particles.SubImageIndex=&Lerp_Float(A=20, B=63, Alpha=$Particles.NormalizedAge)) {isSetParameters=true}
  ]
  rendererClasses [&NiagaraSpriteRendererProperties]
  renderers [{rendererClass=&NiagaraSpriteRendererProperties}]
 }
 Light {
  emitterUpdateScript [
   &EmitterState(
    Inactive Response=::NewEnumerator1
    Life Cycle Mode=::NewEnumerator1
    Loop Behavior=::NewEnumerator1
    Loop Duration Mode=::NewEnumerator1
    UseLoopDelay=false
   )
   &SpawnBurst_Instantaneous(Use Spawn Probability=false, UseLoopCountLimit=false)
  ]
  particleSpawnScript [/Niagara/Modules/Spawn/Initialization/V2/InitializeParticle(
    Color=<rgba 15 1.374 2.2699 1>
    Mesh Scale Mode=::NewEnumerator3
    Mesh Uniform Scale=25
    Position Offset=<-180 0 0>
    UsePositionOffset=true
    Write Material Random=true
   )]
  particleUpdateScript [
   &ParticleState(Kill Particles When Lifetime Has Elapsed=false, Let Infinitely Lived Particles Die When Emitter Deactivates=true)
   /SetVariables(Particles.LightRadius=&Multiply_Float(A=900, B=$System.CubeRadiusScale)) {isSetParameters=true}
   &InterpolateOverTime(Rate Of Change=10, Target Float Value:NiagaraFloat=$System.SpawnOn)
   &ScaleColor(
    Scale Alpha=&FloatFromCurve(
     FloatCurve=DataInterface() {propertyValues {Curve {keys [{interpMode=RCIM_Cubic, tangentMode=RCTM_User, time=0, value=0}, {interpMode=RCIM_Cubic, tangentMode=RCTM_User, time=1, value=1}]}}}
     Scale Curve=$Particles.InterpolateOverTime.InterpolateOverTime_Float.Moving Average
    )
    ScaleA=true
    ScaleRGB=true
   )
   /SetVariables(Particles.LightEnabled:NiagaraBool=&SetBoolByFloatComparison(A=$Particles.InterpolateOverTime.InterpolateOverTime_Float.Moving Average, B=0.1)) {isSetParameters=true}
  ]
  rendererClasses [&NiagaraLightRendererProperties, &NiagaraMeshRendererProperties]
  renderers [
   {rendererClass=&NiagaraLightRendererProperties, rendererIndex=0}
   {rendererClass=&NiagaraMeshRendererProperties, rendererIndex=1}
  ]
 }
 Lightning {
  emitterUpdateScript [
   &EmitterState(
    Enable Distance Culling=true
    Life Cycle Mode=::NewEnumerator1
    Loop Duration=&RandomRangeFloat(Minimum=0.5, Maximum=1)
    MaxDistance=25000
    Module.bUseMaxDistance=true
    Module.bUseMinDistance=false
    Recalculate Duration Each Loop=true
    Scalability Mode=::NewEnumerator1
    UseLoopDelay=false
   )
   /SetVariables(Emitter.PerLoopRandom:NiagaraFloat=&RandomRangeFloat(Minimum=0, Maximum=1)) {isSetParameters=true}
   &SpawnRate(SpawnRate=&Multiply_Float(A=25, B=$System.SpawnOn), Use Spawn Probability=false)
   &SpawnRate(SpawnGroup=1, SpawnRate=&Multiply_Float(A=25, B=$System.SpawnOn), Use Spawn Probability=false) {module=SpawnRate001}
  ]
  particleSpawnScript [
   /Niagara/Modules/Spawn/Initialization/InitializeParticle(
    Color=&RandomRangeLinearColor(Minimum=<rgba 0.2134 0.1969 0.25 1>, Maximum=<rgba 0.3666 0.414 0.5 1>)
    Lifetime=&SelectFloatFromArray(
     Direct Array Index=&SpawnGroup()
     Interpolation=&MaskFloatBySpawnGroup(Masked Float=0, Passthrough Float=1)
     Float Selection Array=DataInterface() {propertyValues {FloatData [0.12,0.07]}}
    )
    Mesh Scale=&VectorFromFloat(Value=2)
    Module.Write Color=true
    Module.Write Lifetime=true
    Module.Write Mass=true
    Module.Write Material Random=true
    Module.Write Position=false
    Module.Write Scale=true
    Module.Write SpriteRotation=false
    Module.Write SpriteSize=true
    Position=&SimulationPosition()
    Sprite Size=&Vector2DFromFloat(Value=&RandomRangeFloat(Minimum=400, Maximum=700))
   )
   /SetVariables(Particles.Position=&TransformPosition(Position=$System.LightningEndPoint, Source Space=::Local, Destination Space=::Simulation)) {isSetParameters=true}
   /SetVariables(
    Particles.LightEnabled:NiagaraBool=&SetBoolByIntComparison(A=&SpawnGroup(), B=0, Comparison Type=::NewEnumerator0)
    Particles.LightPosition:Vector3f=&Add_Vector(A=Linked() {linkedVariable {name=Particles.Position, type=NiagaraPosition}}, B=<0 0 160>)
   ) {isSetParameters=true}
   &InitialMeshOrientation(
    Rotation=&MakeVector(
     X=&Add_Float(A=&Multiply_Float(A=$Emitter.NormalizedLoopAge, B=0.5), B=$Emitter.PerLoopRandom)
     Y=0
     Z=0
    )
    Use Orientation Vector=true
    Use Rotation Vector=true
   ) {module=InitialMeshOrientation001}
   /Niagara/Modules/Update/Orientation/V2/OrientMeshToVector(
    Look At Direction:Vector3f=&Subtract_Vector(
     A=Linked() {linkedVariable {name=User.SourcePoint, type=NiagaraPosition}}
     B=Linked() {linkedVariable {name=Particles.Position, type=NiagaraPosition}}
    )
    Override Quat=false
    UseOptionalSideAxis=false
    UseOptionalUpAxis=false
   )
   /SetVariables(Particles.Scale=&MakeVector(
     X=&Divide_Float(
      A=&VectorLength(Vector=&Subtract_Vector(
        A=Linked() {linkedVariable {name=User.SourcePoint, type=NiagaraPosition}}
        B=Linked() {linkedVariable {name=Particles.Position, type=NiagaraPosition}}
       ))
      B=1000
     )
     Y=&RandomRangeFloat(Minimum=0.8, Maximum=1)
     Z=&RandomRangeFloat(Minimum=0.8, Maximum=1)
    )) {isSetParameters=true}
   /SetVariables(Particles.DynParamsSprite:Vector4f=<1 2 1 1>, Particles.DynParamsSprite2:Vector4f=<40 0 0 0>) {isSetParameters=true}
   /SetVariables(Particles.LightRadius=&RandomRangeFloat(Minimum=50, Maximum=150)) {isSetParameters=true}
   /SetVariables(Particles.CamOffsetSprite=100) {isSetParameters=true}
   &CameraOffset(Camera Offset Amount=70)
   /SetVariables(Particles.MeshIndex=&RandomRangeInt(
     Minimum=0
     Maximum=&MakeCustomIntFromBool(Bool=&SetBoolByFloatComparison(A=$System.BeamLength, B=2200), True INT=1, False INT=0)
    )) {isSetParameters=true}
  ]
  particleUpdateScript [
   &ParticleState()
   &ScaleColor(
    Scale Alpha=&FloatFromCurve(FloatCurve=DataInterface() {propertyValues {Curve {keys [
         <at 0 0>
         {interpMode=RCIM_Cubic, tangentMode=RCTM_User, time=0.1258, value=0.1952}
         <at 0.341 0.0902>
         {interpMode=RCIM_Cubic, tangentMode=RCTM_User, time=0.496, value=0.2532}
         <at 1 0>
        ]}}})
    Scale RGB=&VectorFromFloat(Value=&FloatFromCurve(FloatCurve=DataInterface() {propertyValues {Curve {keys [
          {arriveTangent=3.5209, interpMode=RCIM_Cubic, leaveTangent=3.5209, tangentMode=RCTM_User, time=0, value=0.5}
          {arriveTangent=-20.2701, interpMode=RCIM_Cubic, leaveTangent=-20.2701, tangentMode=RCTM_User, time=0.45, value=5}
          <at 1 1>
         ]}}}))
    ScaleA=true
    ScaleRGB=true
   )
   /SetVariables(Particles.SpriteColor=&MultiplyLinearColor(
     Linear Color=$Particles.Color
     Scale Factor=&MakeVector4_FromLinearColor(LinearColor=&MultiplyLinearColor(
       Linear Color=<rgba 47.3998 4.5799 50 1>
       Scale Factor=&MakeVector4(
        X=1
        Y=1
        Z=1
        W=&FloatFromCurve(FloatCurve=DataInterface() {propertyValues {Curve {keys [
             {interpMode=RCIM_Cubic, tangentMode=RCTM_User, time=0, value=1}
             {arriveTangentWeight=0.031, interpMode=RCIM_Cubic, leaveTangentWeight=0.0366, tangentMode=RCTM_User, time=0.0918, value=0.05}
             {
              arriveTangentWeight=0.16
              interpMode=RCIM_Cubic
              leaveTangent=18.7153
              leaveTangentWeight=0.1479
              tangentMode=RCTM_Break
              time=0.1619
              value=0
             }
             {arriveTangent=-0.2815, interpMode=RCIM_Cubic, leaveTangent=-0.2815, tangentMode=RCTM_User, time=0.3, value=1}
             {arriveTangentWeight=0.0345, interpMode=RCIM_Cubic, leaveTangentWeight=0.0566, tangentMode=RCTM_User, time=0.4, value=0.1}
             {
              arriveTangentWeight=0.1589
              interpMode=RCIM_Cubic
              leaveTangent=29.6956
              leaveTangentWeight=0.1018
              tangentMode=RCTM_Break
              time=0.4765
              value=0.05
             }
             {arriveTangent=0.512, interpMode=RCIM_Cubic, leaveTangent=0.512, tangentMode=RCTM_User, time=0.5641, value=0.8648}
             <at 1 0>
            ]}}})
       )
      ))
    )) {isSetParameters=true}
   &DynamicMaterialParameters(
    Index 0 Param 1=&FloatFromCurve(FloatCurve=DataInterface() {propertyValues {Curve {keys [
         {arriveTangent=-3.0892, interpMode=RCIM_Cubic, leaveTangent=-3.0892, tangentMode=RCTM_User, time=0, value=1}
         {
          arriveTangent=-0.6954
          arriveTangentWeight=0.2357
          interpMode=RCIM_Cubic
          leaveTangent=-0.6954
          leaveTangentWeight=0.2357
          tangentMode=RCTM_User
          time=0.5
          value=0.5
         }
         {arriveTangent=-1.3762, interpMode=RCIM_Cubic, leaveTangent=-1.3762, tangentMode=RCTM_User, time=1, value=0}
        ]}}})
    Index 0 Param 2=&FloatFromCurve(FloatCurve=DataInterface() {propertyValues {Curve {keys [
         {interpMode=RCIM_Cubic, leaveTangent=8.981, tangentMode=RCTM_Break, time=0, value=0}
         {arriveTangent=0.0136, interpMode=RCIM_Cubic, leaveTangent=-5.3647, tangentMode=RCTM_Break, time=0.4833, value=1.5}
         <at 1 -0.3>
        ]}}})
    Index 0 Param 3=&RandomRangeFloat(Minimum=1, Maximum=5)
    Index 1 Param 1=&Multiply_Float(A=&Lerp_Float(A=3, B=1, Alpha=$Emitter.PerLoopRandom), B=$System.Age)
    Index 1 Param 2=&FloatFromCurve(
     FloatCurve=DataInterface() {propertyValues {Curve {keys [
         {arriveTangent=-8.2706, interpMode=RCIM_Cubic, leaveTangent=-8.2706, tangentMode=RCTM_User, time=0, value=1}
         <at 0.25 0 RCIM_Cubic>
         {arriveTangent=-0.0405, interpMode=RCIM_Cubic, leaveTangent=-0.0405, tangentMode=RCTM_User, time=1, value=1}
        ]}}}
     CurveIndex=$Emitter.NormalizedLoopAge
     Scale Curve=&Lerp_Float(A=150, B=250, Alpha=$Emitter.PerLoopRandom)
    )
    Index 1 Param 3=&Lerp_Float(A=0.2, B=2, Alpha=$Emitter.PerLoopRandom)
    Index 1 Param 4=200
    Module.Param0WriteEnabled=true
    Module.Param0WriteEnabled1=true
    Module.Param1WriteEnabled=true
    Module.Param1WriteEnabled1=true
    Module.Param2WriteEnabled=true
    Module.Param2WriteEnabled1=true
    Module.Param3WriteEnabled=false
    Module.Param3WriteEnabled1=true
    Write Parameter Index 1=true
   ) {module=DynamicMaterialParameters001}
   &ScaleSpriteSize(Uniform Curve Sprite Scale=DataInterface() {propertyValues {Curve {keys [
        {arriveTangent=21.5526, interpMode=RCIM_Cubic, leaveTangent=21.5526, tangentMode=RCTM_User, time=0, value=0}
        {arriveTangent=0.5, interpMode=RCIM_Cubic, leaveTangent=0.5, time=0.1, value=1}
        {arriveTangent=-1.8758, interpMode=RCIM_Cubic, leaveTangent=-1.8758, tangentMode=RCTM_User, time=1, value=0}
       ]}}})
  ]
  rendererClasses [
   &NiagaraMeshRendererProperties
   &NiagaraSpriteRendererProperties
   &NiagaraLightRendererProperties
  ]
  renderers [
   {rendererClass=&NiagaraMeshRendererProperties, rendererIndex=0}
   {rendererClass=&NiagaraSpriteRendererProperties, rendererIndex=1}
   {rendererClass=&NiagaraLightRendererProperties, rendererIndex=2}
   {rendererClass=&NiagaraSpriteRendererProperties, rendererIndex=3}
  ]
 }
 NE_FN_Sparks {
  simTarget=GPUComputeSim
  emitterUpdateScript [
   &EmitterState(
    Enable Distance Culling=true
    Life Cycle Mode=::NewEnumerator1
    Loop Duration=&RandomRangeFloat(Minimum=0.15, Maximum=0.4)
    MaxDistance=8000
    Module.bUseMaxDistance=true
    Module.bUseMinDistance=false
    Recalculate Duration Each Loop=true
    Scalability Mode=::NewEnumerator1
    UseLoopDelay=false
   )
   &SpawnBurst_Instantaneous(
    Spawn Count=&MakeIntFromFloat(Float=&Multiply_Float(A=&RandomRangeFloat(Minimum=25, Maximum=40), B=$System.SpawnOn))
    Spawn Time=0.05
    Use Spawn Probability=false
   )
   &SpawnRate(SpawnRate=&Multiply_Float(A=25, B=$System.SpawnOn), Use Spawn Probability=false)
  ]
  particleSpawnScript [
   /SetVariables(Particles.SparkIntensity:NiagaraFloat=&FloatFromCurve(
     FloatCurve=DataInterface() {propertyValues {Curve {keys [
         <at 0 1 RCIM_Cubic>
         {arriveTangent=-2.8017, interpMode=RCIM_Cubic, leaveTangent=-2.8017, tangentMode=RCTM_User, time=0.6, value=0.5}
         <at 1 0 RCIM_Cubic>
        ]}}}
     CurveIndex=&RandomRangeFloat(Minimum=0, Maximum=1)
    )) {isSetParameters=true}
   /Niagara/Modules/Spawn/Initialization/V2/InitializeParticle(
    Color=&LerpLinearColors(StartColor=<rgba 2.1878 0.6732 10 1>, EndColor=<rgba 16.4036 1.3193 25 1>, LerpFactor=$Particles.SparkIntensity)
    Lifetime Max=0.7
    Lifetime Min=0.12
    Lifetime Mode=::NewEnumerator1
    Mass=&Lerp_Float(A=0.3, B=1, Alpha=$Particles.SparkIntensity)
    Mass Mode=::NewEnumerator1
    Position=&TransformPosition(Position=$System.LightningEndPoint, Source Space=::Local, Destination Space=::Simulation)
    Position Mode=::NewEnumerator1
    Sprite Size Mode=::NewEnumerator3
    Uniform Sprite Size=&Lerp_Float(A=3, B=5, Alpha=$Particles.SparkIntensity)
    UsePositionOffset=false
    Write Material Random=true
   )
   /SetVariables(Particles.SpriteSize=&Multiply_Vector2(
     A=$Particles.SpriteSize
     B=&Vector2DFromFloat(Value=&FloatFromCurve(FloatCurve=DataInterface() {propertyValues {Curve {keys [<at 200 1 RCIM_Cubic>,<at 7000 5>]}}}, CurveIndex=$Engine.Owner.LODDistance))
    )) {isSetParameters=true}
   /Niagara/Modules/Spawn/Location/V2/SphereLocation(
    Hemisphere X=true
    MaskSpawnGroup=false
    Non Uniform Scale=<0.5 1 1>
    Override Seed=false
    Sphere Radius=&SelectFloatFromArray(Direct Array Index=&SpawnGroup(), Float Selection Array=DataInterface() {propertyValues {FloatData [5,12]}})
    Surface Only=false
    UseNonUniformScale=true
    UseOffset=false
    UseSphereOrientation=false
   ) {enabled=false}
   &AddVelocityInCone(
    Cone Angle=60
    Cone Axis=&Lerp_Vector(
     A=<0 0 1>
     B=&NormalizeVector(
      Vector To Normalize=&Subtract_Vector(
       A=Linked() {linkedVariable {name=Particles.Position, type=NiagaraPosition}}
       B=Linked() {linkedVariable {name=User.SourcePoint, type=NiagaraPosition}}
      )
      Safe Normalize=true
      Safe Fallback Threshold=0
      Safe Fallback Vector=<0 0 0.001>
     )
     Alpha=0.5
    )
    Cone Axis Coordinate Space=::World
    Module.Use Velocity Falloff On Cone Axis=false
    Velocity Distribution Along Cone Axis=0.75
    Velocity Strength=&Lerp_Float(
     A=&RandomRangeFloat(Minimum=400, Maximum=800)
     B=&RandomRangeFloat(Minimum=1500, Maximum=4000)
     Alpha=$Particles.SparkIntensity
    )
   )
   &DynamicMaterialParameters(
    Index 0 Param 1=2
    Index 0 Param 3=3
    Index 0 Param 4=1500
    Module.Param0WriteEnabled=true
    Module.Param1WriteEnabled=false
    Module.Param2WriteEnabled=true
    Module.Param3WriteEnabled=false
   )
   &CameraOffset(Camera Offset Amount=50)
  ]
  particleUpdateScript [
   &ParticleState()
   &GravityForce(Gravity=<0 0 -2200>)
   &Drag(Drag=2, Use Linear Drag=true, Use Rotational Drag=false)
   &VectorNoiseForce(
    Force Amount=&FloatFromCurve(
     FloatCurve=DataInterface() {propertyValues {Curve {keys [
         {arriveTangent=14.9491, interpMode=RCIM_Cubic, leaveTangent=14.9491, tangentMode=RCTM_User, time=0, value=0}
         {arriveTangent=-3.0044, interpMode=RCIM_Cubic, leaveTangent=-3.0044, tangentMode=RCTM_User, time=0.2, value=1}
         {tangentMode=RCTM_User, time=1, value=0.2}
        ]}}}
     Scale Curve=2500
    )
    Module.Use Random Noise Time Interval=false
   )
   &SolveForcesAndVelocity(Clamp Velocity=false, Limit Acceleration=false)
   &ScaleColor(
    Linear Color Curve=DataInterface() {propertyValues {
      AlphaCurve {keys [<at 0 1>,<at 1 1>]}
      BlueCurve {keys [
        <at 0 1.5 RCIM_Cubic>
        {arriveTangent=-1.452, interpMode=RCIM_Cubic, leaveTangent=-1.452, tangentMode=RCTM_User, time=0.1467, value=1}
        <at 1 0.25>
       ]}
      GreenCurve {keys [
        <at 0 1.5 RCIM_Cubic>
        {arriveTangent=-1.4408, interpMode=RCIM_Cubic, leaveTangent=-1.4408, tangentMode=RCTM_User, time=0.1467, value=1}
        <at 1 0.1698>
       ]}
      RedCurve {keys [
        <at 0 1.5 RCIM_Cubic>
        {arriveTangent=-0.8332, interpMode=RCIM_Cubic, leaveTangent=-0.8332, tangentMode=RCTM_User, time=0.1467, value=1}
        <at 1 0.1219>
       ]}
     }}
    Scale Mode=::NewEnumerator2
   )
   &ScaleSpriteSize(Scale Factor=&Vector2DFromFloat(Value=&FloatFromCurve(FloatCurve=DataInterface() {propertyValues {Curve {keys [<at 0.6 1 RCIM_Cubic>, {arriveTangent=-5.5202, interpMode=RCIM_Cubic, leaveTangent=-5.5202, tangentMode=RCTM_User, time=1, value=0}]}}})))
   &ScaleSpriteSizeBySpeed(
    Max Scale Factor=<1 6>
    Sample Scale Factor By Curve=true
    Scale Factor Curve=DataInterface() {propertyValues {Curve {keys [
        {arriveTangent=0.2757, time=0, value=0}
        {arriveTangent=0.5388, interpMode=RCIM_Cubic, leaveTangent=0.5388, tangentMode=RCTM_User, time=0.6, value=0.3}
        {arriveTangent=2.7622, interpMode=RCIM_Cubic, leaveTangent=2.7622, tangentMode=RCTM_User, time=1, value=1}
       ]}}}
    Velocity Threshold=3000
   )
  ]
  rendererClasses [&NiagaraSpriteRendererProperties]
  renderers [{rendererClass=&NiagaraSpriteRendererProperties}]
 }
 RunesSource {
  emitterSpawnScript [/SetVariables(Emitter.RuneSize=64) {isSetParameters=true}]
  emitterUpdateScript [
   &EmitterState(
    Enable Distance Culling=true
    Life Cycle Mode=::NewEnumerator1
    Loop Duration=3
    MaxDistance=8000
    Module.bUseMaxDistance=true
    Module.bUseMinDistance=false
    Scalability Mode=::NewEnumerator1
    UseLoopDelay=false
   )
   &SpawnRate(
    SpawnRate=&FloatFromCurve(
     FloatCurve=DataInterface() {propertyValues {Curve {keys [
         <at 0 0 RCIM_Cubic>
         {arriveTangentWeight=0.0471, interpMode=RCIM_Cubic, leaveTangentWeight=0.0707, time=0.6, value=1}
         {interpMode=RCIM_Cubic, leaveTangent=-5.8264, time=1, value=0}
        ]}}}
     CurveIndex=$Emitter.NormalizedLoopAge
     Scale Curve=100
    )
    Use Spawn Probability=false
   )
  ]
  particleSpawnScript [
   &InitializeParticle(
    Color Maximum=<rgba 1 15 12 0.9>
    Color Minimum=<rgba 0 2 1 0.7>
    Color Mode=::NewEnumerator2
    Lifetime Max=0.75
    Lifetime Min=0.05
    Lifetime Mode=::NewEnumerator1
    Position=$User.SourcePoint
    Position Mode=::NewEnumerator1
    Sprite Rotation Angle=&SelectFloatFromArray(
     Array Sampling Mode=::NewEnumerator0
     Direct Array Index=?
     Float Selection Array=DataInterface() {propertyValues {FloatData [0,0.25,0.5,0.75]}}
     Random Seed=0
    )
    Sprite Rotation Mode=::NewEnumerator1
    Sprite Size Mode=::NewEnumerator3
    Uniform Sprite Size=$Emitter.RuneSize
    UsePositionOffset=false
    Write Material Random=true
   )
   &CylinderLocation(
    Cylinder Axis:Vector3f=$User.SourceNormal
    Cylinder Height=2
    Cylinder Radius=&FloatFromCurve(
     FloatCurve=DataInterface() {propertyValues {Curve {keys [
         {arriveTangent=-0.3988, interpMode=RCIM_Cubic, leaveTangent=-0.3988, tangentMode=RCTM_User, time=0, value=1}
         {arriveTangent=-1.4812, interpMode=RCIM_Cubic, leaveTangent=-1.4812, tangentMode=RCTM_User, time=1, value=0}
        ]}}}
     CurveIndex=$Emitter.NormalizedLoopAge
     Scale Curve=&RandomRangeFloat(Minimum=350, Maximum=400)
    )
    MaskSpawnGroup=false
    Override Seed=false
    Rotation Coordinate Space:ENiagaraCoordinateSpace=::Simulation
    Rotation Mode:ENiagara_RotationMode=::NewEnumerator1
    Surface Only=true
    Use Endcaps In Surface Only Mode=false
    UseOwnerScale=true
   )
   &CameraOffset(Camera Offset Amount=12)
   &SpriteFacingAndAlignment(Sprite Alignment=$User.SourceUp, Sprite Facing=$User.SourceNormal, WriteAlignment=true, WriteFacing=true)
  ]
  particleUpdateScript [
   &ParticleState()
   &SubUVAnimation(
    End Frame=4
    Sub UV Play Rate:NiagaraFloat=&RandomRangeFloat(Minimum=0.5, Maximum=2)
    SubUV Animation Mode:ENiagara_SubUVLookupMode=::NewEnumerator3
    SubUV Start Time Offset:NiagaraFloat=&RandomRangeFloat(Minimum=0, Maximum=1)
   )
   &ScaleColor(
    Scale Alpha=&FloatFromCurve(FloatCurve=DataInterface() {propertyValues {Curve {keys [
         {arriveTangent=-3.8362, interpMode=RCIM_Cubic, leaveTangent=-3.8362, tangentMode=RCTM_User, time=0, value=1}
         <at 0.1866 0.4>
         {arriveTangent=-4.1865, interpMode=RCIM_Cubic, leaveTangent=-4.1865, tangentMode=RCTM_User, time=0.2015, value=1}
         <at 0.4552 0.3>
         {arriveTangent=-8.2931, interpMode=RCIM_Cubic, leaveTangent=-8.2931, tangentMode=RCTM_User, time=0.4652, value=0.75}
         <at 0.7089 0>
         {arriveTangent=-5.5334, interpMode=RCIM_Cubic, leaveTangent=-5.5334, tangentMode=RCTM_User, time=0.7338, value=0.5}
         <at 1 0>
        ]}}})
    Scale RGB=&VectorFromCurve(VectorCurve=DataInterface() {propertyValues {
       XCurve {keys [<at 0 1>,<at 1 1>]}
       YCurve {keys [<at 0 1>,<at 1 1>]}
       ZCurve {keys [
         {arriveTangent=1.0471, interpMode=RCIM_Cubic, leaveTangent=1.0471, tangentMode=RCTM_User, time=0, value=0.8}
         {interpMode=RCIM_Cubic, tangentMode=RCTM_User, time=0.25, value=1}
         {arriveTangent=-2.0404, interpMode=RCIM_Cubic, leaveTangent=-2.0404, tangentMode=RCTM_User, time=1, value=0}
        ]}
      }})
    ScaleA=true
    ScaleRGB=true
   )
   &ScaleSpriteSize(Scale Sprite Size Mode=::NewEnumerator0)
  ]
  rendererClasses [&NiagaraSpriteRendererProperties]
  renderers [{rendererClass=&NiagaraSpriteRendererProperties}]
 }
 Shimmers {
  emitterUpdateScript [
   &EmitterState(
    Enable Distance Culling=true
    Life Cycle Mode=::NewEnumerator1
    Loop Duration=1
    MaxDistance=10000
    Module.bUseMaxDistance=true
    Module.bUseMinDistance=false
    Scalability Mode=::NewEnumerator1
    UseLoopDelay=true
   )
   &SpawnRate(SpawnRate=&Multiply_Float(A=7, B=$System.SpawnOn), Use Spawn Probability=false)
  ]
  particleSpawnScript [
   /SetVariables(
    Particles.ColorEnd:LinearColor=&RandomRangeLinearColor(Minimum=<rgba 0.25 0 0.138 1>, Maximum=<rgba 0.0466 0 0.25 1>)
    Particles.ColorStart:LinearColor=&RandomRangeLinearColor(Minimum=<rgba 0.36 0.3 0.45 1>, Maximum=<rgba 0.6 0.68 0.74 1>)
   ) {isSetParameters=true}
   &InitializeParticle(
    Color=$Particles.ColorStart
    Lifetime Max=3
    Lifetime Min=2
    Lifetime Mode=::NewEnumerator1
    Position:Vector3f=$System.LightningEndPoint
    Position Mode=::NewEnumerator1
    Sprite Rotation Mode=::NewEnumerator2
    Sprite Size=&RandomRangeVector2D(Minimum=<150 150>, Maximum=<300 300>, Evaluation Type=::NewEnumerator1)
    Sprite Size Mode=::NewEnumerator4
    UsePositionOffset=false
    Write Material Random=true
   )
   &SphereLocation(
    MaskSpawnGroup=false
    Non Uniform Scale=<0.25 1 1>
    Override Seed=false
    Sphere Radius=&RandomRangeFloat(Minimum=25, Maximum=100)
    Surface Only=false
    UseNonUniformScale=true
    UseOffset=false
    UseSphereOrientation=false
   ) {module=SphereLocation001}
   &AddVelocity(Coordinate Space=::World, Velocity=&RandomRangeVector(Minimum=<-25 -2 -2>, Maximum=<-15 2 2>))
   &DynamicMaterialParameters(
    Index 0 Param 1=&RandomRangeFloat(Minimum=0, Maximum=1, Evaluation Type=::NewEnumerator1)
    Index 0 Param 2=&RandomRangeFloat(Minimum=0.4, Maximum=0.6, Evaluation Type=::NewEnumerator1)
    Index 0 Param 3=0.2
    Module.Param0WriteEnabled=true
    Module.Param1WriteEnabled=true
    Module.Param2WriteEnabled=true
    Module.Param3WriteEnabled=true
   )
   &CameraOffset(Camera Offset Amount=&RandomRangeFloat(Minimum=75, Maximum=125, Evaluation Type=::NewEnumerator1))
  ]
  particleUpdateScript [
   &ParticleState()
   &SolveForcesAndVelocity(Clamp Velocity=false, Limit Acceleration=false)
   &SpriteRotationRate(Rotation Rate=&RandomRangeFloat(Minimum=-20, Maximum=20))
   &ScaleColor(
    Scale Alpha=&FloatFromCurve(
     FloatCurve=DataInterface() {propertyValues {Curve {keys [<at 0 0 RCIM_Cubic>,<at 0.5 1 RCIM_Cubic>,<at 1 0>]}}}
     Scale Curve=&RandomRangeFloat(Minimum=0.8, Maximum=1)
    )
    Scale RGB=&MakeVectorFromLinearColorRGB(Color=&LerpLinearColors(
      StartColor=$Particles.ColorStart
      EndColor=$Particles.ColorEnd
      LerpFactor=&FloatFromCurve(FloatCurve=DataInterface() {propertyValues {Curve {keys [<at 0 0 RCIM_Cubic>, {interpMode=RCIM_Cubic, tangentMode=RCTM_Break, time=0.5, value=1}]}}})
     ))
    ScaleA=true
    ScaleRGB=true
   ) {module=ScaleColor001}
   &ScaleSpriteSize(Uniform Curve Sprite Scale=DataInterface() {propertyValues {Curve {keys [<at 0 0.7 RCIM_Cubic>, {interpMode=RCIM_Cubic, tangentMode=RCTM_Break, time=1, value=1}]}}})
  ]
  rendererClasses [&NiagaraSpriteRendererProperties]
  renderers [{rendererClass=&NiagaraSpriteRendererProperties}]
 }
 Sparkles {
  simTarget=GPUComputeSim
  emitterUpdateScript [
   &EmitterState(Enable Distance Culling=true, MaxDistance=8000, Module.bUseMaxDistance=true, Module.bUseMinDistance=false, Scalability Mode=::NewEnumerator1)
   &SpawnRate(SpawnRate=15, Use Spawn Probability=false)
   &SpawnRate(SpawnGroup=1, SpawnRate=&Multiply_Float(A=75, B=$System.SpawnOn), Use Spawn Probability=false) {module=SpawnRate001}
   &SpawnRate(SpawnGroup=2, SpawnRate=&Multiply_Float(A=$System.BeamLength, B=0.02), Use Spawn Probability=false) {module=SpawnRate002}
  ]
  particleSpawnScript [
   &InitializeParticle(
    Color=<rgba 0.5 0 0.0812 1>
    Lifetime=&Multiply_Float(
     A=&RandomRangeFloat(Minimum=0.5, Maximum=1.5)
     B=&SelectFloatFromArray(
      Array Sampling Mode=::NewEnumerator0
      Direct Array Index=?
      Float Selection Array=DataInterface() {propertyValues {FloatData [1,0.8,0.15]}}
      Random Seed=0
     )
    )
    Mesh Scale Mode=::NewEnumerator4
    Sprite Size Mode=::NewEnumerator1
    Uniform Sprite Size Max=25
    Uniform Sprite Size Min=7
    UsePositionOffset=false
    Write Material Random=true
   )
   &BoxLocation(
    Box Offset=&MakeVector(X=&Multiply_Float(A=$User.CubeRadius, B=-1), Y=0, Z=0)
    Box Size=&VectorFromFloat(Value=&Multiply_Float(A=$User.CubeRadius, B=2))
    MaskSpawnGroup=false
    Module.Override Seed=false
    Rotate Box=false
    Surface Only=false
   )
   /SetVariables(Particles.Position=&self:SelectVector(
     Index:NiagaraInt32=&SpawnGroup()
     DefaultPosition:Vector3f=Linked() {linkedVariable {name=Particles.Position, type=NiagaraPosition}}
     SpawnGroup1:Vector3f=&TransformPosition(Position=$System.LightningEndPoint, Source Space=::Local, Destination Space=::Simulation)
     SpawnGroup2:Vector3f=&Add_Vector(
      A=&Lerp_Vector(
       A=Linked() {linkedVariable {name=Engine.Owner.Position, type=NiagaraPosition}}
       B=Linked() {linkedVariable {name=User.SourcePoint, type=NiagaraPosition}}
       Alpha=&RandomRangeFloat(Minimum=0, Maximum=1)
      )
      B=&RandomVector(Vector Scale=&RandomRangeFloat(Minimum=0, Maximum=150))
     )
    )) {isSetParameters=true}
   &AddVelocityInCone(
    Cone Angle=25
    Cone Axis=<0 0 1>
    Cone Axis Coordinate Space=::World
    Module.Use Velocity Falloff On Cone Axis=false
    Velocity Strength=&MaskFloatBySpawnGroup(Masked Float=&RandomRangeFloat(Minimum=200, Maximum=1200), Passthrough Float=1, Spawn Group Mask=1)
   )
  ]
  particleUpdateScript [
   &ParticleState()
   &DynamicMaterialParameters(
    Index 0 Param 1=1
    Index 0 Param 2=1
    Index 0 Param 3=0.5
    Module.Param0WriteEnabled=true
    Module.Param1WriteEnabled=true
    Module.Param2WriteEnabled=true
    Module.Param3WriteEnabled=true
   )
   &ScaleColor(
    Scale Alpha=&FloatFromCurve(FloatCurve=DataInterface() {propertyValues {Curve {keys [
         <at 0 0 RCIM_Cubic>
         {interpMode=RCIM_Cubic, tangentMode=RCTM_User, time=0.0856, value=0.9914}
         {arriveTangent=-2.3503, interpMode=RCIM_Cubic, leaveTangent=-2.3503, tangentMode=RCTM_User, time=0.8081, value=0.5934}
         <at 1 0>
        ]}}})
    Scale RGB=&VectorFromFloat(Value=&Add_Float(
      A=1
      B=&FloatFromCurve(
       FloatCurve=DataInterface() {propertyValues {Curve {keys [
           {arriveTangent=10.4941, interpMode=RCIM_Cubic, time=0, value=0}
           {
            arriveTangent=-2.7982
            arriveTangentWeight=0.0471
            interpMode=RCIM_Cubic
            leaveTangent=-2.7982
            leaveTangentWeight=0.0707
            tangentMode=RCTM_User
            time=0.3
            value=1
           }
           <at 1 0>
          ]}}}
       CurveIndex=$System.NormalizedLoopAge
       Scale Curve=5
      )
     ))
    ScaleA=true
    ScaleRGB=true
   )
   /SetVariables(Particles.Color2=&MakeLinearColorFromVectorAndFloat(
     "Vector (RGB)"=&MakeVectorFromLinearColorRGB(Color=<rgba 0 1 0.8193 1>)
     "Float (Alpha)"=&MakeFloatFromLinearColor(LinearColor=$Particles.Color, Channel:ENiagaraLinearColor_Channels=::NewEnumerator3)
    )) {isSetParameters=true}
   &GravityForce(Gravity=&RandomRangeVector(Minimum=<-1 -1 50>, Maximum=<1 1 300>))
   &VectorNoiseForce(
    Force Amount=&FloatFromCurve(
     FloatCurve=DataInterface() {propertyValues {Curve {keys [
         {arriveTangent=-2.6497, interpMode=RCIM_Cubic, leaveTangent=-2.6497, tangentMode=RCTM_User, time=0, value=1}
         {interpMode=RCIM_Cubic, tangentMode=RCTM_User, time=1, value=0}
        ]}}}
     Scale Curve=5000
    )
    Module.Use Random Noise Time Interval=true
    Random Noise Time Interval=0.05
   )
   &Drag(Drag=&RandomRangeFloat(Minimum=2, Maximum=8), Use Linear Drag=true, Use Rotational Drag=false)
   &ScaleSpriteSize(Scale Factor=&Vector2DFromFloat(Value=&FloatFromCurve(FloatCurve=DataInterface() {propertyValues {Curve {keys [
          <at 0 0 RCIM_Cubic>
          {arriveTangent=0.836, interpMode=RCIM_Cubic, leaveTangent=0.836, time=0.2, value=1}
          {arriveTangent=-1.8455, interpMode=RCIM_Cubic, leaveTangent=-1.8455, tangentMode=RCTM_User, time=0.6507, value=0.544}
          <at 1 0>
         ]}}})))
   &PointForce(
    Force Falloff Distance=400
    Force Origin=&TransformPosition(Position=$System.LightningEndPoint, Source Space=::Local, Destination Space=::Simulation)
    Force Strength=&Multiply_Float(A=7000, B=$System.SpawnOn)
    Module.Use Falloff=true
    Module.UseLinearForce=false
   )
   &SolveForcesAndVelocity(Clamp Velocity=false, Limit Acceleration=false)
  ]
  rendererClasses [&NiagaraSpriteRendererProperties]
  renderers [
   {rendererClass=&NiagaraSpriteRendererProperties}
   {rendererClass=&NiagaraSpriteRendererProperties, rendererIndex=1}
  ]
 }
}
summary {
 systemName=NS_Corruption_Cube_Spawn_Telegraph
 emitters [
  {
   emitterName=Lightning
   rendererClasses [
    &NiagaraMeshRendererProperties
    &NiagaraSpriteRendererProperties
    &NiagaraLightRendererProperties
   ]
  }
  {emitterName=NE_FN_Sparks, simTarget=GPUComputeSim, rendererClasses [&NiagaraSpriteRendererProperties]}
  {emitterName=ImpactSmoke, rendererClasses [&NiagaraSpriteRendererProperties]}
  {emitterName=Light, rendererClasses [&NiagaraLightRendererProperties, &NiagaraMeshRendererProperties]}
  {emitterName=Shimmers, rendererClasses [&NiagaraSpriteRendererProperties]}
  {emitterName=RunesSource, rendererClasses [&NiagaraSpriteRendererProperties]}
  {emitterName=Sparkles, simTarget=GPUComputeSim, rendererClasses [&NiagaraSpriteRendererProperties]}
  {emitterName=Burns, rendererClasses [&NiagaraSpriteRendererProperties]}
  {emitterName=ElectricSprites, rendererClasses [&NiagaraSpriteRendererProperties]}
  {emitterName=GlowSurface, rendererClasses [&NiagaraSpriteRendererProperties]}
 ]
 userVariables [
  User.CubeRadius {type=NiagaraFloat, defaultValue=384, description=""}
  User.SourceNormal {type=Vector3f, description="", defaultValue {struct=Vector3f, value=<0 -1 0>}}
  User.SourcePoint {type=NiagaraPosition, description="", defaultValue {}}
  User.SourceUp {type=Vector3f, description="", defaultValue {struct=Vector3f, value=<0 0 1>}}
 ]
}
